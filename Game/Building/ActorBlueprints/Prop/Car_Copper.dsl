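; Car_Copper is a car prop building Blueprint with head and tail lights, alarm state and a throttled jump-on sound.
; Construction adds spotlights at mesh sockets and, in debug wind mode, swaps in a preview mesh with materials lerped toward intense-wind versions.
; Helper functions randomize brightness and scale light intensity alongside emissive material parameters.
; A nested-sine hash of grid-snapped location gives a deterministic pseudo-random car alarm decision.
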(fn MaterialEditorSine
  (params (Look%20up%20value :type Float%20%28double-precision%29) (Period :type Float%20%28double-precision%29))
  (return (Math|Trig|Sin(Degrees) (/ (* Look%20up%20value 360) Period))))

(fn PlayJumpOnCarSound
  (params (Location :type Vector))
  (if (>= (- (Utilities|Time|GetGameTimeinSeconds) (Variables|PlayerBounce|GetTimeSinceLastJumpSound)) (Variables|PlayerBounce|GetMinTimeBetweenJumpSound))
    (Audio|PlaySoundatLocation (Variables|PlayerBounce|GetSound%20-%20Jump%20on%20Car) Location "0, 0, 0" 0.25)
    (Variables|PlayerBounce|SetTimeSinceLastJumpSound (Utilities|Time|GetGameTimeinSeconds))))

(fn Randomizethetargetbrightnesslevels
  (bind _output_get (Variables|Headlights|SetFinal%20Random%20Light%20Intensity%20Variance (Math|Random|RandomFloatinRange 0.8 1)))
  (Variables|Headlights|SetHeadLightLightBrightness (* (Variables|Headlights|GetHeadLightLightBrightness) _output_get))
  (Variables|Headlights|SetTailLightBrightness (* (Variables|Headlights|GetTailLightBrightness) _output_get))
  (Variables|Headlights|SetHeadlight%20Emissive%20Brightness (* (Variables|Headlights|GetHeadlight%20Emissive%20Brightness) _output_get))
  (Variables|Alarms|SetSirenLightEmissiveMultiplier (* (Variables|Alarms|GetSirenLightEmissiveMultiplier) _output_get)))

(fn SetLightBrightness
  (params (0-1%20Intensity :type Float%20%28double-precision%29) (Include%20Second%20Emissive%20Channel :type Boolean :default true))
  (bind (_r _g _b) (Math|Color|BreakColor (Variables|Alarms|GetInitialEmissiveMultiplier)))
  (bind (_r_2 _g_2 _b_2) (Math|Color|BreakColor (Variables|Alarms|GetInitialEmissiveMultiplier2)))
  ;[1 1 1] Lights
  (if (Variables|Headlights|GetHeadLightsExist)
    (Rendering|Components|Light|SetIntensity (Variables|Headlights|GetHLight1) (* (Variables|Headlights|GetHeadLightLightBrightness) 0-1%20Intensity)))
  (if (Variables|Headlights|GetTailLightsExist)
    (Rendering|Components|Light|SetIntensity (Variables|Headlights|GetTaillight1) (* (Variables|Headlights|GetTailLightBrightness) 0-1%20Intensity)))
  ;[1 1 1] Animate MIDs
  (bind (_index _mid) (Utilities|LoopThroughMIDS)
    (:loop%20body
      (Rendering|Material|SetVectorParameterValue _mid (Variables|Alarms|GetEmissive%20Multiplier) (Math|Conversions|ToLinearColor(Vector) (* (Math|Vector|MakeVector _r _g _b) (* 0-1%20Intensity (Variables|Headlights|GetHeadlight%20Emissive%20Brightness)))) :!class /Script/Engine.MaterialInstanceDynamic)
      (Rendering|Material|SetVectorParameterValue _mid (Variables|Alarms|GetEmissive%20Multiplier%202) (select Include%20Second%20Emissive%20Channel (Math|Conversions|ToLinearColor(Vector) (* (Math|Vector|MakeVector _r_2 _g_2 _b_2) (Variables|Alarms|GetSirenLightEmissiveMultiplier))) 0) :!class /Script/Engine.MaterialInstanceDynamic))
    (:end%20loop
      (return))))

(fn setcanmakebouncingnoisetrue
  (Variables|PlayerBounce|SetcanMakeJumpingNoiseAgain true))

(fn SetCarAlarmEnabledBasedOnLocation
  (params (Location :type Vector))
  (return (< (Math|Float|Absolute(Float) (Math|Float|Fraction (* (* (CallFunction|MaterialEditorSine (+ (+ (/ (Math|Conversions|ToFloat(Integer) (Math|Float|Ceil (.y (/ (+ Location 5.25) (LocalVariables|GetRandom%20Value%205.59614))))) (LocalVariables|GetRandom%20Value%205.59614)) (* (/ (Math|Conversions|ToFloat(Integer) (Math|Float|Ceil (.x (/ (+ Location 5.25) (LocalVariables|GetRandom%20Value%205.59614))))) (LocalVariables|GetRandom%20Value%205.59614)) 0.04523)) (* (CallFunction|MaterialEditorSine (+ (* (CallFunction|MaterialEditorSine (+ (/ (Math|Conversions|ToFloat(Integer) (Math|Float|Ceil (.x (/ (+ Location 5.25) (LocalVariables|GetRandom%20Value%205.59614))))) (LocalVariables|GetRandom%20Value%205.59614)) (* (/ (Math|Conversions|ToFloat(Integer) (Math|Float|Ceil (.y (/ (+ Location 5.25) (LocalVariables|GetRandom%20Value%205.59614))))) (LocalVariables|GetRandom%20Value%205.59614)) -0.21458)) 0.201836) 0.023) (+ (/ (Math|Conversions|ToFloat(Integer) (Math|Float|Ceil (.x (/ (+ Location 5.25) (LocalVariables|GetRandom%20Value%205.59614))))) (LocalVariables|GetRandom%20Value%205.59614)) (/ (Math|Conversions|ToFloat(Integer) (Math|Float|Ceil (.y (/ (+ Location 5.25) (LocalVariables|GetRandom%20Value%205.59614))))) (LocalVariables|GetRandom%20Value%205.59614)))) 0.145872) 0.01914)) 0.1) (CallFunction|MaterialEditorSine (+ (* (CallFunction|MaterialEditorSine (+ (/ (Math|Conversions|ToFloat(Integer) (Math|Float|Ceil (.x (/ (+ Location 5.25) (LocalVariables|GetRandom%20Value%205.59614))))) (LocalVariables|GetRandom%20Value%205.59614)) (* (/ (Math|Conversions|ToFloat(Integer) (Math|Float|Ceil (.y (/ (+ Location 5.25) (LocalVariables|GetRandom%20Value%205.59614))))) (LocalVariables|GetRandom%20Value%205.59614)) -0.21458)) 0.201836) 0.023) (+ (/ (Math|Conversions|ToFloat(Integer) (Math|Float|Ceil (.x (/ (+ Location 5.25) (LocalVariables|GetRandom%20Value%205.59614))))) (LocalVariables|GetRandom%20Value%205.59614)) (/ (Math|Conversions|ToFloat(Integer) (Math|Float|Ceil (.y (/ (+ Location 5.25) (LocalVariables|GetRandom%20Value%205.59614))))) (LocalVariables|GetRandom%20Value%205.59614)))) 0.145872)) 13.1145769))) (Variables|Alarms|GetPercentageChanceOfCarAlarmsBeingDisabled)) (Math|Float|Absolute(Float) (Math|Float|Fraction (* (* (CallFunction|MaterialEditorSine (+ (+ (/ (Math|Conversions|ToFloat(Integer) (Math|Float|Ceil (.y (/ (+ Location 5.25) (LocalVariables|GetRandom%20Value%205.59614))))) (LocalVariables|GetRandom%20Value%205.59614)) (* (/ (Math|Conversions|ToFloat(Integer) (Math|Float|Ceil (.x (/ (+ Location 5.25) (LocalVariables|GetRandom%20Value%205.59614))))) (LocalVariables|GetRandom%20Value%205.59614)) 0.04523)) (* (CallFunction|MaterialEditorSine (+ (* (CallFunction|MaterialEditorSine (+ (/ (Math|Conversions|ToFloat(Integer) (Math|Float|Ceil (.x (/ (+ Location 5.25) (LocalVariables|GetRandom%20Value%205.59614))))) (LocalVariables|GetRandom%20Value%205.59614)) (* (/ (Math|Conversions|ToFloat(Integer) (Math|Float|Ceil (.y (/ (+ Location 5.25) (LocalVariables|GetRandom%20Value%205.59614))))) (LocalVariables|GetRandom%20Value%205.59614)) -0.21458)) 0.201836) 0.023) (+ (/ (Math|Conversions|ToFloat(Integer) (Math|Float|Ceil (.x (/ (+ Location 5.25) (LocalVariables|GetRandom%20Value%205.59614))))) (LocalVariables|GetRandom%20Value%205.59614)) (/ (Math|Conversions|ToFloat(Integer) (Math|Float|Ceil (.y (/ (+ Location 5.25) (LocalVariables|GetRandom%20Value%205.59614))))) (LocalVariables|GetRandom%20Value%205.59614)))) 0.145872) 0.01914)) 0.1) (CallFunction|MaterialEditorSine (+ (* (CallFunction|MaterialEditorSine (+ (/ (Math|Conversions|ToFloat(Integer) (Math|Float|Ceil (.x (/ (+ Location 5.25) (LocalVariables|GetRandom%20Value%205.59614))))) (LocalVariables|GetRandom%20Value%205.59614)) (* (/ (Math|Conversions|ToFloat(Integer) (Math|Float|Ceil (.y (/ (+ Location 5.25) (LocalVariables|GetRandom%20Value%205.59614))))) (LocalVariables|GetRandom%20Value%205.59614)) -0.21458)) 0.201836) 0.023) (+ (/ (Math|Conversions|ToFloat(Integer) (Math|Float|Ceil (.x (/ (+ Location 5.25) (LocalVariables|GetRandom%20Value%205.59614))))) (LocalVariables|GetRandom%20Value%205.59614)) (/ (Math|Conversions|ToFloat(Integer) (Math|Float|Ceil (.y (/ (+ Location 5.25) (LocalVariables|GetRandom%20Value%205.59614))))) (LocalVariables|GetRandom%20Value%205.59614)))) 0.145872)) 13.1145769)))))

(fn StopCarAlarmFromPlaying
  (Variables|Alarms|SetCarAlarmIsCurrentlyMakingNoise false))

(fn togglelightvisibility
  (params (bNewVisibility :type Boolean))
  (bind _hlight2 (Variables|Headlights|GetHLight2))
  (if (Math|Boolean|ANDBoolean (Utilities|IsValid (Variables|Headlights|GetHLight1) :!class /Script/Engine.KismetSystemLibrary) (Utilities|IsValid _hlight2 :!class /Script/Engine.KismetSystemLibrary) (Utilities|IsValid (Variables|Headlights|GetTaillight1) :!class /Script/Engine.KismetSystemLibrary) (Utilities|IsValid (Variables|Headlights|GetTaillight2) :!class /Script/Engine.KismetSystemLibrary))
    (Rendering|SetVisibility _hlight2 bNewVisibility false))
  (return))

(fn ConstructionScript
  (bind _returnvalue (Building|GetBuildingMeshComponent))
  (bind _returnvalue_2 (Building|GetBuildingMeshComponent))
  (bind _returnvalue_3 (Wind|GetWindSpeedCurveforPreview))
  (bind _returnvalue_4 (Wind|GetWindPannerSpeedCurveforPreview))
  (bind _debug_wind_intensity (Variables|Wind|GetDebug%20Wind%20Intensity))
  (bind _returnvalue_5 (Math|Curves|GetLinearColorValue _returnvalue_3 _debug_wind_intensity))
  (bind _debug_wind_intensity_2 (Variables|Wind|GetDebug%20Wind%20Intensity))
  (bind _returnvalue_6 (Math|Curves|GetLinearColorValue _returnvalue_4 _debug_wind_intensity_2))
  (bind (_r _g _b _a) (Math|Color|BreakColor _returnvalue_5))
  (bind (_r_2 _g_2 _b_2 _a_2) (Math|Color|BreakColor _returnvalue_6))
  (bind _returnvalue_7 (Transformation|GetSocketTransform _returnvalue_2 "TLight2" "RTS_World"))
  (bind _returnvalue_8 (Transformation|GetSocketTransform _returnvalue_2 "TLight1" "RTS_World"))
  (bind _returnvalue_9 (Transformation|GetSocketTransform _returnvalue_2 "HLight1" "RTS_World"))
  (bind _returnvalue_10 (Transformation|GetSocketTransform _returnvalue_2 "HLight2" "RTS_World"))
  (Networking|CanExecuteCosmeticEvents
    (:True
      (Utilities|FlowControl|Sequence
        (:then_0
          ;[0 0 0] HeadLights
          (bind _output_get (Variables|Headlights|SetHeadLightsExist (and (Transformation|DoesSocketExist _returnvalue_2 "HLight1") (Transformation|DoesSocketExist _returnvalue_2 "HLight2"))))
          (if _output_get
            (bind _output_get_2 (Variables|Headlights|SetHLight1SocketLocation (.location _returnvalue_9)))
            (Variables|Headlights|Sethlight1%20-%20impact%20bounce%20light%20location _output_get_2)
            (bind _output_get_3 (Variables|Headlights|SetHLight2SocketLocation (.location _returnvalue_10)))
            (Variables|Headlights|Sethlight2%20-%20impact%20bounce%20light%20location _output_get_3)
            (bind _returnvalue_11 (AddComponent|NODE_AddSpotLightComponent-1))
            (bind _output_get_4 (Variables|Headlights|SetHLight1 _returnvalue_11))
            (Transformation|SetWorldTransform _output_get_4 _returnvalue_9 false false)
            (bind _returnvalue_12 (AddComponent|NODE_AddSpotLightComponent-2))
            (bind _output_get_5 (Variables|Headlights|SetHLight2 _returnvalue_12))
            (Transformation|SetWorldTransform _output_get_5 _returnvalue_10 false false)))
        (:then_1
          ;[0 0 0] TailLights
          (bind _output_get_6 (Variables|Headlights|SetTLight1SocketLocation (.location _returnvalue_8)))
          (Variables|Headlights|Settlight1%20-%20impact%20bounce%20light%20location _output_get_6)
          (bind _output_get_7 (Variables|Headlights|SetTLight2SocketLocation (.location _returnvalue_7)))
          (Variables|Headlights|Settlight2%20-%20impact%20bounce%20light%20location _output_get_7)
          (bind _output_get_8 (Variables|Headlights|SetTailLightsExist (and (Transformation|DoesSocketExist (Building|GetBuildingMeshComponent) "TLight1") (Transformation|DoesSocketExist (Building|GetBuildingMeshComponent) "TLight2"))))
          (if _output_get_8
            (bind _returnvalue_13 (AddComponent|NODE_AddSpotLightComponent-3))
            (bind _output_get_9 (Variables|Headlights|SetTaillight1 _returnvalue_13))
            (Transformation|SetWorldTransform _output_get_9 _returnvalue_8 false false)
            (bind _returnvalue_14 (AddComponent|NODE_AddSpotLightComponent-5))
            (bind _output_get_10 (Variables|Headlights|SetTaillight2 _returnvalue_14))
            (Transformation|SetWorldTransform _output_get_10 _returnvalue_7 false false)))
        (:then_2
          (if (Variables|Wind|GetDebugWind)
            ;[0 0 0] Check to see that the assets are valid
            (Utilities|IsValid (Wind|GetWindSpeedCurveforPreview)
              (:Is%20Valid
                (bind (_materials _returnvalue_15) (Wind|GetIntenseWindMaterialsforPreview))
                (Variables|Wind|SetIntenseWindMaterialsForPreview _materials)
                (if _returnvalue_15
                  ;[0 0 0] Setup mesh alternative to base mesh because code controls the base mesh materials and will not allow us to animate them properly, also the base mesh is persistent
                  (bind _returnvalue_16 (AddComponent|NODE_AddStaticMeshComponent-0))
                  (bind _output_get_11 (Variables|Default|SetWind%20Intensity%20Debug%20Mesh _returnvalue_16))
                  (Components|StaticMesh|SetStaticMesh _output_get_11 (Variables|Mesh|GetStaticMesh))
                  (Rendering|SetVisibility _returnvalue false false)
                  (bind _output_get_12 (Variables|Default|SetOriginalMaterials (Rendering|Material|GetMaterials _output_get_11)))
                  (for (_array_element _array_index) _output_get_12
                    (bind _returnvalue_17 (Rendering|Material|CreateDynamicMaterialInstance (Utilities|Array|Get(acopy) (Rendering|Material|GetMaterials _returnvalue) _array_index) "None" "None" :!class /Script/Engine.KismetMaterialLibrary))
                    (bind _returnvalue_18 (Rendering|Material|CreateDynamicMaterialInstance (Utilities|Array|Get(acopy) (Variables|Wind|GetIntenseWindMaterialsForPreview) _array_index) "None" "None" :!class /Script/Engine.KismetMaterialLibrary))
                    (bind _returnvalue_19 (Rendering|Material|CreateDynamicMaterialInstance (Variables|Default|GetWind%20Intensity%20Debug%20Mesh) _array_index _array_element "None" :!class /Script/Engine.PrimitiveComponent))
                    (Variables|Default|SetDebug_TempMaterial _returnvalue_19)
                    ;[0 0 0] Lerp parameters
                    (for _array_element_2 (LocalVariables|GetParametersToLerp)
                      (bind _returnvalue_20 (Rendering|Material|GetScalarParameterValue _returnvalue_18 _array_element_2 :!class /Script/Engine.MaterialInstanceDynamic))
                      (bind _returnvalue_21 (Rendering|Material|GetScalarParameterValue _returnvalue_17 _array_element_2 :!class /Script/Engine.MaterialInstanceDynamic))
                      (Rendering|Material|SetScalarParameterValue (Variables|Default|GetDebug_TempMaterial) _array_element_2 (Math|Float|Lerp _returnvalue_21 _returnvalue_20 (Variables|Wind|GetDebug%20Wind%20Intensity)) :!class /Script/Engine.MaterialInstanceDynamic))
                    ;[0 0 0 0.25] Set Wind Speed Values
                    (Utilities|IsValid _returnvalue_3
                      (:Is%20Valid
                        (Rendering|Material|SetScalarParameterValue (Variables|Default|GetDebug_TempMaterial) "Wind_1_Speed_CodeReferenced" _r :!class /Script/Engine.MaterialInstanceDynamic)
                        (Rendering|Material|SetScalarParameterValue (Variables|Default|GetDebug_TempMaterial) "Wind_2_Speed_CodeReferenced" _g :!class /Script/Engine.MaterialInstanceDynamic)
                        (Rendering|Material|SetScalarParameterValue (Variables|Default|GetDebug_TempMaterial) "Wind_3_Speed_CodeReferenced" _b :!class /Script/Engine.MaterialInstanceDynamic)
                        (Rendering|Material|SetScalarParameterValue (Variables|Default|GetDebug_TempMaterial) "Wind_4_Speed_CodeReferenced" _a :!class /Script/Engine.MaterialInstanceDynamic)
                        ;[0 0 0 0.25] Set Wind Panner Speed Values
                        (Rendering|Material|SetScalarParameterValue (Variables|Default|GetDebug_TempMaterial) "Wind_Panner_Speed_1x_CodeReferenced" _r_2 :!class /Script/Engine.MaterialInstanceDynamic)
                        (Rendering|Material|SetScalarParameterValue (Variables|Default|GetDebug_TempMaterial) "Wind_Panner_Speed_2x_CodeReferenced" _g_2 :!class /Script/Engine.MaterialInstanceDynamic)
                        (Rendering|Material|SetScalarParameterValue (Variables|Default|GetDebug_TempMaterial) "Wind_Panner_Speed_1y_CodeReferenced" _b_2 :!class /Script/Engine.MaterialInstanceDynamic)
                        (Rendering|Material|SetScalarParameterValue (Variables|Default|GetDebug_TempMaterial) "Wind_Panner_Speed_2y_CodeReferenced" _a_2 :!class /Script/Engine.MaterialInstanceDynamic)
                        ;[0 0 0 0.25] Set Wind Direction
                        (Rendering|Material|SetVectorParameterValue (Variables|Default|GetDebug_TempMaterial) "WindVector_CodeControlled" (Math|Color|MakeColor (.x (Math|Vector|RotateVectorAroundAxis (Math|Vector|MakeVector 1) (Variables|Wind|GetWind%20Yaw%20Rotation) "0, 0,1.000000")) (.y (Math|Vector|RotateVectorAroundAxis (Math|Vector|MakeVector 1) (Variables|Wind|GetWind%20Yaw%20Rotation) "0, 0,1.000000")) (.z (Math|Vector|RotateVectorAroundAxis (Math|Vector|MakeVector 1) (Variables|Wind|GetWind%20Yaw%20Rotation) "0, 0,1.000000"))) :!class /Script/Engine.MaterialInstanceDynamic))))
                  (else
                    (Development|PrintString "The intense wind material list needs to be populated in:"))))
              (:Is%20Not%20Valid
                (Utilities|FlowControl|Sequence
                  (:then_0
                    (Development|PrintString "Please add a wind speed curve reference to:"))
                  (:then_1
                    (Utilities|IsValid (Wind|GetWindPannerSpeedCurveforPreview)
                      (:Is%20Not%20Valid
                        (Development|PrintString "Please add a wind speed curve reference to:")))))))
            (Development|PrintString (Utilities|GetDisplayName self))
            (else
              (Rendering|SetVisibility (Building|GetBuildingMeshComponent) true false))))))))
