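; Default car building prop with head and tail lights, a damage-triggered alarm, jump-on bounce, low-health smoke and random oil decals.
; Construction adds spotlights at mesh sockets, using a sine hash of world location to vary lights and alarm per car, plus debug wind preview.
; Damage flashes lights and emissive MIDs via timelines and alerts AI through noise events.
; Landing impacts drive world-position-offset bounce; BeginPlay staggers Athena setup to reduce hitches.
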(event EventOnBounceAnimationUpdate
  (params (Data :type Fort%20Bounce%20Data%20Structure%20%28by%20ref%29))
  ;[0 0 0] Bounce Light
  (if (Variables|Headlights|GetHeadLightsExist)
    (bind _newworldposition (Utilities|BouncePointinSpace (Variables|Headlights|GetHLight1SocketLocation) Data)
      (:StartBounceAndUpdate
        (Variables|Headlights|Sethlight1%20-%20impact%20bounce%20light%20location _newworldposition)
        (bind _newworldposition_2 (Utilities|BouncePointinSpace (Variables|Headlights|GetHLight2SocketLocation) Data)
          (:StartBounceAndUpdate
            (Variables|Headlights|Sethlight2%20-%20impact%20bounce%20light%20location _newworldposition_2))))))
  (if (Variables|Headlights|GetTailLightsExist)
    (bind _newworldposition_3 (Utilities|BouncePointinSpace (Variables|Headlights|GetTLight1SocketLocation) Data)
      (:StartBounceAndUpdate
        (Variables|Headlights|Settlight1%20-%20impact%20bounce%20light%20location _newworldposition_3)
        (bind _newworldposition_4 (Utilities|BouncePointinSpace (Variables|Headlights|GetTLight2SocketLocation) Data)
          (:StartBounceAndUpdate
            (Variables|Headlights|Settlight2%20-%20impact%20bounce%20light%20location _newworldposition_4)
            (CallFunction|HitBounceFinished)))))
    (else
      (CallFunction|HitBounceFinished))))

(event Custom|Hit%20Bounce%20Finished
  ;[0 0 0] Bounce Light
  (if (Variables|Headlights|GetHeadLightsExist)
    (Transformation|SetWorldLocation (Variables|Headlights|GetHLight1) (+ (Variables|Default|GetBounce%20Offset) (Variables|Headlights|Gethlight1%20-%20impact%20bounce%20light%20location)) false false)
    (Transformation|SetWorldLocation (Variables|Headlights|GetHLight2) (+ (Variables|Default|GetBounce%20Offset) (Variables|Headlights|Gethlight2%20-%20impact%20bounce%20light%20location)) false false))
  (if (Variables|Headlights|GetTailLightsExist)
    (Transformation|SetWorldLocation (Variables|Headlights|GetTaillight1) (+ (Variables|Default|GetBounce%20Offset) (Variables|Headlights|Gettlight1%20-%20impact%20bounce%20light%20location)) false false)
    (Transformation|SetWorldLocation (Variables|Headlights|GetTaillight2) (+ (Variables|Default|GetBounce%20Offset) (Variables|Headlights|Gettlight2%20-%20impact%20bounce%20light%20location)) false false)))

(event Custom|Player%20jump%20based%20bounce%20
  ;[0 0 0] Bounce Light
  (if (Variables|Headlights|GetHeadLightsExist)
    (Transformation|SetWorldLocation (Variables|Headlights|GetHLight1) (+ (Variables|Default|GetBounce%20Offset) (Variables|Headlights|Gethlight1%20-%20impact%20bounce%20light%20location)) false false)
    (Transformation|SetWorldLocation (Variables|Headlights|GetHLight2) (+ (Variables|Default|GetBounce%20Offset) (Variables|Headlights|Gethlight2%20-%20impact%20bounce%20light%20location)) false false))
  (if (Variables|Headlights|GetTailLightsExist)
    (Transformation|SetWorldLocation (Variables|Headlights|GetTaillight1) (+ (Variables|Default|GetBounce%20Offset) (Variables|Headlights|Gettlight1%20-%20impact%20bounce%20light%20location)) false false)
    (Transformation|SetWorldLocation (Variables|Headlights|GetTaillight2) (+ (Variables|Default|GetBounce%20Offset) (Variables|Headlights|Gettlight2%20-%20impact%20bounce%20light%20location)) false false)))

(event EventBeginPlay
  (bind _returnvalue (Math|Random|RandomBool))
  (bind _returnvalue_2 (Math|Float|SelectFloat 0.0 0.5 _returnvalue))
  ;[1 1 1] beginplay - branches off for athena and staggers execution to combat hitches
  (Utilities|FlowControl|Sequence
    (:then_0
      (Networking|CanExecuteCosmeticEvents
        (:True
          (if (Utilities|Enum|Equal(Enum) (Fortnite|GetSubGame) "Athena")
            (Utilities|FlowControl|Delay (Math|Random|RandomFloatinRange 0.5 0.9))
            (else
              (CallFunction|RandomizeTheTargetBrightnessLevels)
              ;; Spawn oil decal
              (if (Transformation|DoesSocketExist (Building|GetBuildingMeshComponent) "Grill")
                (if (Variables|Default|GetCanDropOilDecal)
                  (if (> (Math|Random|RandomFloatinRange 0.0 1) 0.5)
                    (bind _returnvalue_3 (Rendering|Material|CreateDynamicMaterialInstance "/Game/Vehicles/FORT_Vehicles_01/Materials/OilDecal.OilDecal" "None" "None" :!class /Script/Engine.KismetMaterialLibrary))
                    (Rendering|Material|SetScalarParameterValue _returnvalue_3 "XOffset" _returnvalue_2 :!class /Script/Engine.MaterialInstanceDynamic)
                    (Rendering|Material|SetScalarParameterValue _returnvalue_3 "YOffset" _returnvalue_2 :!class /Script/Engine.MaterialInstanceDynamic)
                    (Rendering|Decal|SpawnDecalatLocation _returnvalue_3 (Math|Vector|MakeVector 34 (Math|Float|Lerp 30 75 (Math|Random|RandomFloat)) (Math|Float|Lerp 30 75 (Math|Random|RandomFloat))) (Math|Transform|TransformLocation (Transformation|GetWorldTransform (Building|GetBuildingMeshComponent)) (+ (* "1.000000,1.000000, 0" (Math|Transform|InverseTransformLocation (Transformation|GetWorldTransform (Building|GetBuildingMeshComponent)) (Transformation|GetSocketLocation (Building|GetBuildingMeshComponent) "Grill"))) "0,-150.000000, 0")) (Math|Rotator|MakeRotator (Math|Random|RandomFloatinRange -360 360) -90)))))))
          (Utilities|FlowControl|Sequence
            (:then_0
              (CallFunction|GenerateMeshMIDs))
            (:then_2
              (CallFunction|ValidateLightVisibility)
              (CallFunction|SetLightBrightness (Math|Conversions|ToFloat(Boolean) (Variables|Headlights|GetThis%20Car's%20Lights%20Have%20Been%20Enabled)) false true))))))
    (:then_1
      ;[1 1 1] This handles Valet support for PIE.  When enabled, this prop may be deleted.
      (Helpers|CustomHandleActorStreamedinEditor self))))

(event Custom|FireOffCarAlarm
  (bind _alarmlightcounter (Variables|Alarms|GetAlarmLightCounter))
  (bind _returnvalue (+ _alarmlightcounter 1))
  (bind _suspensionmovementamount (Variables|PlayerBounce|GetSuspensionMovementAmount))
  ;[1 1 1] Car alarm
  (Utilities|FlowControl|Gate false :!label _label
    (:Exit
      (Utilities|FlowControl|Sequence
        (:then_0
          (goto _label Close))
        (:then_1
          ;[1 1 1] Car bounce
          (Networking|CanExecuteCosmeticEvents
            (:True
              (CallFunction|PlayJumponCarSound (select (Math|Vector|Equal(Vector) (stale Collision|EventHit HitLocation) "0, 0, 0" 0.0001) (Transformation|GetActorLocation) (stale Collision|EventHit HitLocation)))
              (Rendering|LOD|SetEvaluateWorldPositionOffset (Variables|Debug|GetStaticMeshComponent) true :!class /Script/Engine.StaticMeshComponent)
              (bind (_direction _vertical_bounce_component) (Timeline|BounceCar PlayFromStart)
                (:Update
                  (Variables|Default|SetBounce%20Offset (* (select (Math|Vector|Equal(Vector) (stale Collision|EventHit HitNormal) "0, 0, 0" 0.0001) "0, 0,1.000000" (stale Collision|EventHit HitNormal)) (* _vertical_bounce_component _suspensionmovementamount)))
                  (for (_ _array_index) (Rendering|Material|GetMaterials (Variables|Debug|GetStaticMeshComponent))
                    (bind _returnvalue_2 (CallFunction|GetorCreateMIDOnDefaultMesh _array_index))
                    (Rendering|Material|SetVectorParameterValue _returnvalue_2 "worldpositionoffset_modulatedbyvertexcolor" (Math|Conversions|ToLinearColor(Vector) (* (select (Math|Vector|Equal(Vector) (stale Collision|EventHit HitNormal) "0, 0, 0" 0.0001) "0, 0,1.000000" (stale Collision|EventHit HitNormal)) (* _vertical_bounce_component _suspensionmovementamount))) :!class /Script/Engine.MaterialInstanceDynamic)
                    (CallFunction|PlayerJumpBasedBounce)))
                (:Finished
                  ; Finished
                  (CallFunction|PlayerJumpBasedBounce)
                  (Rendering|LOD|SetEvaluateWorldPositionOffset (Variables|Debug|GetStaticMeshComponent) false :!class /Script/Engine.StaticMeshComponent))))))
        (:then_2
          (Audio|PlaySoundatLocation (Variables|Damage|GetSound%20-%20Hit%20AutoBody) (Transformation|GetActorLocation))
          (if (Math|Boolean|ANDBoolean (Variables|Alarms|GetCarAlarmEnabled) (Health|HasHealthLeft) (Variables|Alarms|GetUseCarAlarm))
            (if (< (Building|GetBuildingHealthPercentage self) 0.3)
              (if (Variables|Damage|GetParticleEffectHas_NOT_AlreadyBeenTriggered)
                (CallFunction|TriggerLowHealthParticleEffect)))
            (if (Variables|Alarms|GetUse%20Emissive%202)
              (if (not (Variables|Alarms|GetSirenLightOn))
                (Utilities|Time|SetTimerbyFunctionName self "StartSirenLight" 0.5)
                (Utilities|Time|SetTimerbyFunctionName self "FadeSirenLight" 10)))
            (if (and (Variables|Alarms|GetSoundIsRetriggerable) (Utilities|NotEqual(Object) (Variables|Alarms|GetSound%20-%20HitAlarm) 0))
              (bind _returnvalue_3 (Audio|SpawnSoundAttached (Variables|Alarms|GetSound%20-%20HitAlarm) (Building|GetBuildingMeshComponent)))
              (Variables|Default|SetAudibleAlarmSound _returnvalue_3)
              (Utilities|Time|SetTimerbyFunctionName self "StopCarAlarmFromPlaying" (Variables|/Script/Engine.SoundBase|GetDuration (Variables|Alarms|GetSound%20-%20HitAlarm)))
              (Variables|Alarms|SetSoundIsRetriggerable false))
            (Variables|Alarms|SetAlarmLightCounter 0)
            (Utilities|FlowControl|Sequence
              (:then_0
                (Utilities|FlowControl|DoOnce false
                  (:Completed
                    (CallFunction|ToggleLightVisibility true))))
              (:then_1
                (bind (_direction_2 _alarmlightmultiplier) (Timeline|Timeline_0 PlayFromStart)
                  (:Update
                    (Variables|Alarms|SetCurrentFlashAnimationCycleFinished false)
                    (Variables|Alarms|SetOverallAlarmCycleFinished (>= _returnvalue (Variables|Alarms|GetNumberOfTimesTheLightsFlashWhenHit)) :!label _label_2)
                    (CallFunction|SetLightBrightness _alarmlightmultiplier)
                    (if (Variables|Alarms|GetOverallAlarmCycleFinished)
                      (Variables|Alarms|SetSoundIsRetriggerable true)
                      (elif (Variables|Alarms|GetCurrentFlashAnimationCycleFinished)
                        (Timeline|Timeline_0 PlayFromStart))))
                  (:Finished
                    _))
                (Variables|Alarms|SetCurrentFlashAnimationCycleFinished true)
                (Variables|Alarms|SetAlarmLightCounter _returnvalue)
                (goto _label_2)))))))))

(event EventOnDamagePlayEffects
  (params (Damage :type Float%20%28single-precision%29) (DamageTags :type Gameplay%20Tag%20Container%20Structure%20%28by%20ref%29) (Momentum :type Vector :default "0, 0, 0") (HitInfo :type Hit%20Result%20Structure%20%28by%20ref%29) (InstigatedBy :type Fort%20Pawn%20Object%20Reference) (DamageCauser :type Actor%20Object%20Reference) (EffectContext :type Gameplay%20Effect%20Context%20Handle%20Structure))
  (bind _alarmlightcounter (Variables|Alarms|GetAlarmLightCounter))
  (bind _returnvalue (+ _alarmlightcounter 1))
  ;[1 1 1] Car alarm
  (Audio|PlaySoundatLocation (Variables|Damage|GetSound%20-%20Hit%20AutoBody) (Transformation|GetActorLocation))
  (if (Math|Boolean|ANDBoolean (Variables|Alarms|GetCarAlarmEnabled) (Health|HasHealthLeft) (Variables|Alarms|GetUseCarAlarm))
    (if (< (Building|GetBuildingHealthPercentage self) 0.3)
      (if (Variables|Damage|GetParticleEffectHas_NOT_AlreadyBeenTriggered)
        (CallFunction|TriggerLowHealthParticleEffect)))
    (if (Variables|Alarms|GetUse%20Emissive%202)
      (if (not (Variables|Alarms|GetSirenLightOn))
        (Utilities|Time|SetTimerbyFunctionName self "StartSirenLight" 0.5)
        (Utilities|Time|SetTimerbyFunctionName self "FadeSirenLight" 10)))
    (if (and (Variables|Alarms|GetSoundIsRetriggerable) (Utilities|NotEqual(Object) (Variables|Alarms|GetSound%20-%20HitAlarm) 0))
      (bind _returnvalue_2 (Audio|SpawnSoundAttached (Variables|Alarms|GetSound%20-%20HitAlarm) (Building|GetBuildingMeshComponent)))
      (Variables|Default|SetAudibleAlarmSound _returnvalue_2)
      (Utilities|Time|SetTimerbyFunctionName self "StopCarAlarmFromPlaying" (Variables|/Script/Engine.SoundBase|GetDuration (Variables|Alarms|GetSound%20-%20HitAlarm)))
      (Variables|Alarms|SetSoundIsRetriggerable false))
    (Variables|Alarms|SetAlarmLightCounter 0)
    (Utilities|FlowControl|Sequence
      (:then_0
        (Utilities|FlowControl|DoOnce false
          (:Completed
            (CallFunction|ToggleLightVisibility true))))
      (:then_1
        (Timeline|Timeline_0 PlayFromStart)))))

(event Collision|EventHit
  (params (MyComp :type Primitive%20Component%20Object%20Reference) (Other :type Actor%20Object%20Reference) (OtherComp :type Primitive%20Component%20Object%20Reference) (bSelfMoved :type Boolean) (HitLocation :type Vector :default "0, 0, 0") (HitNormal :type Vector :default "0, 0, 0") (NormalImpulse :type Vector :default "0, 0, 0") (Hit :type Hit%20Result%20Structure%20%28by%20ref%29))
  (bind _returnvalue (Math|Vector|Equal(Vector) HitLocation "0, 0, 0" 0.0001))
  (bind _returnvalue_2 (Math|Vector|Equal(Vector) HitNormal "0, 0, 0" 0.0001))
  (bind _returnvalue_3 (Transformation|GetActorLocation))
  (bind _returnvalue_4 (select _returnvalue _returnvalue_3 HitLocation))
  (bind _returnvalue_5 (select _returnvalue_2 "0, 0,1.000000" HitNormal))
  (bind _suspensionmovementamount (Variables|PlayerBounce|GetSuspensionMovementAmount))
  ;[1 1 1] Car bounce
  (if (Variables|PlayerBounce|GetUseCarBounce)
    (if (<= (.z (Transformation|GetComponentVelocity OtherComp)) -100)
      (bind _asfort_player_pawn (Utilities|Casting|CastToFortPlayerPawn Other)
        (:then
          (AI|Perception|MakeNoiseEventatLocation _asfort_player_pawn 1000 _returnvalue_4))
        (:CastFailed
          _))
      (Networking|CanExecuteCosmeticEvents
        (:True
          (CallFunction|PlayJumponCarSound _returnvalue_4)
          (Rendering|LOD|SetEvaluateWorldPositionOffset (Variables|Debug|GetStaticMeshComponent) true :!class /Script/Engine.StaticMeshComponent)
          (Timeline|BounceCar PlayFromStart))))))

(event EventOnDamageServer
  (params (Damage :type Float%20%28single-precision%29) (DamageTags :type Gameplay%20Tag%20Container%20Structure%20%28by%20ref%29) (Momentum :type Vector :default "0, 0, 0") (HitInfo :type Hit%20Result%20Structure%20%28by%20ref%29) (InstigatedBy :type Controller%20Object%20Reference) (DamageCauser :type Actor%20Object%20Reference) (EffectContext :type Gameplay%20Effect%20Context%20Handle%20Structure))
  (bind _returnvalue (Pawn|GetControlledPawn InstigatedBy))
  ;[1 1 1] server side make noise
  (if (Variables|Alarms|GetCarAlarmEnabled)
    (if (Variables|Alarms|GetCarAlarmIsCurrentlyMakingNoise)
      _
      (elif (Utilities|IsValid InstigatedBy :!class /Script/Engine.KismetSystemLibrary)
        (if (Utilities|IsValid _returnvalue :!class /Script/Engine.KismetSystemLibrary)
          (AI|Perception|MakeNoiseEventatLocation _returnvalue 2500 (Transformation|GetActorLocation self)) ; Make a noise if the instigator is valid
        )))
    (Variables|Alarms|SetCarAlarmIsCurrentlyMakingNoise true)
    (Utilities|FlowControl|Delay 0.75)
    (Variables|Alarms|SetCarAlarmIsCurrentlyMakingNoise false)))

(event Custom|StartSirenLight
  (bind (_r _g _b) (Math|Color|BreakColor (Variables|Alarms|GetInitialEmissiveMultiplier2)))
  ;[1 1 1] Siren animation
  (Variables|Alarms|SetSirenLightOn true)
  (bind (_direction _sirenlightmultiplier) (Timeline|Timeline_2 Play)
    (:Update
      (Variables|Alarms|SetSirenLightEmissiveMultiplier _sirenlightmultiplier)
      (bind (_index _mid) (Utilities|LoopThroughMIDS (Variables|Default|GetAnimatingMIDArray))
        (:loop%20body
          (Rendering|Material|SetVectorParameterValue _mid (Variables|Alarms|GetEmissive%20Multiplier%202) (Math|Conversions|ToLinearColor(Vector) (* (Math|Vector|MakeVector _r _g _b) _sirenlightmultiplier)) :!class /Script/Engine.MaterialInstanceDynamic))))))

(event Custom|FadeSirenLight
  (bind (_r _g _b) (Math|Color|BreakColor (Variables|Alarms|GetInitialEmissiveMultiplier2)))
  ;[1 1 1] Siren animation
  (Variables|Alarms|SetSirenLightOn false)
  (Timeline|Timeline_2 Reverse))

(event Custom|TriggerLowHealthParticleEffect
  (bind _returnvalue (Building|GetBuildingMeshComponent))
  ;[1 1 1] fire particle component
  (if (Transformation|DoesSocketExist _returnvalue "Grill")
    (Variables|Damage|SetParticleEffectHas_NOT_AlreadyBeenTriggered false)
    (bind _returnvalue_2 (Effects|Components|ParticleSystem|SpawnFXAtLocation "/Game/Effects/Fort_Effects/World/Destruction/Cars/NS_CarSmoking.NS_CarSmoking" (Transformation|GetSocketLocation _returnvalue "Grill")))
    (Variables|Default|SetEngineSteamParticleEffect _returnvalue_2)))

(event EventEndPlay
  (params (EndPlayReason :type EEndPlayReason%20Enum :default "Destroyed"))
  (bind _audiblealarmsound (Variables|Default|GetAudibleAlarmSound))
  (Components|Activation|Deactivate (Variables|Default|GetEngineSteamParticleEffect))
  (Default|CallOnPlayDeathEffects)
  (Utilities|IsValid _audiblealarmsound
    (:Is%20Valid
      (Audio|Components|Audio|FadeOut _audiblealarmsound 0.0 0.1 "Linear" :!class /Script/Engine.AudioComponent))))

(event EventOnDeathPlayEffects
  (params (Damage :type Float%20%28single-precision%29) (DamageTags :type Gameplay%20Tag%20Container%20Structure%20%28by%20ref%29) (Momentum :type Vector :default "0, 0, 0") (HitInfo :type Hit%20Result%20Structure%20%28by%20ref%29) (InstigatedBy :type Fort%20Pawn%20Object%20Reference) (DamageCauser :type Actor%20Object%20Reference) (EffectContext :type Gameplay%20Effect%20Context%20Handle%20Structure))
  (bind _audiblealarmsound (Variables|Default|GetAudibleAlarmSound))
  (Components|Activation|Deactivate (Variables|Default|GetEngineSteamParticleEffect))
  (Default|CallOnPlayDeathEffects)
  (Utilities|IsValid _audiblealarmsound
    (:Is%20Valid
      (Audio|Components|Audio|FadeOut _audiblealarmsound 0.0 0.1 "Linear" :!class /Script/Engine.AudioComponent))))

(event EventOnDayPhaseChanged
  (params (CurrentDayPhase :type EFortDayPhase%20Enum :default "Morning") (PreviousDayPhase :type EFortDayPhase%20Enum :default "Morning") (bAtCreation :type Boolean))
  (Networking|CanExecuteCosmeticEvents
    (:True
      (if bAtCreation
        _
        (elif (Variables|Headlights|GetThis%20Car's%20Lights%20Have%20Been%20Enabled)
          (Utilities|FlowControl|Delay (Math|Random|RandomFloatinRange 0.01 2.0)) ; Space them out for effect and to help hitching
          (CallFunction|ValidateLightVisibility))))))

(fn ConstructCleanUpLights
  (Utilities|FlowControl|Sequence
    (:then_0
      (Utilities|IsValid (Variables|Headlights|GetHLight1)
        (:Is%20Valid
          (Components|DestroyComponent (Variables|Headlights|GetHLight1))
          (Variables|Headlights|SetHLight1 0))))
    (:then_1
      (Utilities|IsValid (Variables|Headlights|GetHLight2)
        (:Is%20Valid
          (Components|DestroyComponent (Variables|Headlights|GetHLight2))
          (Variables|Headlights|SetHLight2 0))))
    (:then_2
      (Utilities|IsValid (Variables|Headlights|GetTaillight1)
        (:Is%20Valid
          (Components|DestroyComponent (Variables|Headlights|GetTaillight1))
          (Variables|Headlights|SetTaillight1 0))))
    (:then_3
      (Utilities|IsValid (Variables|Headlights|GetTaillight2)
        (:Is%20Valid
          (Components|DestroyComponent (Variables|Headlights|GetTaillight2))
          (Variables|Headlights|SetTaillight2 0))))))

(fn construct_ValidateLightsOnVars
  (Utilities|FlowControl|Sequence
    (:then_0
      (bind _output_get (Variables|Headlights|SetThis%20Car's%20Lights%20Have%20Been%20Enabled (Variables|Headlights|GetTurnHeadLightsOn)))
      (if _output_get
        _
        (elif (Variables|Headlights|GetRandomly%20start%20with%20some%20cars%20lights%20on)
          (bind (_newparam _returnvalue) (CallFunction|SetCarAlarmEnabledBasedonLocation (+ (Transformation|GetActorLocation) "114.246880,82.785645,45.521286")))
          (Variables|Headlights|SetThis%20Car's%20Lights%20Have%20Been%20Enabled (> _returnvalue (Variables|Headlights|GetPercentage%20chance%20of%20lights%20being%20left%20off))))))
    (:then_1
      (CallFunction|SetUsesDayPhaseChange (Variables|Headlights|GetThis%20Car's%20Lights%20Have%20Been%20Enabled)))
    (:then_2
      (return (Variables|Headlights|GetThis%20Car's%20Lights%20Have%20Been%20Enabled)))))

(fn generateMeshMIDs
  (if (> (Rendering|Material|GetNumMaterials (Building|GetBuildingMeshComponent)) 1)
    (for (_array_element _array_index) (Rendering|Material|GetMaterialSlotNames (Building|GetBuildingMeshComponent))
      (if (Utilities|String|Contains (Utilities|String|ToString(Name) _array_element) "light")
        (bind _returnvalue (Rendering|Material|CreateDynamicMaterialInstance (Building|GetBuildingMeshComponent) _array_index 0 "None" :!class /Script/Engine.PrimitiveComponent))
        (Utilities|Array|Add (Variables|Default|GetAnimatingMIDArray) _returnvalue)))
    (else
      (bind _returnvalue_2 (Rendering|Material|CreateDynamicMaterialInstance (Building|GetBuildingMeshComponent) 0 0 "None" :!class /Script/Engine.PrimitiveComponent))
      (Utilities|Array|Add (Variables|Default|GetAnimatingMIDArray) _returnvalue_2))))

(fn getCarMatsForLights
  (bind _returnvalue (Building|GetBuildingMeshComponent))
  (if (> (Utilities|Array|Length (Variables|Default|GetAnimatingMIDArray)) 1)
    ;[1 1 1] Find material slots with "light" in the name. A temporary solution for the S11 cars
    (for (_array_element _array_index) (Rendering|Material|GetMaterialSlotNames _returnvalue)
      (if (Utilities|String|Contains (Utilities|String|ToString(Name) _array_element) "light")
        (bind _asmaterial_instance_dynamic (Utilities|Casting|CastToMaterialInstanceDynamic (Rendering|Material|GetMaterial _returnvalue _array_index)))
        (Utilities|Array|Add (LocalVariables|Getl_finalOutput) _asmaterial_instance_dynamic)))
    (return (LocalVariables|Getl_finalOutput))
    (else
      (return (Variables|Default|GetAnimatingMIDArray)))))

(fn MaterialEditorSine
  (params (Look%20up%20value :type Float%20%28double-precision%29) (Period :type Float%20%28double-precision%29))
  (return (Math|Trig|Sin(Degrees) (/ (* Look%20up%20value 360) Period))))

(fn PlayJumpOnCarSound
  (params (Location :type Vector))
  (if (>= (- (Utilities|Time|GetGameTimeinSeconds) (Variables|PlayerBounce|GetTimeSinceLastJumpSound)) (Variables|PlayerBounce|GetMinTimeBetweenJumpSound))
    (Audio|PlaySoundatLocation (Variables|PlayerBounce|GetSound%20-%20Jump%20on%20Car) Location "0, 0, 0" 0.25)
    (Variables|PlayerBounce|SetTimeSinceLastJumpSound (Utilities|Time|GetGameTimeinSeconds))))

(fn Randomizethetargetbrightnesslevels
  (Variables|Headlights|SetFinal%20Random%20Light%20Intensity%20Variance (Math|Random|RandomFloatinRange 0.8 1))
  (Variables|Headlights|SetHeadLightLightBrightness (* (Variables|Headlights|GetHeadLightLightBrightness) (Variables|Headlights|GetFinal%20Random%20Light%20Intensity%20Variance)))
  (Variables|Headlights|SetTailLightBrightness (* (Variables|Headlights|GetTailLightBrightness) (Variables|Headlights|GetFinal%20Random%20Light%20Intensity%20Variance)))
  (Variables|Headlights|SetHeadlight%20Emissive%20Brightness (* (Variables|Headlights|GetHeadlight%20Emissive%20Brightness) (Variables|Headlights|GetFinal%20Random%20Light%20Intensity%20Variance)))
  (Variables|Alarms|SetSirenLightEmissiveMultiplier (* (Variables|Alarms|GetSirenLightEmissiveMultiplier) (Variables|Headlights|GetFinal%20Random%20Light%20Intensity%20Variance))))

(fn SetLightBrightness
  (params (0-1%20Intensity :type Float%20%28double-precision%29) (Include%20Second%20Emissive%20Channel :type Boolean :default true) (force :type Boolean))
  (bind (_r _g _b) (Math|Color|BreakColor (Variables|Alarms|GetInitialEmissiveMultiplier)))
  (bind (_r_2 _g_2 _b_2) (Math|Color|BreakColor (Variables|Alarms|GetInitialEmissiveMultiplier2)))
  (bind _hlight1 (Variables|Headlights|GetHLight1))
  (bind _taillight1 (Variables|Headlights|GetTaillight1))
  (bind _taillight2 (Variables|Headlights|GetTaillight2))
  (bind _hlight2 (Variables|Headlights|GetHLight2))
  (if (or force (not (Math|Float|NearlyEqual(Float) 0-1%20Intensity (Variables|Default|GetcurrentLightRuntimeIntensity) 0.01)))
    (Variables|Default|SetcurrentLightRuntimeIntensity 0-1%20Intensity)
    (Utilities|FlowControl|Sequence
      (:then_0
        (if (Variables|Headlights|GetHeadLightsExist)
          (Utilities|FlowControl|Sequence
            (:then_0
              (Utilities|IsValid _hlight1
                (:Is%20Valid
                  (Rendering|Components|Light|SetIntensity _hlight1 (* (Variables|Headlights|GetHeadLightLightBrightness) (Variables|Get0-1%20Intensity))))))
            (:then_1
              (Utilities|IsValid _hlight2
                (:Is%20Valid
                  (Rendering|Components|Light|SetIntensity _hlight2 (* (Variables|Headlights|GetHeadLightLightBrightness) (Variables|Get0-1%20Intensity)))))))))
      (:then_1
        (if (Variables|Headlights|GetTailLightsExist)
          (Utilities|FlowControl|Sequence
            (:then_0
              (Utilities|IsValid _taillight1
                (:Is%20Valid
                  (Rendering|Components|Light|SetIntensity _taillight1 (* (Variables|Headlights|GetTailLightBrightness) (Variables|Get0-1%20Intensity))))))
            (:then_1
              (Utilities|IsValid _taillight2
                (:Is%20Valid
                  (Rendering|Components|Light|SetIntensity _taillight2 (* (Variables|Headlights|GetTailLightBrightness) (Variables|Get0-1%20Intensity)))))))))
      (:then_2
        ;[1 1 1] Animate MIDs
        (bind (_index _mid) (Utilities|LoopThroughMIDS (Variables|Default|GetAnimatingMIDArray))
          (:loop%20body
            (Rendering|Material|SetVectorParameterValue _mid (Variables|Alarms|GetEmissive%20Multiplier) (Math|Conversions|ToLinearColor(Vector) (* (Math|Vector|MakeVector _r _g _b) (* (Variables|Get0-1%20Intensity) (Variables|Headlights|GetHeadlight%20Emissive%20Brightness)))) :!class /Script/Engine.MaterialInstanceDynamic)
            (Rendering|Material|SetVectorParameterValue _mid (Variables|Alarms|GetEmissive%20Multiplier%202) (select (Variables|GetInclude%20Second%20Emissive%20Channel) (Math|Conversions|ToLinearColor(Vector) (* (Math|Vector|MakeVector _r_2 _g_2 _b_2) (Variables|Alarms|GetSirenLightEmissiveMultiplier))) 0) :!class /Script/Engine.MaterialInstanceDynamic)
            (Rendering|Material|SetScalarParameterValue _mid "Emissive Val " (Math|Conversions|ToFloat(Boolean) (> (Math|Float|Clamp(Float) (Variables|Get0-1%20Intensity)) 0.0)) :!class /Script/Engine.MaterialInstanceDynamic)))))))

(fn setcanmakebouncingnoisetrue
  (Variables|PlayerBounce|SetcanMakeJumpingNoiseAgain true))

(fn SetCarAlarmEnabledBasedOnLocation
  (params (Location :type Vector))
  (bind _random_value_5_59614 (LocalVariables|GetRandom%20Value%205.59614))
  (bind _returnvalue (+ Location 5.25))
  (bind _returnvalue_2 (/ _returnvalue _random_value_5_59614))
  (bind _y (.y _returnvalue_2))
  (bind _returnvalue_3 (Math|Float|Ceil _y))
  (bind _returnvalue_4 (Math|Conversions|ToFloat(Integer) _returnvalue_3))
  (bind _returnvalue_5 (/ _returnvalue_4 _random_value_5_59614))
  (bind _x (.x _returnvalue_2))
  (bind _returnvalue_6 (Math|Float|Ceil _x))
  (bind _returnvalue_7 (Math|Conversions|ToFloat(Integer) _returnvalue_6))
  (bind _returnvalue_8 (/ _returnvalue_7 _random_value_5_59614))
  (bind _returnvalue_9 (+ _returnvalue_8 _returnvalue_5))
  (bind _returnvalue_10 (* _returnvalue_5 -0.21458))
  (bind _returnvalue_11 (+ _returnvalue_8 _returnvalue_10))
  (bind _returnvalue_12 (CallFunction|MaterialEditorSine _returnvalue_11 0.201836))
  (bind _returnvalue_13 (* _returnvalue_12 0.023))
  (bind _returnvalue_14 (+ _returnvalue_13 _returnvalue_9))
  (bind _returnvalue_15 (CallFunction|MaterialEditorSine _returnvalue_14 0.145872))
  (bind _returnvalue_16 (* _returnvalue_15 0.01914))
  (bind _returnvalue_17 (* _returnvalue_8 0.04523))
  (bind _returnvalue_18 (+ _returnvalue_5 _returnvalue_17))
  (bind _returnvalue_19 (+ _returnvalue_18 _returnvalue_16))
  (bind _returnvalue_20 (CallFunction|MaterialEditorSine _returnvalue_19 0.1))
  (bind _returnvalue_21 (* _returnvalue_20 _returnvalue_15))
  (bind _returnvalue_22 (* _returnvalue_21 13.114577))
  (bind _returnvalue_23 (Math|Float|Fraction _returnvalue_22))
  (bind _returnvalue_24 (Math|Float|Absolute(Float) _returnvalue_23))
  (bind _output_get (Variables|Alarms|SetCarAlarmEnabled (and (Variables|Alarms|GetUseCarAlarm) (> _returnvalue_24 (Variables|Alarms|GetPercentageChanceOfCarAlarmsBeingDisabled)))))
  (return _output_get _returnvalue_24))

(fn StopCarAlarmFromPlaying
  (Variables|Alarms|SetCarAlarmIsCurrentlyMakingNoise false))

(fn togglelightvisibility
  (params (bNewVisibility :type Boolean))
  (bind _taillight1 (Variables|Headlights|GetTaillight1))
  (bind _hlight1 (Variables|Headlights|GetHLight1))
  (bind _hlight2 (Variables|Headlights|GetHLight2))
  (bind _taillight2 (Variables|Headlights|GetTaillight2))
  (Utilities|FlowControl|Sequence
    (:then_0
      (Utilities|IsValid _hlight1
        (:Is%20Valid
          (Rendering|SetVisibility _hlight1 bNewVisibility false))))
    (:then_1
      (Utilities|IsValid _hlight2
        (:Is%20Valid
          (Rendering|SetVisibility _hlight2 bNewVisibility false))))
    (:then_2
      (Utilities|IsValid _taillight1
        (:Is%20Valid
          (Rendering|SetVisibility _taillight1 bNewVisibility false))))
    (:then_3
      (Utilities|IsValid _taillight2
        (:Is%20Valid
          (Rendering|SetVisibility _taillight2 bNewVisibility false))))))

(fn ConstructionScript
  (bind _returnvalue (Building|GetBuildingMeshComponent))
  (bind _returnvalue_2 (Building|GetBuildingMeshComponent))
  (bind _returnvalue_3 (Wind|GetWindSpeedCurveforPreview))
  (bind _returnvalue_4 (Wind|GetWindPannerSpeedCurveforPreview))
  (bind _debug_wind_intensity (Variables|Wind|GetDebug%20Wind%20Intensity))
  (bind _returnvalue_5 (Math|Curves|GetLinearColorValue _returnvalue_3 _debug_wind_intensity))
  (bind _debug_wind_intensity_2 (Variables|Wind|GetDebug%20Wind%20Intensity))
  (bind _returnvalue_6 (Math|Curves|GetLinearColorValue _returnvalue_4 _debug_wind_intensity_2))
  (bind (_r _g _b _a) (Math|Color|BreakColor _returnvalue_5))
  (bind (_r_2 _g_2 _b_2 _a_2) (Math|Color|BreakColor _returnvalue_6))
  (Networking|CanExecuteCosmeticEvents
    (:True
      (Utilities|FlowControl|Sequence
        (:then_0
          (bind _doeshavelights (CallFunction|ConstructValidateLightsonVars))
          (if _doeshavelights
            (Utilities|FlowControl|Sequence
              (:then_0
                ;[0 0 0] HeadLights
                (bind _output_get (Variables|Headlights|SetHeadLightsExist (and (Transformation|DoesSocketExist _returnvalue_2 "HLight1") (Transformation|DoesSocketExist _returnvalue_2 "HLight2"))))
                (if _output_get
                  (bind _output_get_2 (Variables|Headlights|SetHLight1SocketLocation (.location (Transformation|GetSocketTransform _returnvalue_2 "HLight1" "RTS_World"))))
                  (Variables|Headlights|Sethlight1%20-%20impact%20bounce%20light%20location _output_get_2)
                  (bind _output_get_3 (Variables|Headlights|SetHLight2SocketLocation (.location (Transformation|GetSocketTransform _returnvalue_2 "HLight2" "RTS_World"))))
                  (Variables|Headlights|Sethlight2%20-%20impact%20bounce%20light%20location _output_get_3)
                  (bind _returnvalue_7 (AddComponent|NODE_AddSpotLightComponent-1 :RelativeTransform (Transformation|GetSocketTransform _returnvalue_2 "HLight1" "RTS_Component")))
                  (Variables|Headlights|SetHLight1 _returnvalue_7)
                  (bind _returnvalue_8 (AddComponent|NODE_AddSpotLightComponent-2 :RelativeTransform (Transformation|GetSocketTransform _returnvalue_2 "HLight2" "RTS_Component")))
                  (Variables|Headlights|SetHLight2 _returnvalue_8)))
              (:then_1
                ;[0 0 0] TailLights
                (bind _output_get_4 (Variables|Headlights|SetTLight1SocketLocation (.location (Transformation|GetSocketTransform _returnvalue_2 "TLight1" "RTS_World"))))
                (Variables|Headlights|Settlight1%20-%20impact%20bounce%20light%20location _output_get_4)
                (bind _output_get_5 (Variables|Headlights|SetTLight2SocketLocation (.location (Transformation|GetSocketTransform _returnvalue_2 "TLight2" "RTS_World"))))
                (Variables|Headlights|Settlight2%20-%20impact%20bounce%20light%20location _output_get_5)
                (bind _output_get_6 (Variables|Headlights|SetTailLightsExist (and (Transformation|DoesSocketExist (Building|GetBuildingMeshComponent) "TLight1") (Transformation|DoesSocketExist (Building|GetBuildingMeshComponent) "TLight2"))))
                (if _output_get_6
                  (bind _returnvalue_9 (AddComponent|NODE_AddSpotLightComponent-3 :RelativeTransform (Transformation|GetSocketTransform _returnvalue_2 "TLight1" "RTS_Component")))
                  (Variables|Headlights|SetTaillight1 _returnvalue_9)
                  (bind _returnvalue_10 (AddComponent|NODE_AddSpotLightComponent-5 :RelativeTransform (Transformation|GetSocketTransform _returnvalue_2 "TLight2" "RTS_Component")))
                  (Variables|Headlights|SetTaillight2 _returnvalue_10))))
            (else
              (CallFunction|ConstructCleanUpLights))))
        (:then_1
          ;[1 1 1] Debug Wind
          (if (Variables|Wind|GetDebugWind)
            ;;[0 0 0] Check to see that the assets are valid
            (Utilities|IsValid (Wind|GetWindSpeedCurveforPreview)
              (:Is%20Valid
                (bind (_materials _returnvalue_11) (Wind|GetIntenseWindMaterialsforPreview))
                (Variables|Wind|SetIntenseWindMaterialsForPreview _materials)
                (if _returnvalue_11
                  ;;[0 0 0] Setup mesh alternative to base mesh because code controls the base mesh materials and will not allow us to animate them properly, also the base mesh is persistent
                  (bind _returnvalue_12 (AddComponent|NODE_AddStaticMeshComponent-0))
                  (bind _output_get_7 (Variables|Default|SetWind%20Intensity%20Debug%20Mesh _returnvalue_12))
                  (Components|StaticMesh|SetStaticMesh _output_get_7 (Variables|Mesh|GetStaticMesh))
                  (Rendering|SetVisibility _returnvalue false false)
                  (bind _output_get_8 (Variables|Default|SetOriginalMaterials (Rendering|Material|GetMaterials _output_get_7)))
                  (for (_array_element _array_index) _output_get_8
                    (bind _returnvalue_13 (Rendering|Material|CreateDynamicMaterialInstance (Utilities|Array|Get(acopy) (Rendering|Material|GetMaterials _returnvalue) _array_index) "None" "None" :!class /Script/Engine.KismetMaterialLibrary))
                    (bind _returnvalue_14 (Rendering|Material|CreateDynamicMaterialInstance (Utilities|Array|Get(acopy) (Variables|Wind|GetIntenseWindMaterialsForPreview) _array_index) "None" "None" :!class /Script/Engine.KismetMaterialLibrary))
                    (bind _returnvalue_15 (Rendering|Material|CreateDynamicMaterialInstance (Variables|Default|GetWind%20Intensity%20Debug%20Mesh) _array_index _array_element "None" :!class /Script/Engine.PrimitiveComponent))
                    (Variables|Default|SetDebug_TempMaterial _returnvalue_15)
                    ;;[0 0 0] Lerp parameters
                    (for _array_element_2 (LocalVariables|GetParametersToLerp)
                      (bind _returnvalue_16 (Rendering|Material|GetScalarParameterValue _returnvalue_14 _array_element_2 :!class /Script/Engine.MaterialInstanceDynamic))
                      (bind _returnvalue_17 (Rendering|Material|GetScalarParameterValue _returnvalue_13 _array_element_2 :!class /Script/Engine.MaterialInstanceDynamic))
                      (Rendering|Material|SetScalarParameterValue (Variables|Default|GetDebug_TempMaterial) _array_element_2 (Math|Float|Lerp _returnvalue_17 _returnvalue_16 (Variables|Wind|GetDebug%20Wind%20Intensity)) :!class /Script/Engine.MaterialInstanceDynamic))
                    ;;[0 0 0 0.25] Set Wind Speed Values
                    (Utilities|IsValid _returnvalue_3
                      (:Is%20Valid
                        (Rendering|Material|SetScalarParameterValue (Variables|Default|GetDebug_TempMaterial) "Wind_1_Speed_CodeReferenced" _r :!class /Script/Engine.MaterialInstanceDynamic)
                        (Rendering|Material|SetScalarParameterValue (Variables|Default|GetDebug_TempMaterial) "Wind_2_Speed_CodeReferenced" _g :!class /Script/Engine.MaterialInstanceDynamic)
                        (Rendering|Material|SetScalarParameterValue (Variables|Default|GetDebug_TempMaterial) "Wind_3_Speed_CodeReferenced" _b :!class /Script/Engine.MaterialInstanceDynamic)
                        (Rendering|Material|SetScalarParameterValue (Variables|Default|GetDebug_TempMaterial) "Wind_4_Speed_CodeReferenced" _a :!class /Script/Engine.MaterialInstanceDynamic)
                        ;;[0 0 0 0.25] Set Wind Panner Speed Values
                        (Rendering|Material|SetScalarParameterValue (Variables|Default|GetDebug_TempMaterial) "Wind_Panner_Speed_1x_CodeReferenced" _r_2 :!class /Script/Engine.MaterialInstanceDynamic)
                        (Rendering|Material|SetScalarParameterValue (Variables|Default|GetDebug_TempMaterial) "Wind_Panner_Speed_2x_CodeReferenced" _g_2 :!class /Script/Engine.MaterialInstanceDynamic)
                        (Rendering|Material|SetScalarParameterValue (Variables|Default|GetDebug_TempMaterial) "Wind_Panner_Speed_1y_CodeReferenced" _b_2 :!class /Script/Engine.MaterialInstanceDynamic)
                        (Rendering|Material|SetScalarParameterValue (Variables|Default|GetDebug_TempMaterial) "Wind_Panner_Speed_2y_CodeReferenced" _a_2 :!class /Script/Engine.MaterialInstanceDynamic)
                        ;;[0 0 0 0.25] Set Wind Direction
                        (Rendering|Material|SetVectorParameterValue (Variables|Default|GetDebug_TempMaterial) "WindVector_CodeControlled" (Math|Color|MakeColor (.x (Math|Vector|RotateVectorAroundAxis (Math|Vector|MakeVector 1) (Variables|Wind|GetWind%20Yaw%20Rotation) "0, 0,1.000000")) (.y (Math|Vector|RotateVectorAroundAxis (Math|Vector|MakeVector 1) (Variables|Wind|GetWind%20Yaw%20Rotation) "0, 0,1.000000")) (.z (Math|Vector|RotateVectorAroundAxis (Math|Vector|MakeVector 1) (Variables|Wind|GetWind%20Yaw%20Rotation) "0, 0,1.000000"))) :!class /Script/Engine.MaterialInstanceDynamic))))
                  (else
                    (Development|PrintString "The intense wind material list needs to be populated in:"))))
              (:Is%20Not%20Valid
                (Utilities|FlowControl|Sequence
                  (:then_0
                    (Development|PrintString "Please add a wind speed curve reference to:"))
                  (:then_1
                    (Utilities|IsValid (Wind|GetWindPannerSpeedCurveforPreview)
                      (:Is%20Not%20Valid
                        (Development|PrintString "Please add a wind speed curve reference to:")))))))
            (Development|PrintString (Utilities|GetDisplayName self))
            (else
              (Rendering|SetVisibility (Building|GetBuildingMeshComponent) true false))))))))

(fn validateLightVisibility
  (if (Variables|Headlights|GetThis%20Car's%20Lights%20Have%20Been%20Enabled)
    (LocalVariables|Setl_validatedTimeOfDay (select (and (> (Math|Conversions|ToInteger(Byte) (TimeOfDay|GetCurrentDayPhase :!class /Script/FortniteGame.FortKismetLibrary)) -1) (< (Math|Conversions|ToInteger(Byte) (TimeOfDay|GetCurrentDayPhase :!class /Script/FortniteGame.FortKismetLibrary)) 4)) (TimeOfDay|GetCurrentDayPhase :!class /Script/FortniteGame.FortKismetLibrary) "Day"))
    (switch Utilities|FlowControl|Switch|SwitchonEFortDayPhase (LocalVariables|Getl_validatedTimeOfDay)
      (:Morning
        _)
      (:Day
        _)
      (:Evening
        (CallFunction|ToggleLightVisibility true :!label _label)
        (bind (_index _mid) (Utilities|LoopThroughMIDS (Variables|Default|GetAnimatingMIDArray) :!label _label_2)
          (:loop%20body
            (Rendering|Material|SetScalarParameterValue _mid "PassedInTimeOfDay" (Utilities|Select :Index (LocalVariables|Getl_validatedTimeOfDay) :Morning 0.0 :Day 0.0 :Evening 1.0 :Night 1.0) :!class /Script/Engine.MaterialInstanceDynamic))))
      (:Night
        (goto _label)))
    (CallFunction|ToggleLightVisibility false)
    (goto _label_2)))
