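; Burger food truck car prop Blueprint that adds an emissive glow change to the light visibility toggle inherited from its parent.
; After the parent call, it sets the static mesh materials' "Emissive Multiplier" to the Alarm On or Alarm Off color, based on the new visibility.
; A FlipFlop with only its A branch wired gates this, so the material updates on every other call.
; The construction script only calls the parent.
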(event Eventtogglelightvisibility
  (params (bNewVisibility :type Boolean))
  (|CallParentFunction bNewVisibility)
  (Utilities|FlowControl|FlipFlop
    (:A
      (Rendering|Material|SetVectorParameterValueonMaterials (Variables|Debug|GetStaticMeshComponent) "Emissive Multiplier" (select bNewVisibility (Math|Conversions|ToVector(LinearColor) (Variables|Default|GetAlarm%20On)) (Math|Conversions|ToVector(LinearColor) (Variables|Default|GetAlarm%20Off)))))))

(fn ConstructionScript
  (|CallParentFunction))
