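; Pizza food truck car prop Blueprint that extends a parent class's light-visibility toggle with a matching emissive change.
; The toggle event calls the parent, then sets the "Emissive Multiplier" vector parameter on the static mesh's materials.
; That value comes from either the Alarm On or Alarm Off color, chosen by the new visibility flag.
; The construction script only calls the parent.
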(event Eventtogglelightvisibility
  (params (bNewVisibility :type Boolean))
  (|CallParentFunction bNewVisibility)
  (Rendering|Material|SetVectorParameterValueonMaterials (Variables|Debug|GetStaticMeshComponent) "Emissive Multiplier" (Math|Conversions|ToVector(LinearColor) (select bNewVisibility (Variables|Default|GetAlarm%20Off) (Variables|Default|GetAlarm%20On)))))

(fn ConstructionScript
  (|CallParentFunction))
