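; Battle bus prop for Athena warmup that fires its booster as the game phase changes and reskins itself for calendar events.
; GetReady starts flames, audio, shakes, rumble and a timeline driving glow, spotlight and force fields (spawned outside mobile/performance mode); BusLocked stops it.
; Enabled booster collision launches overlapping pawns where cosmetic events cannot execute.
; The active event's RepNotify async-loads materials from parallel per-event arrays, then applies them.
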(event EventBeginPlay
  (|CallParentFunction)
  (Utilities|FlowControl|Sequence
    (:then_0
      (bind _asfort_game_mode_athena (Utilities|Casting|CastToFortGameModeAthena (Game|GetGameMode)))
      (Utilities|String|LogString "WarmupBus: begin setup")
      (MapInfo|BindEventtoOnMapInfoInitialized _asfort_game_mode_athena (EventDispatchers|CreateEvent :!event Custom|CheckForCalendarEvents)))
    (:then_1
      (Class|FortAsyncActionGameStateReady|GameStateReadyAsync
        (:OnReady
          (Collision|SetCollisionEnabled (Variables|Default|GetBoosterCollision) "NoCollision")
          (bind _returnvalue (GameplayEventRouter|GetEventRouter "/Script/FortniteGame.FortGameState" self))
          (bind (_asynctaskproxy _payload) (GameplayEventRouter|StartListeningToStatefulEvent _returnvalue "/Script/FortniteGame.FortGamePhaseStepUpdatedEvent")
            (:then
              (Variables|Default|SetGamePhaseStepUpdatedAsyncTask _asynctaskproxy))
            (:OnEventReceived
              _)
            (:OnSavedEventStateExists
              _))
          (CallFunction|GamePhaseStepChanged (Utilities|Struct|BreakFortGamePhaseStepUpdatedEvent _payload)))))))

(event Custom|TestBoosters
  (bind _booster (Variables|Default|GetBooster))
  (bind _returnvalue (Transformation|GetWorldLocation _booster))
  (bind _booster_2 (Variables|Default|GetBooster))
  (bind _returnvalue_2 (Transformation|GetWorldLocation _booster_2))
  (bind _booster_3 (Variables|Default|GetBooster))
  (bind _boosterforce (Variables|Default|GetBoosterForce))
  (bind _spotlight (Variables|Default|GetSpotLight))
  (Components|Activation|Activate (Variables|Default|GetFX_BoosterFlames) true)
  (Collision|SetCollisionEnabled (Variables|Default|GetBoosterCollision) "QueryOnly")
  (Audio|SpawnSoundatLocation "/Game/Athena/Sounds/Vehicles/Bus/Athena_Bus_Warp_Thruster_Activate_A_Cue.Athena_Bus_Warp_Thruster_Activate_A_Cue" _returnvalue_2)
  (bind _returnvalue_3 (Audio|SpawnSoundatLocation "/Game/Athena/Sounds/Vehicles/Bus/Athena_Bus_Warp_Warmup_Phase_A_Cue.Athena_Bus_Warp_Warmup_Phase_A_Cue" _returnvalue_2))
  (Variables|Default|SetThrusterAudio _returnvalue_3)
  (CallFunction|StartCameraShakeBus)
  (Utilities|FlowControl|Sequence
    (:then_0
      ;[1 0.02535 0] Set to FX MI for the warmup booster material
      (Rendering|Material|SetMaterial (Variables|Default|GetBooster) 0 (Variables|Default|GetBoosterWarmupMat))
      (Rendering|Material|SetMaterial (Variables|Default|GetBooster) 1 (Variables|Default|GetBoosterWarmupMat)))
    (:then_1
      (bind (_direction _buswarmupdata _forceintensity) (Timeline|Battlebus_BoosterWarmup PlayFromStart)
        (:Update
          (Utilities|FlowControl|Sequence
            (:then_0
              (Variables|/Game/Effects/Niagara/Blueprints/PhsyicsFields/FN_WashForce.FN_WashForce_C|SetStrength (* _forceintensity 2.0) (Variables|Default|GetBoosterForce))
              (Rendering|Material|SetScalarParameterValueonMaterials _booster_3 "EmissiveBrightness" (* (.x _buswarmupdata) 1e+01))
              (Rendering|Material|SetVectorParameterValueonMaterials _booster_3 "EmissiveColorRedChannel" (Math|Conversions|ToVector(LinearColor) (Math|Color|Lerp(LinearColor) "(R=0.000000,G=0.937500,B=0.919598,A=1.000000)" "(R=1.000000,G=0.000000,B=0.911479,A=1.000000)" (.y _buswarmupdata)))))
            (:then_1
              (Rendering|Components|Light|SetIntensity _spotlight (Math|Float|Lerp 0.0 3.2e+02 (.y _buswarmupdata)))
              (Rendering|Lighting|SetAttenuationRadius _spotlight (Math|Float|Lerp 0.0 4e+03 (.y _buswarmupdata)))
              (Rendering|Components|Light|SetLightColor _spotlight (Math|Color|Lerp(LinearColor) "(R=0.047743,G=0.520517,B=1.000000,A=1.000000)" "(R=1.000000,G=0.072272,B=0.577580,A=1.000000)" (.y _buswarmupdata)) true))))
        (:Finished
          (bind _boosterforce_2 (Variables|Default|GetBoosterForce)
            (:then
              (Actor|DestroyActor _boosterforce_2))
            (:else)))
        (:BusBoostTrigger
          (Audio|PlaySoundatLocation "/Game/Athena/Sounds/Vehicles/Bus/Athena_Bus_Warp_Warmup_Phase_C_Cue.Athena_Bus_Warp_Warmup_Phase_C_Cue" _returnvalue)
          (CallFunction|FXonBoostersKickIn)
          (if (not (or (Class|PlatformHelpersLibrary|IsPcPerfMode) (TouchPlatform|IsMobileGame)))
            (bind _returnvalue_4 (Game|SpawnActorfromClass "/Game/Effects/Niagara/Blueprints/PhsyicsFields/FN_RadialForce.FN_RadialForce_C" (Transformation|GetActorTransform) "Undefined" "MultiplyWithRoot" 0 1.0 2.0 7e+03 1.5 0.0 0))
            (Variables|Default|SetBoosterForceBurst _returnvalue_4)
            (Transformation|AddActorLocalOffset (Variables|Default|GetBoosterForceBurst) "0.000000,-1000.000000,0.000000" false false)))
        (:Audio%20Bus%20Booster%20B
          (Audio|PlaySoundatLocation "/Game/Athena/Sounds/Vehicles/Bus/Athena_Bus_Warp_Warmup_Phase_B_Cue.Athena_Bus_Warp_Warmup_Phase_B_Cue" _returnvalue))))
    (:then_2
      (if (not (or (Class|PlatformHelpersLibrary|IsPcPerfMode) (TouchPlatform|IsMobileGame)))
        (bind _returnvalue_5 (Game|SpawnActorfromClass "/Game/Effects/Niagara/Blueprints/PhsyicsFields/FN_WashForce.FN_WashForce_C" (Transformation|GetActorTransform) "Undefined" "MultiplyWithRoot" 0 4.5e+03 0.0 -999.0 0))
        (Variables|Default|SetBoosterForce _returnvalue_5)
        (Transformation|AddActorLocalOffset _boosterforce "0.000000,-1000.000000,0.000000" false false)
        (Actor|SetLifeSpan _boosterforce 11.0)))))

(event EventEndPlay
  (params (EndPlayReason :type EEndPlayReason%20Enum :default "Destroyed"))
  (|CallParentFunction EndPlayReason)
  (bind _gamephasestepupdatedasynctask (Variables|Default|GetGamePhaseStepUpdatedAsyncTask)
    (:then
      (GameplayEventRouter|StopListeningtoStatefulEvent _gamephasestepupdatedasynctask))
    (:else)))

(event Custom|Game%20Phase%20Step%20Changed
  (params (GamePhaseStep :type EAthenaGamePhaseStep%20Enum))
  (bind _booster (Variables|Default|GetBooster))
  (bind _returnvalue (Transformation|GetWorldLocation _booster))
  (bind _booster_2 (Variables|Default|GetBooster))
  (bind _returnvalue_2 (Transformation|GetWorldLocation _booster_2))
  (bind _booster_3 (Variables|Default|GetBooster))
  (bind _boosterforce (Variables|Default|GetBoosterForce))
  (bind _spotlight (Variables|Default|GetSpotLight))
  (switch Utilities|FlowControl|Switch|SwitchonEAthenaGamePhaseStep GamePhaseStep
    (:GetReady
      (Components|Activation|Activate (Variables|Default|GetFX_BoosterFlames) true)
      (Collision|SetCollisionEnabled (Variables|Default|GetBoosterCollision) "QueryOnly")
      (Audio|SpawnSoundatLocation "/Game/Athena/Sounds/Vehicles/Bus/Athena_Bus_Warp_Thruster_Activate_A_Cue.Athena_Bus_Warp_Thruster_Activate_A_Cue" _returnvalue_2)
      (bind _returnvalue_3 (Audio|SpawnSoundatLocation "/Game/Athena/Sounds/Vehicles/Bus/Athena_Bus_Warp_Warmup_Phase_A_Cue.Athena_Bus_Warp_Warmup_Phase_A_Cue" _returnvalue_2))
      (Variables|Default|SetThrusterAudio _returnvalue_3)
      (CallFunction|StartCameraShakeBus)
      (Utilities|FlowControl|Sequence
        (:then_0
          ;[1 0.02535 0] Set to FX MI for the warmup booster material
          (Rendering|Material|SetMaterial (Variables|Default|GetBooster) 0 (Variables|Default|GetBoosterWarmupMat))
          (Rendering|Material|SetMaterial (Variables|Default|GetBooster) 1 (Variables|Default|GetBoosterWarmupMat)))
        (:then_1
          (Timeline|Battlebus_BoosterWarmup PlayFromStart))
        (:then_2
          (if (not (or (Class|PlatformHelpersLibrary|IsPcPerfMode) (TouchPlatform|IsMobileGame)))
            (bind _returnvalue_4 (Game|SpawnActorfromClass "/Game/Effects/Niagara/Blueprints/PhsyicsFields/FN_WashForce.FN_WashForce_C" (Transformation|GetActorTransform) "Undefined" "MultiplyWithRoot" 0 4.5e+03 0.0 -999.0 0))
            (Variables|Default|SetBoosterForce _returnvalue_4)
            (Transformation|AddActorLocalOffset _boosterforce "0.000000,-1000.000000,0.000000" false false)
            (Actor|SetLifeSpan _boosterforce 11.0)))))
    (:BusLocked
      (Components|Activation|Deactivate (Variables|Default|GetFX_BoosterFlames))
      (Components|DestroyComponent (Variables|Default|GetBooster_Forcefeedback))
      (bind _thrusteraudio (Variables|Default|GetThrusterAudio)
        (:then
          (Audio|Components|Audio|FadeOut _thrusteraudio 2.0 0.0 "Linear" :!class /Script/Engine.AudioComponent)
          (CallFunction|StopCameraShakebus)
          (Timeline|Battlebus_BoosterWarmup Stop))
        (:else)))))

(event Custom|StopBoosters
  (bind _booster (Variables|Default|GetBooster))
  (bind _returnvalue (Transformation|GetWorldLocation _booster))
  (bind _booster_2 (Variables|Default|GetBooster))
  (bind _spotlight (Variables|Default|GetSpotLight))
  (Components|Activation|Deactivate (Variables|Default|GetFX_BoosterFlames))
  (Components|DestroyComponent (Variables|Default|GetBooster_Forcefeedback))
  (bind _thrusteraudio (Variables|Default|GetThrusterAudio)
    (:then
      (Audio|Components|Audio|FadeOut _thrusteraudio 2.0 0.0 "Linear" :!class /Script/Engine.AudioComponent)
      (CallFunction|StopCameraShakebus)
      (Timeline|Battlebus_BoosterWarmup Stop))
    (:else)))

(event Custom|StartCameraShakeBus
  (CameraShake|Start (Variables|Default|GetEngineCameraShake))
  (bind _returnvalue (ForceFeedback|SpawnForceFeedbackatLocation :ForceFeedbackEffect "/Game/Athena/Aircraft/BattleBus_Booster/FF_PartyBus_Booster.FF_PartyBus_Booster" :Location (Transformation|GetActorLocation) :bLooping false :IntensityMultiplier 1.0 :StartTime 0.0 :AttenuationSettings "/Game/Athena/Aircraft/BattleBus_Booster/FF_PartyBus_Booster_Attenuation.FF_PartyBus_Booster_Attenuation" :bAutoDestroy true))
  (Variables|Default|SetBooster_Forcefeedback _returnvalue))

(event Custom|FX_OnBoostersKickIn
  (CameraShake|Start (Variables|Default|GetEngineFiredUpCameraShake))
  (ForceFeedback|SpawnForceFeedbackatLocation :ForceFeedbackEffect "/Game/Weapons/ForceFeedback/Explosions/FF_Generic_ExplosionLarge_Far.FF_Generic_ExplosionLarge_Far" :Location (Transformation|GetActorLocation) :bLooping false :IntensityMultiplier 2.0 :StartTime 0.0 :AttenuationSettings "/Game/Athena/Aircraft/BattleBus_Booster/FF_PartyBus_Booster_Attenuation.FF_PartyBus_Booster_Attenuation" :bAutoDestroy true))

(event OnComponentBeginOverlap(BoosterCollision)
  (params (OverlappedComponent :type Primitive%20Component%20Object%20Reference) (OtherActor :type Actor%20Object%20Reference) (OtherComp :type Primitive%20Component%20Object%20Reference) (OtherBodyIndex :type Integer) (bFromSweep :type Boolean) (SweepResult :type Hit%20Result%20Structure%20%28by%20ref%29))
  (bind _asfort_player_pawn (Utilities|Casting|CastToFortPlayerPawn OtherActor))
  (Networking|CanExecuteCosmeticEvents
    (:False
      (Character|LaunchCharacter _asfort_player_pawn (Math|Vector|MakeVector (.x (* (Math|Vector|Normalize (- (Transformation|GetActorLocation OtherActor) (Transformation|GetWorldLocation (Variables|Default|GetBoosterCollision)))) 1.5e+03)) (.y (* (Math|Vector|Normalize (- (Transformation|GetActorLocation OtherActor) (Transformation|GetWorldLocation (Variables|Default|GetBoosterCollision)))) 1.5e+03)) (Math|Float|Max(Float) (+ (.z (* (Math|Vector|Normalize (- (Transformation|GetActorLocation OtherActor) (Transformation|GetWorldLocation (Variables|Default|GetBoosterCollision)))) 1.5e+03)) 0.0) 5e+02)) true true)
      (GameplayCue|ExecuteGameplayCue _asfort_player_pawn (Variables|Default|GetGCN_ThrusterImpulse) (Class|AthenaFunctionLibrary|MakeSimpleFortEffectContext self self)))))

(event Custom|StopCameraShakebus
  (CameraShake|StopAllCameraShakes (Variables|Default|GetEngineCameraShake) true)
  (ForceFeedback|Stop (Variables|Default|GetBooster_Forcefeedback)))

(event Custom|DebugWarmupBoosters
  ;[1 1 1] Debug for testing boosters
  (Utilities|FlowControl|Sequence
    (:then_0
      (CallFunction|TestBoosters))
    (:then_1
      (Utilities|FlowControl|Delay 11.0)
      (CallFunction|StopBoosters))))

(event Custom|CheckForCalendarEvents
  ;[0.3 0 1] See if any tracked calendar events are active
  ; NOTE: these strings and arrays are still hardcoded. Would be great to load this from BBID in native.
  (Utilities|String|LogString "WarmupBus: mapinfo ready")
  (bind _array_element (Utilities|Array|ForEachLoopwithBreak (Variables|Default|GetEventKey) :!label _label)
    (:LoopBody
      (if (CallFunction|CalendarEventIsActive _array_element)
        (Utilities|FlowControl|Sequence
          (:then_0
            (goto _label Break))
          (:then_1
            (Variables|Default|SetCalendar_Event _array_element)))))))

(event Custom|TryLoadMaterials
  ;[0.3 0 1] Try Update Materials - ensure that all the required materials are loaded, and async load them if not, before trying to update the materials
  (Utilities|CheckorLoadMaterial (CallFunction|GetMaterialSoftRefforCurrentEvent (Variables|Default|GetBasketMaterial))
    (:Loaded
      (Utilities|CheckorLoadMaterial (CallFunction|GetMaterialSoftRefforCurrentEvent (Variables|Default|GetBusMaterial))
        (:Loaded
          (Utilities|CheckorLoadMaterial (CallFunction|GetMaterialSoftRefforCurrentEvent (Variables|Default|GetCannisterMaterial))
            (:Loaded
              (Utilities|CheckorLoadMaterial (CallFunction|GetMaterialSoftRefforCurrentEvent (Variables|Default|GetBoosterMaterial))
                (:Loaded
                  (Utilities|CheckorLoadMaterial (CallFunction|GetMaterialSoftRefforCurrentEvent (Variables|Default|GetBoosterWarmUpMaterial))
                    (:Loaded
                      (CallFunction|UpdateBusMaterials))))))))))))

(fn CalendarEventIsActive
  (params (EventName :type String))
  (bind _biseventactive (CalendarEvents|GetCalendarEventInformation EventName))
  (return (or _biseventactive (and (Variables|Default|GetbUseDebugCalendarEvent) (Utilities|String|EqualExactly(String) EventName (Variables|Default|GetDebugCalendarEvent))))))

(fn GetMaterialForCurrentEvent
  (params (Array :type Array%20of%20Material%20Interface%20Soft%20Object%20References))
  (bind _calendar_event (Variables|Default|GetCalendar_Event))
  (bind _eventkey (Variables|Default|GetEventKey))
  (bind _returnvalue (Utilities|Array|FindItem _eventkey _calendar_event))
  (if (!= _returnvalue -1)
    (return (Utilities|ResolveSoftReference (Utilities|Array|Get(acopy) Array _returnvalue)))
    (else
      (return 0))))

(fn GetMaterialSoftRefForCurrentEvent
  (params (MaterialArray :type Array%20of%20Material%20Interface%20Soft%20Object%20References))
  (bind _calendar_event (Variables|Default|GetCalendar_Event))
  (bind _eventkey (Variables|Default|GetEventKey))
  (bind _returnvalue (Utilities|Array|FindItem _eventkey _calendar_event))
  (if (!= _returnvalue -1)
    (return (Utilities|Array|Get(acopy) MaterialArray _returnvalue))
    (else
      (return 0))))

(fn OnRep_Calendar_Event
  (Utilities|String|LogString "WarmupBus: Event found")
  (CallFunction|TryLoadMaterials))

(fn UpdateBusMaterials
  (bind _staticmeshcomponent (Variables|Debug|GetStaticMeshComponent))
  ;[1 1 1] Use index from current calendar event to find the right materials.
  (Utilities|String|LogString "WarmupBus: attempting to set materials")
  (Rendering|Material|SetMaterial (Variables|Default|GetBooster) 0 (CallFunction|GetMaterialforCurrentEvent (Variables|Default|GetBoosterMaterial)))
  (Rendering|Material|SetMaterial _staticmeshcomponent 0 (CallFunction|GetMaterialforCurrentEvent (Variables|Default|GetBasketMaterial)))
  (Rendering|Material|SetMaterial _staticmeshcomponent 1 (CallFunction|GetMaterialforCurrentEvent (Variables|Default|GetBusMaterial)))
  (Rendering|Material|SetMaterial (Variables|Default|GetBattlebus_Booster_Canister_L) 0 (CallFunction|GetMaterialforCurrentEvent (Variables|Default|GetCannisterMaterial)))
  (Variables|Default|SetBoosterWarmupMat (CallFunction|GetMaterialforCurrentEvent (Variables|Default|GetBoosterWarmUpMaterial))))

(fn ConstructionScript
  (|CallParentFunction))
