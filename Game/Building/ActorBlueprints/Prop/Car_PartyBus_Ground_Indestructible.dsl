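; Indestructible warmup party bus prop that reskins its mesh and materials for calendar events.
; When playlist data is ready, BeginPlay checks hardcoded event keys (with debug override) to set Calendar Event, whose OnRep starts asset loading.
; The key's index selects soft references from parallel arrays; the mesh loads alongside a chain of six material loads.
; UpdateBusMaterials then swaps the mesh, enables collision and assigns materials to slot indices hardcoded per mesh.
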(event EventBeginPlay
  (Class|FortAsyncActionCurrentPlaylistReady|PlaylistDataReadyAsync
    (:OnReady
      ;[0.3 0 1] See if any tracked calendar events are active
      ; NOTE: these strings and arrays are still hardcoded. Would be great to load this from BBID in native.
      (Utilities|String|LogString "WarmupBus: mapinfo ready")
      (bind _array_element (Utilities|Array|ForEachLoopwithBreak (Variables|EventTracking|GetEventKey))
        (:LoopBody
          (if (Default|CalendarEventIsActive _array_element)
            (Utilities|FlowControl|Sequence
              (:then_0
                _)
              (:then_1
                _))
            (Variables|EventTracking|SetCalendar%20Event _array_element)))))))

(event Custom|TryLoadMaterials
  ;[0.3 0 1] Try Update Materials - ensure that all the required materials are loaded, and async load them if not, before trying to update the materials
  (Utilities|FlowControl|Sequence
    (:then_0
      (Default|CheckorLoadObject (Default|GetMeshSoftRefForCurrentEvent (Variables|EventBusses|GetBusMesh))))
    (:then_1
      (Default|CheckorLoadObject (Default|GetMaterialSoftRefforCurrentEvent (Variables|EventBusses|GetBasketMaterial))
        (:Loaded
          (Default|CheckorLoadObject (Default|GetMaterialSoftRefforCurrentEvent (Variables|EventBusses|GetBusMaterial))
            (:Loaded
              (Default|CheckorLoadObject (Default|GetMaterialSoftRefforCurrentEvent (Variables|EventBusses|GetWindowMaterial))
                (:Loaded
                  (Default|CheckorLoadObject (Default|GetMaterialSoftRefforCurrentEvent (Variables|EventBusses|GetUndercarriageMaterial))
                    (:Loaded
                      (Default|CheckorLoadObject (Default|GetMaterialSoftRefforCurrentEvent (Variables|EventBusses|GetFrameworkMaterial))
                        (:Loaded
                          (Default|CheckorLoadObject (Default|GetMaterialSoftRefforCurrentEvent (Variables|EventBusses|GetMobileMaterial))
                            (:Loaded
                              (Default|UpdateBusMaterials))))))))))))))) ; Check & load mesh first
)

(fn CalendarEventIsActive
  (params (EventName :type String))
  (bind _biseventactive (CalendarEvents|GetCalendarEventInformation EventName))
  (return (or _biseventactive (and (Variables|EventTracking|GetbUseDebugCalendarEvent) (Utilities|String|EqualExactly(String) EventName (Variables|EventTracking|GetDebugCalendarEvent))))))

(fn GetMaterialForCurrentEvent
  (params (Array :type Array%20of%20Material%20Interface%20Soft%20Object%20References))
  (bind _calendar_event (Variables|EventTracking|GetCalendar%20Event))
  (bind _eventkey (Variables|EventTracking|GetEventKey))
  (bind _returnvalue (Utilities|Array|FindItem _eventkey _calendar_event))
  (if (!= _returnvalue -1)
    (return (Utilities|ResolveSoftReference (Utilities|Array|Get(acopy) Array _returnvalue)))
    (else
      (return 0))))

(fn GetMaterialSoftRefForCurrentEvent
  (params (MaterialArray :type Array%20of%20Material%20Interface%20Soft%20Object%20References))
  (bind _calendar_event (Variables|EventTracking|GetCalendar%20Event))
  (bind _eventkey (Variables|EventTracking|GetEventKey))
  (bind _returnvalue (Utilities|Array|FindItem _eventkey _calendar_event))
  (if (!= _returnvalue -1)
    (return (Utilities|Array|Get(acopy) MaterialArray _returnvalue))
    (else
      (return 0))))

(fn GetMeshForCurrentEvent
  (params (Array :type Array%20of%20Static%20Mesh%20Soft%20Object%20References))
  (bind _calendar_event (Variables|EventTracking|GetCalendar%20Event))
  (bind _eventkey (Variables|EventTracking|GetEventKey))
  (bind _returnvalue (Utilities|Array|FindItem _eventkey _calendar_event))
  (if (!= _returnvalue -1)
    (return (Utilities|ResolveSoftReference (Utilities|Array|Get(acopy) Array _returnvalue)))
    (else
      (return 0))))

(fn GetMeshSoftRefforCurrentEvent
  (params (MeshArray :type Array%20of%20Static%20Mesh%20Soft%20Object%20References))
  (bind _calendar_event (Variables|EventTracking|GetCalendar%20Event))
  (bind _eventkey (Variables|EventTracking|GetEventKey))
  (bind _returnvalue (Utilities|Array|FindItem _eventkey _calendar_event))
  (if (!= _returnvalue -1)
    (return (Utilities|Array|Get(acopy) MeshArray _returnvalue))
    (else
      (return 0))))

(fn OnRep_CalendarEvent
  (Utilities|String|LogString "WarmupBus: Event found")
  (CallFunction|TryLoadMaterials))

(fn Setmeshindices
  (params (Bus%20Mesh :type Static%20Mesh%20Object%20Reference))
  ; Setting index based on bus mesh. In the future, would like to use an Enum for this if event busses are still hardcoded.
  (if (== Bus%20Mesh "/Game/Athena/Aircraft/BattleBus_Mark_III/Meshes/Battlebus_Mark_III_ENV.Battlebus_Mark_III_ENV")
    (Variables|MatIndex|SetBasket%20Index 0)
    (Variables|MatIndex|SetBase%20Index 1) ; Using new Env mesh
    (Variables|MatIndex|SetFramework%20Index 2)
    (Variables|MatIndex|SetWindow%20Index 4)
    (Variables|MatIndex|SetUndercarriage%20Index 3)
    (Variables|MatIndex|SetMobile%20Index 5)
    (return)
    (elif (== Bus%20Mesh "/Game/Athena/Aircraft/PartyBus_Birthday/Meshes/Birthday_7_Vehicle_ENV.Birthday_7_Vehicle_ENV")
      (Variables|MatIndex|SetBasket%20Index 0) ; Using Birthday mesh
      (Variables|MatIndex|SetBase%20Index 1)
      (Variables|MatIndex|SetFramework%20Index 2)
      (Variables|MatIndex|SetWindow%20Index 4)
      (Variables|MatIndex|SetUndercarriage%20Index 3)
      (Variables|MatIndex|SetMobile%20Index 6)
      (return)
      (else
        (Variables|MatIndex|SetBasket%20Index 6) ; Using default bus
        (Variables|MatIndex|SetBase%20Index 0)
        (Variables|MatIndex|SetFramework%20Index 5)
        (Variables|MatIndex|SetWindow%20Index 1)
        (Variables|MatIndex|SetUndercarriage%20Index 3)
        (Variables|MatIndex|SetMobile%20Index 8)
        (return)))))

(fn UpdateBusMaterials
  (bind _mobilematerial (Variables|EventBusses|GetMobileMaterial))
  (bind _material (Default|GetMaterialforCurrentEvent _mobilematerial))
  (bind _basketmaterial (Variables|EventBusses|GetBasketMaterial))
  (bind _material_2 (Default|GetMaterialforCurrentEvent _basketmaterial))
  (bind _undercarriagematerial (Variables|EventBusses|GetUndercarriageMaterial))
  (bind _material_3 (Default|GetMaterialforCurrentEvent _undercarriagematerial))
  (bind _windowmaterial (Variables|EventBusses|GetWindowMaterial))
  (bind _material_4 (Default|GetMaterialforCurrentEvent _windowmaterial))
  (bind _busmesh (Variables|EventBusses|GetBusMesh))
  (bind _mesh (Default|GetMeshforCurrentEvent _busmesh))
  (bind _busmaterial (Variables|EventBusses|GetBusMaterial))
  (bind _material_5 (Default|GetMaterialforCurrentEvent _busmaterial))
  (bind _frameworkmaterial (Variables|EventBusses|GetFrameworkMaterial))
  (bind _material_6 (Default|GetMaterialforCurrentEvent _frameworkmaterial))
  (bind _staticmeshcomponent (Variables|StaticMeshActor|GetStaticMeshComponent))
  (bind _staticmeshcomponent_2 (Variables|StaticMeshActor|GetStaticMeshComponent))
  ;[1 1 1] Use index from current calendar event to find the right materials.
  ;; Swapping the mesh and updating collision
  (Utilities|IsValid _mesh
    (:Is%20Valid
      (Utilities|String|LogString "WarmupBus: attempting to set mesh")
      (Components|StaticMesh|SetStaticMesh (Variables|StaticMeshActor|GetStaticMeshComponent) _mesh)
      (Utilities|IsValid _staticmeshcomponent_2
        (:Is%20Valid
          (Collision|SetCollisionEnabled _staticmeshcomponent_2 "QueryAndPhysics"))))
    (:Is%20Not%20Valid
      _))
  (CallFunction|SetMeshIndices _mesh) ; Sets the values for material indices based on bus mesh
  ;; Setting all the materials
  (Utilities|String|LogString "WarmupBus: attempting to set materials")
  (Utilities|IsValid _material_2
    (:Is%20Valid
      (Rendering|Material|SetMaterial _staticmeshcomponent (Variables|MatIndex|GetBasket%20Index) _material_2)
      (Utilities|IsValid _material_5
        (:Is%20Valid
          (Rendering|Material|SetMaterial _staticmeshcomponent (Variables|MatIndex|GetBase%20Index) _material_5)
          (Utilities|IsValid _material_4
            (:Is%20Valid
              (Rendering|Material|SetMaterial _staticmeshcomponent (Variables|MatIndex|GetWindow%20Index) _material_4)
              (Utilities|IsValid _material_3
                (:Is%20Valid
                  (Rendering|Material|SetMaterial _staticmeshcomponent (Variables|MatIndex|GetUndercarriage%20Index) _material_3)
                  (Utilities|IsValid _material_6
                    (:Is%20Valid
                      (Rendering|Material|SetMaterial _staticmeshcomponent (Variables|MatIndex|GetFramework%20Index) _material_6)
                      (Utilities|IsValid _material
                        (:Is%20Valid
                          (Rendering|Material|SetMaterial _staticmeshcomponent (Variables|MatIndex|GetMobile%20Index) _material))))))))))))))

(fn ConstructionScript
  (|CallParentFunction))
