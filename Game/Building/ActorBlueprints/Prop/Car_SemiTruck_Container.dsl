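; Car_SemiTruck_Container is a building prop Blueprint for a semi-truck container.
; On BeginPlay it sets its static mesh component to affect navigation generation, so the AI navmesh accounts for the container.
; The construction script is empty, and the file shows no other logic.
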(event EventBeginPlay
  (Fortnite|AI|SetComponentAffectNavigationGeneration (Variables|Debug|GetStaticMeshComponent) true))

(fn ConstructionScript)
