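; CobWebs01 is a cobweb building prop that destroys itself when a pawn touches it or when it is hit.
; An overlap on its box collision casts the other actor to FortPawn before destroying the cobweb, while the Hit event destroys it with no check.
; The construction script calls the parent, then sizes and positions the box collision from extent and offset variables.
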(event OnComponentBeginOverlap(BoxCollision)
  (params (OverlappedComponent :type Primitive%20Component%20Object%20Reference) (OtherActor :type Actor%20Object%20Reference) (OtherComp :type Primitive%20Component%20Object%20Reference) (OtherBodyIndex :type Integer) (bFromSweep :type Boolean) (SweepResult :type Hit%20Result%20Structure%20%28by%20ref%29))
  (Utilities|Casting|CastToFortPawn OtherActor)
  (Actor|DestroyActor))

(event Collision|EventHit
  (params (MyComp :type Primitive%20Component%20Object%20Reference) (Other :type Actor%20Object%20Reference) (OtherComp :type Primitive%20Component%20Object%20Reference) (bSelfMoved :type Boolean) (HitLocation :type Vector :default "0, 0, 0") (HitNormal :type Vector :default "0, 0, 0") (NormalImpulse :type Vector :default "0, 0, 0") (Hit :type Hit%20Result%20Structure%20%28by%20ref%29))
  (Actor|DestroyActor))

(fn ConstructionScript
  (bind _box_collision (Variables|CobWebs01|GetBox%20Collision))
  (|CallParentFunction)
  (Components|Box|SetBoxExtent _box_collision (Variables|Default|GetCollisionBoxExtents) true)
  (Transformation|SetRelativeLocation _box_collision (Variables|Default|GetCollisionBoxOffset) false false))
