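; Farm weather vane building prop whose rooster mesh sways and reacts to bounce animation.
; On BeginPlay, off mobile platforms, it issues two eased MoveComponentTo calls that yaw the rooster slightly to opposite sides, with no loop.
; On bounce animation updates, it lazily creates and caches a dynamic material instance for the rooster as SecondaryMID, then passes the bounce data to that material's parameters.
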(event EventOnBounceAnimationUpdate
  (params (Data :type Fort%20Bounce%20Data%20Structure%20%28by%20ref%29))
  (bind _roostermesh (Variables|Farm_WeatherVane_01|GetRoosterMesh))
  (if (Utilities|Equal(Object) (Variables|Default|GetSecondaryMID) 0)
    (bind _returnvalue (Rendering|Material|CreateDynamicMaterialInstance _roostermesh 0 (Rendering|Material|GetMaterial _roostermesh 0) "None" :!class /Script/Engine.PrimitiveComponent))
    (Rendering|Material|SetMaterial _roostermesh 0 _returnvalue)
    (Variables|Default|SetSecondaryMID _returnvalue))
  (Utilities|SetModelMaterialBounceParameters (Variables|Default|GetSecondaryMID) Data))

(event EventBeginPlay
  (bind _roostermesh (Variables|Farm_WeatherVane_01|GetRoosterMesh))
  (if (not (Fortnite|IsMobilePlatform))
    (Components|MoveComponentTo _roostermesh "0, 0, 0" "0,5.000000, 0" true true 2 true)
    (Components|MoveComponentTo _roostermesh "0, 0, 0" "0,-5.000000, 0" true true 2 true)))

(fn ConstructionScript)
