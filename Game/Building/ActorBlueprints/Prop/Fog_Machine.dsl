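; Fog machine building prop that cycles fog on and off until destroyed, with tunable color, opacity, size and spray rate.
; Unless turned off, the server waits a random start delay, then toggles IsActive on random intervals padded by the fade timeline length.
; OnRep_IsActive has remote clients toggle smoke spray and audio, and fade the fog plane's opacity and erosion through a reversible timeline.
; Construction applies fog settings; editor events preview both states.
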(event Custom|EditorActivate
  ;[1 1 1] [debug] editor events
  (Variables|Default|SetIsActive true)
  (CallFunction|HandleActive))

(event Custom|EditorDeactivate
  ;[1 1 1] [debug] editor events
  (Variables|Default|SetIsActive false)
  (CallFunction|HandleActive))

(event Custom|ServerCycleActive
  (bind _timeline_fade (Variables|Fog_Machine|GetTimeline_Fade))
  (bind _returnvalue (Components|Timeline|GetTimelineLength _timeline_fade))
  ;[1 1 1] [server] cycle active/inactive
  (if (Variables|Default|GetIsDead)
    _
    (elif (Variables|Default|GetIsActive)
      (Variables|Default|SetIsActive false)
      (Utilities|FlowControl|Delay (Math|Random|RandomFloatinRange (+ (Variables|Default|GetMinActiveDuration) _returnvalue) (+ _returnvalue (Variables|Default|GetMaxActiveDuration))))
      (else
        (Variables|Default|SetIsActive true)
        (Utilities|FlowControl|Delay (Math|Random|RandomFloatinRange (+ (Variables|Default|GetMinInactiveDuration) _returnvalue) (+ _returnvalue (Variables|Default|GetMaxInactiveDuration)))))))
  (CallFunction|ServerCycleActive))

(event EventBeginPlay
  ;[1 0.2031 0.2031] [server] beginplay
  (if (not (Variables|Default|GetPropIsTurnedOff))
    (Networking|SwitchHasAuthority
      (:Authority
        (Utilities|FlowControl|Delay (Math|Random|RandomFloatinRange (Variables|Default|GetMinInitialStartDelay) (Variables|Default|GetMaxInitialStartDelay)))
        (CallFunction|ServerCycleActive)))))

(event EventOnDeathServer
  (params (Damage :type Float%20%28single-precision%29) (DamageTags :type Gameplay%20Tag%20Container%20Structure%20%28by%20ref%29) (Momentum :type Vector :default "0, 0, 0") (HitInfo :type Hit%20Result%20Structure%20%28by%20ref%29) (InstigatedBy :type Controller%20Object%20Reference) (DamageCauser :type Actor%20Object%20Reference) (EffectContext :type Gameplay%20Effect%20Context%20Handle%20Structure))
  ;[1 1 1] [server] handle death
  (Variables|Default|SetIsDead true)
  (Variables|Default|SetIsActive false))

(event Custom|HandleActive
  (bind _sm_fog_machine_fogplane (Variables|Default|GetSM_Fog_Machine_FogPlane))
  ;[1 1 1] [client] handle SFX/VFX
  (Components|Activation|SetActive (Variables|Default|GetNS_FogMachine_SmokeSpray) (Variables|Default|GetIsActive) false)
  (if (Variables|Default|GetIsActive)
    (Development|SetHiddeninGame (Variables|Default|GetSM_Fog_Machine_FogPlane) false false)
    (bind (_direction _opacity _erode) (Timeline|Timeline_Fade PlayFromStart)
      (:Update
        (Rendering|Material|SetScalarParameterValueonMaterials _sm_fog_machine_fogplane "FogOpacity" _opacity)
        (Rendering|Material|SetScalarParameterValueonMaterials _sm_fog_machine_fogplane "FogErode" _erode))
      (:Finished
        (if (Utilities|Enum|Equal(Enum) _direction "Backward")
          (Development|SetHiddeninGame (Variables|Default|GetSM_Fog_Machine_FogPlane) true false))))
    (else
      (Timeline|Timeline_Fade ReverseFromEnd))))

(event EventOnDeathPlayEffects
  (params (Damage :type Float%20%28single-precision%29) (DamageTags :type Gameplay%20Tag%20Container%20Structure%20%28by%20ref%29) (Momentum :type Vector :default "0, 0, 0") (HitInfo :type Hit%20Result%20Structure%20%28by%20ref%29) (InstigatedBy :type Fort%20Pawn%20Object%20Reference) (DamageCauser :type Actor%20Object%20Reference) (EffectContext :type Gameplay%20Effect%20Context%20Handle%20Structure))
  (Audio|SpawnSoundatLocation "/Game/Athena/Sounds/Props/Prop_Fog_Destroy_Cue.Prop_Fog_Destroy_Cue" (Transformation|GetActorLocation))
  (Audio|Components|Audio|Stop (Variables|Default|GetAudio)))

(fn OnRep_FogMachineState)

(fn OnRep_IsActive
  (Networking|SwitchHasAuthority
    (:Remote
      (CallFunction|HandleActive)
      (if (Variables|Default|GetIsActive)
        (Audio|Components|Audio|Play (Variables|Default|GetAudio) 0.0)
        (else
          (Audio|PlaySoundatLocation "/Game/Athena/Sounds/Props/Prop_Fog_End_Cue.Prop_Fog_End_Cue" (Transformation|GetActorLocation))
          (Audio|Components|Audio|Stop (Variables|Default|GetAudio)))))))

(fn ConstructionScript
  (Utilities|FlowControl|Sequence
    (:then_0
      (bind _returnvalue (Rendering|Material|CreateDynamicMaterialInstance (Variables|Default|GetSM_Fog_Machine_FogPlane) 0 0 "None" :!class /Script/Engine.PrimitiveComponent))
      (Variables|Default|SetFOG%20MID _returnvalue)
      (Rendering|Material|SetScalarParameterValue (Variables|Default|GetFOG%20MID) "FogOpacity" (Variables|FogBaseSetting|GetFogOpacity) :!class /Script/Engine.MaterialInstanceDynamic)
      (Rendering|Material|SetScalarParameterValue (Variables|Default|GetFOG%20MID) "WPO_Height" (Variables|FogBaseSetting|GetFogWPO_Height) :!class /Script/Engine.MaterialInstanceDynamic)
      (Rendering|Material|SetVectorParameterValue (Variables|Default|GetFOG%20MID) "ColorLow" (Variables|FogBaseSetting|GetFogColorLow) :!class /Script/Engine.MaterialInstanceDynamic)
      (Rendering|Material|SetVectorParameterValue (Variables|Default|GetFOG%20MID) "ColorHigh" (Variables|FogBaseSetting|GetFogColorHigh) :!class /Script/Engine.MaterialInstanceDynamic)
      (Transformation|SetRelativeScale3D (Variables|Default|GetSM_Fog_Machine_FogPlane) (Variables|FogBaseSetting|GetFogPlaneSize)))
    (:then_1
      (Effects|Components|ParticleSystem|SetColorParameter (Variables|Default|GetNS_FogMachine_SmokeSpray) "SmokeColor" (Variables|FogBaseSetting|GetFogColorHigh))
      (Effects|Components|ParticleSystem|SetFloatParameter (Variables|Default|GetNS_FogMachine_SmokeSpray) "FogParticlesSpawnRate" (Math|Float|Clamp(Float) (Variables|FogBaseSetting|GetFogParticlesSpawnRate) 0.0 2.0)))))
