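; Gnome04 is a building prop actor Blueprint for a gnome that shows a particle effect when it holds destruction loot.
; On BeginPlay it waits two seconds, then checks HasDestructionLoot.
; If the check passes, it adds a particle system component raised above the actor's origin.
; The construction script is empty.
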(event EventBeginPlay
  (Utilities|FlowControl|Delay 2)
  (if (Curve|HasDestructionLoot)
    (AddComponent|NODE_AddParticleSystemComponent-0 :RelativeTransform (Math|Transform|MakeTransform "0, 0,72.000000"))))

(fn ConstructionScript)
