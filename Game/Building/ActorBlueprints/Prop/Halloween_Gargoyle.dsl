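; Halloween gargoyle prop that plays ambient audio on clients when the local player has this prop's quest and has not completed its objective.
; BeginPlay waits through a fixed delay, marked as a workaround, then continues only on the non-authority side.
; It checks the quest item and objective backend name against the first local Fort player controller only.
; Quest-level completion is ignored, and the construction script is empty.
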(event EventBeginPlay
  ;[1 0.3415 0] Work Around until better way found
  (Utilities|FlowControl|Delay 5.0)
  ;[1 1 1] Check Each Controller to see if they have the quest for this Prop AND haven't already completed the quest - if so, turn on music, otherwise, silent
  (Networking|SwitchHasAuthority
    (:Remote
      (bind (_questisvalid _questcompleted _objectivecompleted) (Default|AthenaCheckQuestandObjectiveCompletion (Utilities|Array|Get(acopy) (Fortnite|GetLocalFortPlayerControllers true)) (Variables|Default|GetQuestItem) (Variables|Default|GetObjectiveBackendName)))
      (if (and _questisvalid (not _objectivecompleted))
        (Audio|Components|Audio|Play (Variables|Default|GetAmbAudio) 0.0)))))

(fn ConstructionScript)
