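; Halloween haunted house building prop Blueprint that plays music when a player approaches and silences it when the prop is destroyed.
; An overlap on OverlapBox casts the other actor to FortPlayerPawn, then a DoOnce gate starts the Music audio component a single time.
; The death effects event quickly fades the music out if it is still playing.
; The construction script only calls its parent.
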(event OnComponentBeginOverlap(OverlapBox)
  (params (OverlappedComponent :type Primitive%20Component%20Object%20Reference) (OtherActor :type Actor%20Object%20Reference) (OtherComp :type Primitive%20Component%20Object%20Reference) (OtherBodyIndex :type Integer) (bFromSweep :type Boolean) (SweepResult :type Hit%20Result%20Structure%20%28by%20ref%29))
  (Utilities|IsValid (Variables|Default|GetMusic)
    (:Is%20Valid
      (Utilities|Casting|CastToFortPlayerPawn OtherActor)
      (Utilities|FlowControl|DoOnce false
        (:Completed
          (Audio|Components|Audio|Play (Variables|Default|GetMusic) 0.0))))))

(event EventOnDeathPlayEffects
  (params (Damage :type Float%20%28single-precision%29) (DamageTags :type Gameplay%20Tag%20Container%20Structure%20%28by%20ref%29) (Momentum :type Vector :default "0, 0, 0") (HitInfo :type Hit%20Result%20Structure%20%28by%20ref%29) (InstigatedBy :type Fort%20Pawn%20Object%20Reference) (DamageCauser :type Actor%20Object%20Reference) (EffectContext :type Gameplay%20Effect%20Context%20Handle%20Structure))
  (bind _music (Variables|Default|GetMusic))
  (Utilities|IsValid _music
    (:Is%20Valid
      (if (Audio|Components|Audio|IsPlaying _music)
        (Audio|Components|Audio|FadeOut _music 0.25 0.0 "Linear" :!class /Script/Engine.AudioComponent)))))

(fn ConstructionScript
  (|CallParentFunction))
