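; Halloween prop Blueprint (witch) that plays a random sound and a particle effect when a player pawn enters its overlap box.
; The server rolls a trigger chance, picks a sound index, then sets IsActive for a random retrigger delay.
; OnRep_IsActive plays effects on remote clients, and death stops the sound.
; The construction script sizes the box and configures particles, including an inflatable Battle Bus variant, or destroys the box if turned off.
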(event OnComponentBeginOverlap(OverlapBox)
  (params (OverlappedComponent :type Primitive%20Component%20Object%20Reference) (OtherActor :type Actor%20Object%20Reference) (OtherComp :type Primitive%20Component%20Object%20Reference) (OtherBodyIndex :type Integer) (bFromSweep :type Boolean) (SweepResult :type Hit%20Result%20Structure%20%28by%20ref%29))
  (if (not (Variables|HalloweenProp|GetPropIsTurnedOff))
    (Networking|SwitchHasAuthority
      (:Authority
        (Utilities|Casting|CastToFortPlayerPawnAthena OtherActor)
        (if (Variables|HalloweenProp|GetIsActive)
          _
          (elif (<= (Math|Random|RandomFloatinRange 0.0 1.0) (Variables|HalloweenProp|GetRandomTriggerPercentChance))
            (if (not (Variables|HalloweenProp|GetHasBeenTriggered))
              (Variables|HalloweenProp|SetHasBeenTriggered true)
              (if (> (Utilities|Array|Length (Variables|HalloweenProp|GetTriggeredSoundCues)) 1)
                (Variables|HalloweenProp|SetRandomAudioTriggerIndex (Math|Random|RandomIntegerinRange 0 (Utilities|Array|LastIndex (Variables|HalloweenProp|GetTriggeredSoundCues)))))
              (Variables|HalloweenProp|SetIsActive true)
              (Utilities|FlowControl|Delay (Math|Random|RandomFloatinRange (Variables|HalloweenProp|GetMinRetriggerDelay) (Variables|HalloweenProp|GetMaxRetriggerDelay)))
              (Variables|HalloweenProp|SetIsActive false)
              (Variables|HalloweenProp|SetHasBeenTriggered false))))))))

(event Custom|TestTriggeredEffects
  ;[1 1 1] editor debug
  (CallFunction|TriggerEffects))

(event EventOnDeathPlayEffects
  (params (Damage :type Float%20%28single-precision%29) (DamageTags :type Gameplay%20Tag%20Container%20Structure%20%28by%20ref%29) (Momentum :type Vector :default "0, 0, 0") (HitInfo :type Hit%20Result%20Structure%20%28by%20ref%29) (InstigatedBy :type Fort%20Pawn%20Object%20Reference) (DamageCauser :type Actor%20Object%20Reference) (EffectContext :type Gameplay%20Effect%20Context%20Handle%20Structure))
  (bind _currentsound (Variables|Default|GetCurrentSound)
    (:then
      (Audio|Components|Audio|Stop _currentsound))
    (:else)))

(fn OnRep_IsActive
  (if (Variables|HalloweenProp|GetIsActive)
    (Networking|SwitchHasAuthority
      (:Remote
        (CallFunction|TriggerEffects)))))

(fn TriggerEffects
  (bind _randomaudiotriggerindex (Variables|HalloweenProp|GetRandomAudioTriggerIndex))
  (bind _triggeredsoundcues (Variables|HalloweenProp|GetTriggeredSoundCues))
  (bind _output (Utilities|Array|Get(acopy) _triggeredsoundcues _randomaudiotriggerindex))
  (bind _particlesystem (Variables|Default|GetParticleSystem))
  (Utilities|IsValid _output
    (:Is%20Valid
      (bind _returnvalue (Audio|SpawnSoundatLocation _output (Transformation|GetActorLocation)))
      (Variables|Default|SetCurrentSound _returnvalue))
    (:Is%20Not%20Valid
      _))
  (Utilities|IsValid _particlesystem
    (:Is%20Valid
      (Components|Activation|Activate _particlesystem true))))

(fn ConstructionScript
  (bind _triggeredparticlesystem (Variables|HalloweenProp|GetTriggeredParticleSystem))
  (bind _particlesystem (Variables|Default|GetParticleSystem))
  (bind _particlesystem_2 (Variables|Default|GetParticleSystem))
  (if (Variables|HalloweenProp|GetPropIsTurnedOff)
    (Components|DestroyComponent (Variables|Default|GetOverlapBox))
    (else
      (Components|Box|SetBoxExtent (Variables|Default|GetOverlapBox) (Variables|HalloweenProp|GetOverlapBoxExtent) true)
      (Transformation|SetRelativeLocation (Variables|Default|GetOverlapBox) (+ (Variables|HalloweenProp|GetOverlapBoxOffset) "0.000000,0.000000,75.000000") false false)
      (Utilities|IsValid _triggeredparticlesystem
        (:Is%20Valid
          (Effects|Components|ParticleSystem|SetTemplate (Variables|Default|GetParticleSystem) _triggeredparticlesystem)
          (Effects|Components|ParticleSystem|SetVectorParameter _particlesystem_2 "LocationOffset" (+ (Variables|HalloweenProp|GetTriggeredFXLocationOffset) "0.000000,0.000000,75.000000"))
          (Effects|Components|ParticleSystem|SetVectorParameter _particlesystem_2 "LightRaySize" (Variables|HalloweenProp|GetTriggeredFXScale3D))
          (Transformation|SetRelativeRotation (Variables|Default|GetParticleSystem) (Variables|HalloweenProp|GetTriggeredFXRelativeRotation) false false)
          (if (Variables|HalloweenProp|GetIsInflatableBattleBus)
            (Effects|Components|ParticleSystem|SetFloatParameter _particlesystem "NormalLightsSpawnRate" 0.0)
            (Effects|Components|ParticleSystem|SetFloatParameter _particlesystem "BBLightsSpawnRate" 1.0)
            (else
              (Effects|Components|ParticleSystem|SetFloatParameter _particlesystem "NormalLightsSpawnRate" 1.0)
              (Effects|Components|ParticleSystem|SetFloatParameter _particlesystem "BBLightsSpawnRate" 0.0))))))))
