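; Blueprint for a flat haystack building prop that plays hay-specific foley sounds as it is searched and looted.
; Three custom events, OnBeginSearch, OnLootRepeat and OnLoot, each play their own cue (search, search-done and loot) at the actor's location.
; The construction script is empty, and the file contains no search or loot logic beyond these audio responses.
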(event Custom|OnBeginSearch
  ;[1 1 1] Audio
  (Audio|PlaySoundatLocation "/Game/Sounds/Foley_Loot/Containers/Hay/Prop_HayPile_Search_Cue.Prop_HayPile_Search_Cue" (Transformation|GetActorLocation)))

(event Custom|OnLootRepeat
  ;[1 1 1] Audio
  (Audio|PlaySoundatLocation "/Game/Sounds/Foley_Loot/Containers/Hay/Prop_HayPile_SearchDone_Cue.Prop_HayPile_SearchDone_Cue" (Transformation|GetActorLocation)))

(event Custom|OnLoot
  ;[1 1 1] Audio
  (Audio|PlaySoundatLocation "/Game/Sounds/Foley_Loot/Containers/Hay/Prop_HayPile_Loot_Cue.Prop_HayPile_Loot_Cue" (Transformation|GetActorLocation)))

(fn ConstructionScript)
