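; Searchable hall table prop Blueprint, apparently for a covered mansion living-area piece, that plays foley audio when looted.
; The OnBeginSearch custom event plays the nightstand search cue at the actor's location.
; The OnLootRepeat custom event plays the nightstand search-end cue at the same location.
; The construction script is empty, and no other logic is defined in this graph.
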(event Custom|OnBeginSearch
  (Audio|PlaySoundatLocation "/Game/Sounds/Foley_Loot/Containers/Nightstand/Nightstand_Search_Cue.Nightstand_Search_Cue" (Transformation|GetActorLocation)))

(event Custom|OnLootRepeat
  (Audio|PlaySoundatLocation "/Game/Sounds/Foley_Loot/Containers/Nightstand/Nightstand_SearchEnd_Cue.Nightstand_SearchEnd_Cue" (Transformation|GetActorLocation)))

(fn ConstructionScript)
