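; Actor Blueprint for a cheap lawn light prop with a spot light and an audio component.
; The construction script tints the spot light with a color variable and turns off its shadow casting on Nintendo Switch.
; On BeginPlay it destroys the lawn light audio component when the RemoveAudioComponent flag is set.
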(event EventBeginPlay
  (if (Variables|Default|GetRemoveAudioComponent)
    (Components|DestroyComponent (Variables|Default|GetLawnLightAudio))))

(fn ConstructionScript
  (Rendering|Components|Light|SetLightColor (Variables|Default|GetSpotLight) (Variables|Default|GetLightColorTint) true)
  (if (TouchPlatform|IsNintendoSwitch)
    (Rendering|Components|Light|SetCastShadows (Variables|Default|GetSpotLight) false)))
