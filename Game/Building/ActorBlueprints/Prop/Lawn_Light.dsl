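; Lawn_Light is a placeable lawn light prop Blueprint built around a SpotLight component.
; Its construction script applies the LightColorTint variable to the spotlight as an sRGB color, so the tint can vary through that variable.
; On Nintendo Switch, the same script turns off shadow casting on the spotlight.
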(fn ConstructionScript
  (Rendering|Components|Light|SetLightColor (Variables|Default|GetSpotLight) (Variables|Default|GetLightColorTint) true)
  (if (TouchPlatform|IsNintendoSwitch)
    (Rendering|Components|Light|SetCastShadows (Variables|Default|GetSpotLight) false)))
