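; Military_TireCourse is a prop Blueprint for a tire course, placed in the world as building dressing.
; On BeginPlay it assigns the TirePiles_03 jump-collision mesh to the Jump_Collision static mesh component of its Prop_TirePile_01 variable.
; The construction script is empty.
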(event EventBeginPlay
  (Components|StaticMesh|SetStaticMesh (Variables|Prop_TirePile_01|GetJump_Collision) "/Game/Environments/Props/Rubble/Meshes/TirePiles_03_JumpCollision.TirePiles_03_JumpCollision"))

(fn ConstructionScript)
