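; Parent Blueprint for building props that drives time-of-day lights and mesh emissives, plus a debug wind preview.
; Construction caches non-black emissive parameters from M_BaseSimple_4Layers materials, into native or Blueprint arrays, and computes animation curve lengths.
; Functions animate light intensity by curve or day phase, and fade MID emissives between phase values.
; With DebugWind on, construction swaps in a preview mesh whose materials interpolate toward intense-wind variants with curve-driven speeds.
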(fn CallAnimationCurveAgain
  (for _array_element (Variables|Building|TimeOfDayLights|GetTimeOfDayControlledLightDataArray)
    (bind _array_element_lightcomponent (Utilities|Struct|BreakTimeOfDayControlledLightData _array_element))
    (Utilities|IsValid _array_element_lightcomponent
      (:Is%20Valid
        (FortFX|AddLightAnimationIntensity _array_element_lightcomponent 0.0 0.0 (Variables|TimeOfDayControlledLights|GetLightAnimationCurve) false (Variables|TimeOfDayControlledLights|GetCodeControlled_Animation%20Curve%20Animation%20Length))))))

(fn GetTimeOfDayBlueprintDefaultVariables
  (return (Variables|TimeOfDayControlledLights|GetAlternateShadowStaticMesh) (Variables|TimeOfDayControlledLights|GetVolumetric%20Light%20Scattering%20Intensity) (Variables|TimeOfDayControlledLights|GetDisable%20TOD%20Lights%20and%20Material%20Emissive%20Values) (Variables|TimeOfDayControlledLights|GetDisable%20Static%20Mesh%20Shadow%20Casting%20When%20Lights%20Are%20Active) (Variables|TimeOfDayControlledLights|GetUse%20An%20Alternate%20Shadow%20Mesh%20When%20The%20Light%20is%20Off%20) (Variables|TimeOfDayControlledLights|GetCast%20Volumetric%20Shadows)))

(fn handleBaseSimple4LayerMaterials
  (params (MID :type Material%20Instance%20Dynamic%20Object%20Reference) (Duration :type Float%20%28double-precision%29) (start :type Float%20%28double-precision%29) (end :type Float%20%28double-precision%29))
  (for _array_element (Variables|TimeOfDayControlledLights|Get4LayerColorInfo)
    (bind (_array_element_index_11_82de82864bfee42b9f075a9f42967d6c _array_element_mid_12_a1a3af804e767f217ae1d1a1e6803a2f _array_element_emissivecolors_13_fb19d1914b992df77232c1ae3e5e4c83) (Utilities|Struct|BreakBuildingProp4LayerMaterialStruct _array_element))
    (bind _keys (Utilities|Map|Keys _array_element_emissivecolors_13_fb19d1914b992df77232c1ae3e5e4c83))
    (for (_array_element_2 _array_index) _keys
      (bind _values (Utilities|Map|Values _array_element_emissivecolors_13_fb19d1914b992df77232c1ae3e5e4c83))
      (FortFX|AddMIDAnimationColor _array_element_mid_12_a1a3af804e767f217ae1d1a1e6803a2f _array_element_2 (Math|Color|MakeColor (Math|Color|BreakColor (Utilities|Array|Get(aref) _values _array_index)) (.G (Math|Color|BreakColor (Utilities|Array|Get(aref) _values _array_index))) (.B (Math|Color|BreakColor (Utilities|Array|Get(aref) _values _array_index))) (Variables|Getstart)) (Math|Color|MakeColor (Math|Color|BreakColor (Utilities|Array|Get(aref) _values _array_index)) (.G (Math|Color|BreakColor (Utilities|Array|Get(aref) _values _array_index))) (.B (Math|Color|BreakColor (Utilities|Array|Get(aref) _values _array_index))) (Variables|Getend)) 0 false (Variables|GetDuration)))))

(fn init_4LayerMaterials
  (bind _array (Utilities|Array|MakeArray "Emissive" "Emissive2" "Emissive3" "Emissive4"))
  (bind _l_currentemissivemap (LocalVariables|Getl_currentEmissiveMap))
  (bind _l_currentmaterialindex (LocalVariables|Getl_currentMaterialIndex))
  (bind _l_currentmid (LocalVariables|Getl_currentMID))
  (if (not (Variables|TimeOfDayControlledLights|GetDisable%20TOD%20Lights%20and%20Material%20Emissive%20Values))
    (for _index (range (Rendering|Material|GetNumMaterials (Building|GetBuildingMeshComponent)))
      (LocalVariables|Setl_currentMaterialIndex _index)
      (bind _returnvalue (CallFunction|GetorCreateMIDOnDefaultMesh _index))
      (bind _output_get (LocalVariables|Setl_currentMID _returnvalue))
      (bind _returnvalue_2 (Rendering|Material|GetBaseMaterial _output_get))
      (if (Utilities|Equal(Object) _returnvalue_2 "/Game/Athena/WinterArt/Materials/M_BaseSimple_4Layers.M_BaseSimple_4Layers")
        (bind _asmaterial_instance_constant (Utilities|Casting|CastToMaterialInstanceConstant (Variables|/Script/Engine.MaterialInstance|GetParent _output_get)))
        (bind _array_element (Utilities|Array|ForEachLoopwithBreak (Variables|/Script/Engine.MaterialInstance|GetVectorParameterValues _asmaterial_instance_constant) :!label _label)
          (:LoopBody
            (bind (_parameterinfo _parametervalue) (Utilities|Struct|BreakVectorParameterValue _array_element))
            (bind _parameterinfo_name (Utilities|Struct|BreakMaterialParameterInfo _parameterinfo))
            (if (Utilities|Array|ContainsItem _array _parameterinfo_name)
              (Math|Integer|IncrementInt (LocalVariables|Getl_foundEmissives)
                (:%20%20
                  (if (Math|Color|NotEqual(LinearColor) _parametervalue "(R=0.000000,G=0.000000,B=0.000000,A=1.000000)")
                    (Utilities|Map|Add (LocalVariables|Getl_currentEmissiveMap) _parameterinfo_name _parametervalue) ; Param has info so add it to the list
                  )
                  (if (== (LocalVariables|Getl_foundEmissives) (Utilities|Array|Length _array))
                    (goto _label Break))))))
          (:Completed
            ;[1 1 1] If we found any emissives in use, add them to the list
            (if (> (Utilities|Map|Length (LocalVariables|Getl_currentEmissiveMap)) 0)
              (if (Building|TimeOfDayLights|UseNativeLightAnimation)
                (Utilities|Array|Add (Variables|Building|TimeOfDayLights|GetFourLayerColorInfo) (Utilities|Struct|MakeNativeBuildingPropFourLayerMaterial _l_currentmaterialindex _l_currentmid _l_currentemissivemap))
                (else
                  (Utilities|Array|Add (Variables|TimeOfDayControlledLights|Get4LayerColorInfo) (Utilities|Struct|MakeBuildingProp4LayerMaterialStruct _l_currentmaterialindex _l_currentmid _l_currentemissivemap))))))) ; Looks through each material for each emissive param
      ))))

(fn OLD_SetMIDValuesForLights
  (bind _returnvalue (Utilities|FindScalarValueAcrossTimeOfDayDayPhaseFloats (Variables|TimeOfDayControlledLights|GetEmissive%20Intensity%20Over%20Time%20of%20Day)))
  (LocalVariables|Setl_start _returnvalue)
  (bind _returnvalue_2 (Utilities|FindScalarValueAcrossTimeOfDayDayPhaseFloats (Variables|TimeOfDayControlledLights|GetEmissive%20Intensity%20Over%20Time%20of%20Day) 1))
  (LocalVariables|Setl_end _returnvalue_2)
  (for _index (range (Rendering|Material|GetNumMaterials (Building|GetBuildingMeshComponent)))
    (bind _output_get (LocalVariables|Setl_currentIndex _index))
    (bind _returnvalue_3 (CallFunction|GetorCreateMIDOnDefaultMesh _output_get))
    (bind _output_get_2 (LocalVariables|Setl_mid _returnvalue_3))
    (bind _returnvalue_4 (Rendering|Material|GetBaseMaterial _output_get_2))
    (Rendering|Material|SetMaterial (Building|GetBuildingMeshComponent) (LocalVariables|Getl_currentIndex) _output_get_2)
    (if (Utilities|Equal(Object) _returnvalue_4 "/Game/Athena/WinterArt/Materials/M_BaseSimple_4Layers.M_BaseSimple_4Layers")
      (CallFunction|HandleBaseSimple4LayerMaterials (LocalVariables|Getl_mid) (Variables|TimeOfDayControlledLights|GetDay%20Phase%20Transition%20Length) (LocalVariables|Getl_start) (LocalVariables|Getl_end)) ; If 4 Layer Material
      (else
        ;[0 0 0] Light Cone Opacity
        (FortFX|AddMIDAnimationFloat (LocalVariables|Getl_mid) "Light Cone Opacity" (* (LocalVariables|Getl_start) (Utilities|Array|Get(acopy) (Variables|TimeOfDayControlledLights|GetCodeControlled_LightConeOpacity) (LocalVariables|Getl_currentIndex))) (* (Utilities|Array|Get(acopy) (Variables|TimeOfDayControlledLights|GetCodeControlled_LightConeOpacity) (LocalVariables|Getl_currentIndex)) (LocalVariables|Getl_end)) 0 false (Variables|TimeOfDayControlledLights|GetDay%20Phase%20Transition%20Length))
        ;[0 0 0] Emissive color
        (FortFX|AddMIDAnimationColor (LocalVariables|Getl_mid) "Emissive Multiplier" (Math|Color|LinearColor*Float (Utilities|Array|Get(acopy) (Variables|TimeOfDayControlledLights|GetCodeControlled_EmissiveColor) (LocalVariables|Getl_currentIndex)) (LocalVariables|Getl_start)) (Math|Color|LinearColor*Float (Utilities|Array|Get(acopy) (Variables|TimeOfDayControlledLights|GetCodeControlled_EmissiveColor) (LocalVariables|Getl_currentIndex)) (LocalVariables|Getl_end)) 0 false (Variables|TimeOfDayControlledLights|GetDay%20Phase%20Transition%20Length))))))

(fn SetLightparameters
  (params (Previous%20to%20Current%20Day%20Phase%20Alpha :type Float%20%28double-precision%29))
  (for _array_element (Variables|Building|TimeOfDayLights|GetTimeOfDayControlledLightDataArray)
    (bind _array_element_lightcomponent (Utilities|Struct|BreakTimeOfDayControlledLightData _array_element))
    (Utilities|ManageLightOverTimebyDayPhase _array_element_lightcomponent (Variables|TimeOfDayControlledLights|GetLight%20Intensity%20Over%20Time%20of%20Day%20) Previous%20to%20Current%20Day%20Phase%20Alpha))
  (return))

(fn ConstructionScript
  (bind _returnvalue (Wind|GetWindSpeedCurveforPreview))
  (bind _returnvalue_2 (Wind|GetWindPannerSpeedCurveforPreview))
  (bind _debug_wind_intensity (Variables|Wind|GetDebug%20Wind%20Intensity))
  (bind _returnvalue_3 (Math|Curves|GetLinearColorValue _returnvalue _debug_wind_intensity))
  (bind _debug_wind_intensity_2 (Variables|Wind|GetDebug%20Wind%20Intensity))
  (bind _returnvalue_4 (Math|Curves|GetLinearColorValue _returnvalue_2 _debug_wind_intensity_2))
  (bind (_r _g _b _a) (Math|Color|BreakColor _returnvalue_3))
  (bind (_r_2 _g_2 _b_2 _a_2) (Math|Color|BreakColor _returnvalue_4))
  (bind _returnvalue_5 (Wind|GetWindPannerSpeedCurveforPreview))
  (Networking|CanExecuteCosmeticEvents
    (:True
      (Utilities|FlowControl|Sequence
        (:then_0
          (CallFunction|Init4LayerMaterials))
        (:then_1
          (Utilities|BuildingLightsPopulateLightAnimationCurveLength (Variables|TimeOfDayControlledLights|GetAnimate%20Lights%20With%20A%20Curve%20-%20Disables%20time%20of%20day%20light%20controls) (Variables|TimeOfDayControlledLights|GetAnimate%20Mesh%20MID%20Emissive%20Value%20with%20a%20Curve%20-%20Disables%20time%20of%20day%20light%20controls) (Variables|TimeOfDayControlledLights|GetLightAnimationCurve) (Variables|TimeOfDayControlledLights|GetCodeControlled_Animation%20Curve%20Animation%20Length))
          ; Wind debug functions
          (if (Variables|Wind|GetDebugWind)
            (Utilities|IsValid (Wind|GetWindSpeedCurveforPreview)
              (:Is%20Valid
                (Utilities|IsValid _returnvalue_5
                  (:Is%20Valid
                    (bind _returnvalue_6 (AddComponent|NODE_AddStaticMeshComponent-0))
                    (bind _output_get (Variables|Default|SetWind%20Intensity%20Debug%20Mesh _returnvalue_6))
                    (Components|StaticMesh|SetStaticMesh _output_get (Variables|Mesh|GetStaticMesh))
                    (Rendering|SetVisibility (Building|GetBuildingMeshComponent) false false)
                    (bind _output_get_2 (Variables|Wind|SetOriginalMaterials (Rendering|Material|GetMaterials _output_get)))
                    (bind _materials (Wind|GetIntenseWindMaterialsforPreview))
                    (Variables|Wind|SetIntenseWindMaterialsForPreview _materials)
                    (if (== (Utilities|Array|Length (Variables|Wind|GetOriginalMaterials)) (Utilities|Array|Length (Variables|Wind|GetIntenseWindMaterialsForPreview)))
                      (for (_array_element _array_index) _output_get_2
                        (bind _returnvalue_7 (Rendering|Material|CreateDynamicMaterialInstance (Variables|Default|GetWind%20Intensity%20Debug%20Mesh) _array_index _array_element "None" :!class /Script/Engine.PrimitiveComponent))
                        (Variables|Wind|SetDebug_TempMaterial _returnvalue_7)
                        (bind _asmaterial_instance (Utilities|Casting|CastToMaterialInstance (Utilities|Array|Get(acopy) (Variables|Wind|GetIntenseWindMaterialsForPreview) _array_index)))
                        (Rendering|Material|InterpolateMaterialInstanceParameters (Variables|Wind|GetDebug_TempMaterial) (Utilities|Casting|CastToMaterialInstance _array_element) _asmaterial_instance (Variables|Wind|GetDebug%20Wind%20Intensity))
                        (Rendering|Material|SetMaterial (Variables|Default|GetWind%20Intensity%20Debug%20Mesh) _array_index (Variables|Wind|GetDebug_TempMaterial))
                        (Utilities|IsValid _returnvalue
                          (:Is%20Valid
                            (Rendering|Material|SetScalarParameterValue (Variables|Wind|GetDebug_TempMaterial) "Wind_1_Speed_CodeReferenced" _r :!class /Script/Engine.MaterialInstanceDynamic)
                            (Rendering|Material|SetScalarParameterValue (Variables|Wind|GetDebug_TempMaterial) "Wind_2_Speed_CodeReferenced" _g :!class /Script/Engine.MaterialInstanceDynamic)
                            (Rendering|Material|SetScalarParameterValue (Variables|Wind|GetDebug_TempMaterial) "Wind_3_Speed_CodeReferenced" _b :!class /Script/Engine.MaterialInstanceDynamic)
                            (Rendering|Material|SetScalarParameterValue (Variables|Wind|GetDebug_TempMaterial) "Wind_4_Speed_CodeReferenced" _a :!class /Script/Engine.MaterialInstanceDynamic)
                            (Rendering|Material|SetScalarParameterValue (Variables|Wind|GetDebug_TempMaterial) "Wind_Panner_Speed_1x_CodeReferenced" _r_2 :!class /Script/Engine.MaterialInstanceDynamic)
                            (Rendering|Material|SetScalarParameterValue (Variables|Wind|GetDebug_TempMaterial) "Wind_Panner_Speed_2x_CodeReferenced" _g_2 :!class /Script/Engine.MaterialInstanceDynamic)
                            (Rendering|Material|SetScalarParameterValue (Variables|Wind|GetDebug_TempMaterial) "Wind_Panner_Speed_1y_CodeReferenced" _b_2 :!class /Script/Engine.MaterialInstanceDynamic)
                            (Rendering|Material|SetScalarParameterValue (Variables|Wind|GetDebug_TempMaterial) "Wind_Panner_Speed_2y_CodeReferenced" _a_2 :!class /Script/Engine.MaterialInstanceDynamic)
                            (Rendering|Material|SetVectorParameterValue (Variables|Wind|GetDebug_TempMaterial) "WindVector_CodeControlled" (Math|Color|MakeColor (.x (Math|Vector|RotateVectorAroundAxis (Math|Vector|MakeVector 1) (Variables|Wind|GetDebugWindYaw) "0, 0,1.000000")) (.y (Math|Vector|RotateVectorAroundAxis (Math|Vector|MakeVector 1) (Variables|Wind|GetDebugWindYaw) "0, 0,1.000000")) (.z (Math|Vector|RotateVectorAroundAxis (Math|Vector|MakeVector 1) (Variables|Wind|GetDebugWindYaw) "0, 0,1.000000"))) :!class /Script/Engine.MaterialInstanceDynamic))))
                      (else
                        (Development|PrintString "The intense wind material list needs to be populated in:")
                        (Development|PrintString (Utilities|GetDisplayName self)))))
                  (:Is%20Not%20Valid
                    (Development|PrintString "Please add a wind speed curve reference to:" :!label _label))))
              (:Is%20Not%20Valid
                (Utilities|FlowControl|Sequence
                  (:then_0
                    (Development|PrintString "Please add a wind speed curve reference to:"))
                  (:then_1
                    (Utilities|IsValid _returnvalue_5
                      (:Is%20Not%20Valid
                        (goto _label)))))))
            (Development|PrintString (Utilities|GetDisplayName self))
            (else
              (Rendering|SetVisibility (Building|GetBuildingMeshComponent) true false))))))))

(fn validateMIDValuesForLights4Layer
  (bind _returnvalue (Utilities|FindScalarValueAcrossTimeOfDayDayPhaseFloats (Variables|TimeOfDayControlledLights|GetEmissive%20Intensity%20Over%20Time%20of%20Day)))
  (LocalVariables|Setl_start _returnvalue)
  (bind _returnvalue_2 (Utilities|FindScalarValueAcrossTimeOfDayDayPhaseFloats (Variables|TimeOfDayControlledLights|GetEmissive%20Intensity%20Over%20Time%20of%20Day) 1))
  (LocalVariables|Setl_end _returnvalue_2)
  (for _index (range (Rendering|Material|GetNumMaterials (Building|GetBuildingMeshComponent)))
    (bind _output_get (LocalVariables|Setl_currentIndex _index))
    (bind _returnvalue_3 (CallFunction|GetorCreateMIDOnDefaultMesh _output_get))
    (bind _output_get_2 (LocalVariables|Setl_mid _returnvalue_3))
    (bind _returnvalue_4 (Rendering|Material|GetBaseMaterial _output_get_2))
    (Rendering|Material|SetMaterial (Building|GetBuildingMeshComponent) (LocalVariables|Getl_currentIndex) _output_get_2)
    (if (Utilities|Equal(Object) _returnvalue_4 "/Game/Athena/WinterArt/Materials/M_BaseSimple_4Layers.M_BaseSimple_4Layers")
      (CallFunction|HandleBaseSimple4LayerMaterials (LocalVariables|Getl_mid) (Variables|TimeOfDayControlledLights|GetDay%20Phase%20Transition%20Length) (LocalVariables|Getl_start) (LocalVariables|Getl_end)) ; If 4 Layer Material
    )))
