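; Parent building-prop actor that starts Curie fires around itself when destroyed.
; On server death, it skips silent, despawn and similar instant-death tags, then, if Curie is active, waits briefly before sphere-overlapping from a billboard marker.
; Landscape hits ignite at that point using configurable fuel and radius; other actors ignite at their own location with fixed values, crediting the killer's PlayerState.
; A custom event debug-draws the overlap radius.
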(event EventOnDeathServer
  (params (Damage :type Float%20%28single-precision%29) (DamageTags :type Gameplay%20Tag%20Container%20Structure%20%28by%20ref%29) (Momentum :type Vector :default "0, 0, 0") (HitInfo :type Hit%20Result%20Structure%20%28by%20ref%29) (InstigatedBy :type Controller%20Object%20Reference) (DamageCauser :type Actor%20Object%20Reference) (EffectContext :type Gameplay%20Effect%20Context%20Handle%20Structure))
  (bind _returnvalue (MissionHelpers|HasTags DamageTags (GameplayTags|MakeGameplayTagContainerfromArray (Utilities|Array|MakeArray "(TagName=\"Gameplay.Effect.InstantDeath.Silent\")" "(TagName=\"Gameplay.Effect.InstantDeath.PersistentUpdate\")" "(TagName=\"Gameplay.Effect.InstantDeath.Detachment\")" "(TagName=\"Gameplay.Effect.InstantDeath.Despawn\")" "(TagName=\"Gameplay.Effect.InstantDeath.Culled\")" "(TagName=\"Gameplay.Effect.InstantDeath.SilentDetachment\")")) false false))
  (if _returnvalue
    _
    (elif (Curie|IsCurieActive)
      (Utilities|Time|SetTimerbyEvent (delegate Custom|CustomEvent_0) 0.5))))

(event Custom|CustomEvent_0
  (CallFunction|ExecuteFireApplicationinRadius (stale EventOnDeathServer InstigatedBy)))

(event Custom|Draw%20Fire%20Overlap%20Radius
  (Rendering|Debug|DrawDebugSphere (Transformation|GetWorldLocation (Variables|Default|GetBillboard%20-%20Fire%20Start%20Location)) (Variables|FireParams|GetFireOverlapRadius) 12 "(R=1.000000,G=0.000000,B=0.051370,A=1.000000)" 3.0 1.0))

(fn ApplyFireToActor
  (params (TargetActor :type Actor%20Object%20Reference))
  (Curie|ExecuteFireApplication (Variables|FireParams|GetFireInstigator) self TargetActor 1e+02 (Collision|MakeHitResult false false 0.0 0.0 (Transformation|GetActorLocation TargetActor) (Transformation|GetActorLocation TargetActor) "0,0,1" "0,0,1" 0 TargetActor) 2 128.0))

(fn ApplyFireToLandscape
  (params (LandscapeActor :type Landscape%20Object%20Reference))
  (Curie|ExecuteFireApplication (Variables|FireParams|GetFireInstigator) self LandscapeActor 1e+02 (Collision|MakeHitResult false false 0.0 0.0 (Variables|FireParams|GetFireStartLocation) (Variables|FireParams|GetFireStartLocation) "0,0,1" "0,0,1" 0 LandscapeActor) (Variables|FireParams|GetFirePropogationFuel) (Variables|FireParams|GetLandscapeIgnitionRadius)))

(fn ExecuteFireApplicationInRadius
  (params (LastInstigator :type Controller%20Object%20Reference))
  (Variables|FireParams|SetFireInstigator (Variables|/Script/Engine.Controller|GetPlayerState LastInstigator))
  (Variables|FireParams|SetFireStartLocation (Transformation|GetWorldLocation (Variables|Default|GetBillboard%20-%20Fire%20Start%20Location)))
  (bind _outactors (Collision|SphereOverlapActors (Variables|FireParams|GetFireStartLocation) (Variables|FireParams|GetFireOverlapRadius) (Utilities|Array|MakeArray "ObjectTypeQuery3" "ObjectTypeQuery1" "ObjectTypeQuery2" "ObjectTypeQuery5")))
  (for _array_element _outactors
    (bind _aslandscape (Utilities|Casting|CastToLandscape _array_element)
      (:then
        (CallFunction|ApplyFiretoLandscape _aslandscape))
      (:CastFailed
        (CallFunction|ApplyFiretoActor _array_element)))))

(fn ConstructionScript
  (|CallParentFunction))
