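; Parent_StopLight is a parent Blueprint for stop light props, with its behavior left empty.
; BeginPlay and the construction script only call the parent's versions.
; The custom events Begin Working and Switchlights have no logic attached.
; A note in the construction script says its logic was disconnected as too expensive when many are placed near one another and needs a complete rework.
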(event EventBeginPlay
  (|CallParentFunction))

(event Custom|Begin%20Working)

(event Custom|Switchlights)

(fn ConstructionScript
  ;[1 0 0.06319] Disconnected this, needs complete rework to not ruin everything with how expensive this is when many are placed near one another.-
  (|CallParentFunction))
