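; Parent Blueprint for tree props, adding wind-blown leaf particles and an optional wind-intensity preview.
; Construction Script's DebugWind path hides the building mesh for a duplicate whose dynamic materials interpolate toward intense-wind materials, driven by wind curves and yaw.
; It also spawns leaf particles, except on mobile and Switch, configuring hue, direction and cylinder or sphere bounds.
; BeginPlay sets particle draw distance, destruction removes the test wind impulse, and ChangeWindIntensity is an empty event.
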(event EventDestroyed
  (CallFunction|RemoveTestWind 5))

(event Custom|ChangeWindIntensity
  (params (Intensity :type Float%20%28double-precision%29 :default 0.5) (Time%20It%20Takes%20To%20Reach%20New%20Intensity :type Float%20%28double-precision%29)))

(event EventBeginPlay
  ;[1 0.2314 0.6519] Set Max Draw Distance for Leaf/Wind Particles
  (bind _output_get (Variables|WindFX|SetWindFX_MaxDrawDistance 2.2e+03))
  (LOD|SetMaxDrawDistance (Variables|WindFX|GetWindFX) _output_get))

(fn BuildExposedAssetProviderVersion(Implementation)
  (params (Version :type Fort%20Exposed%20Asset%20Provider%20Version%20Structure%20%28by%20ref%29))
  (|CallParentFunction Version)
  (ExposedAssetProvider|AddVersion Version 1)
  (return))

(fn CollectAssetReferencestoAllow(Implementation)
  (params (Collector :type Fort%20Exposed%20Asset%20Reference%20Collector%20Structure%20%28by%20ref%29))
  (|CallParentFunction Collector)
  (ExposedAssetProvider|CollectIndirectAssetReference Collector (LocalVariables|GetWindFX_Mat))
  (return))

(fn RemoveTestWind
  (params (BlendTime :type Float%20%28double-precision%29))
  (bind _returnvalue (Fortnite|GetWindManager))
  (bind _windhandle (Variables|Default|GetWindHandle))
  (Networking|CanExecuteCosmeticEvents
    (:True
      (Utilities|IsValid _returnvalue
        (:Is%20Valid
          (if (Wind|HasWindImpulse _returnvalue _windhandle)
            (Wind|RemoveWindImpulse _returnvalue _windhandle BlendTime)))))))

(fn ConstructionScript
  (bind _returnvalue (Wind|GetWindPannerSpeedCurveforPreview))
  (bind _returnvalue_2 (Wind|GetWindSpeedCurveforPreview))
  (bind _returnvalue_3 (Wind|GetWindPannerSpeedCurveforPreview))
  (bind (_origin _boxextent) (Collision|GetComponentBounds (Variables|Debug|GetStaticMeshComponent)))
  (bind _debug_wind_intensity (Variables|Default|GetDebug%20Wind%20Intensity))
  (bind _returnvalue_4 (Math|Curves|GetLinearColorValue _returnvalue_2 _debug_wind_intensity))
  (bind _debug_wind_intensity_2 (Variables|Default|GetDebug%20Wind%20Intensity))
  (bind _returnvalue_5 (Math|Curves|GetLinearColorValue _returnvalue_3 _debug_wind_intensity_2))
  (bind (_r _g _b _a) (Math|Color|BreakColor _returnvalue_4))
  (bind (_r_2 _g_2 _b_2 _a_2) (Math|Color|BreakColor _returnvalue_5))
  (bind _windfx (Variables|WindFX|GetWindFX))
  (bind _boundradius (Variables|WindFX|GetBoundRadius))
  ;[1 1 1] Check to see that the assets are valid
  (Utilities|FlowControl|Sequence
    (:then_0
      (if (Variables|Default|GetDebugWind)
        (Utilities|IsValid (Wind|GetWindSpeedCurveforPreview)
          (:Is%20Valid
            (Utilities|IsValid _returnvalue
              (:Is%20Valid
                (bind _returnvalue_6 (AddComponent|NODE_AddStaticMeshComponent-0))
                (bind _output_get (Variables|Default|SetWind%20Intensity%20Debug%20Mesh _returnvalue_6))
                (Components|StaticMesh|SetStaticMesh _output_get (Variables|Mesh|GetStaticMesh))
                (Rendering|SetVisibility (Building|GetBuildingMeshComponent) false false)
                (bind _output_get_2 (LocalVariables|SetOriginalMaterials (Rendering|Material|GetMaterials _output_get)))
                (bind _materials (Wind|GetIntenseWindMaterialsforPreview))
                (LocalVariables|SetIntenseWindMaterialsForPreview _materials)
                ;[1 1 1] test materials
                (if (== (Utilities|Array|Length (LocalVariables|GetOriginalMaterials)) (Utilities|Array|Length (LocalVariables|GetIntenseWindMaterialsForPreview)))
                  (for (_array_element _array_index) _output_get_2
                    (bind _returnvalue_7 (Rendering|Material|CreateDynamicMaterialInstance (Variables|Default|GetWind%20Intensity%20Debug%20Mesh) _array_index _array_element "None" :!class /Script/Engine.PrimitiveComponent))
                    (Variables|Default|SetDebug_TempMaterial _returnvalue_7)
                    (bind _asmaterial_instance (Utilities|Casting|CastToMaterialInstance (Utilities|Array|Get(acopy) (LocalVariables|GetIntenseWindMaterialsForPreview) _array_index)))
                    (Rendering|Material|InterpolateMaterialInstanceParameters (Variables|Default|GetDebug_TempMaterial) (Utilities|Casting|CastToMaterialInstance _array_element) _asmaterial_instance (Variables|Default|GetDebug%20Wind%20Intensity))
                    (Rendering|Material|SetMaterial (Variables|Default|GetWind%20Intensity%20Debug%20Mesh) _array_index (Variables|Default|GetDebug_TempMaterial))
                    (Utilities|IsValid _returnvalue_2
                      (:Is%20Valid
                        ;[1 1 1 0.25] Set Wind Speed Values
                        (Rendering|Material|SetScalarParameterValue (Variables|Default|GetDebug_TempMaterial) "Wind_1_Speed_CodeReferenced" _r :!class /Script/Engine.MaterialInstanceDynamic)
                        (Rendering|Material|SetScalarParameterValue (Variables|Default|GetDebug_TempMaterial) "Wind_2_Speed_CodeReferenced" _g :!class /Script/Engine.MaterialInstanceDynamic)
                        (Rendering|Material|SetScalarParameterValue (Variables|Default|GetDebug_TempMaterial) "Wind_3_Speed_CodeReferenced" _b :!class /Script/Engine.MaterialInstanceDynamic)
                        (Rendering|Material|SetScalarParameterValue (Variables|Default|GetDebug_TempMaterial) "Wind_4_Speed_CodeReferenced" _a :!class /Script/Engine.MaterialInstanceDynamic)
                        ;[1 1 1 0.25] Set Wind Panner Speed Values
                        (Rendering|Material|SetScalarParameterValue (Variables|Default|GetDebug_TempMaterial) "Wind_Panner_Speed_1x_CodeReferenced" _r_2 :!class /Script/Engine.MaterialInstanceDynamic)
                        (Rendering|Material|SetScalarParameterValue (Variables|Default|GetDebug_TempMaterial) "Wind_Panner_Speed_2x_CodeReferenced" _g_2 :!class /Script/Engine.MaterialInstanceDynamic)
                        (Rendering|Material|SetScalarParameterValue (Variables|Default|GetDebug_TempMaterial) "Wind_Panner_Speed_1y_CodeReferenced" _b_2 :!class /Script/Engine.MaterialInstanceDynamic)
                        (Rendering|Material|SetScalarParameterValue (Variables|Default|GetDebug_TempMaterial) "Wind_Panner_Speed_2y_CodeReferenced" _a_2 :!class /Script/Engine.MaterialInstanceDynamic)
                        ;[1 1 1 0.25] Set Wind Direction
                        (Rendering|Material|SetVectorParameterValue (Variables|Default|GetDebug_TempMaterial) "WindVector_CodeControlled" (Math|Color|MakeColor (.x (Math|Vector|RotateVectorAroundAxis (Math|Vector|MakeVector 1) (Variables|Default|GetDebugWindYaw) "0, 0,1.000000")) (.y (Math|Vector|RotateVectorAroundAxis (Math|Vector|MakeVector 1) (Variables|Default|GetDebugWindYaw) "0, 0,1.000000")) (.z (Math|Vector|RotateVectorAroundAxis (Math|Vector|MakeVector 1) (Variables|Default|GetDebugWindYaw) "0, 0,1.000000"))) :!class /Script/Engine.MaterialInstanceDynamic))))
                  (else
                    (Development|PrintString "The intense wind material list needs to be populated in:")
                    (Development|PrintString (Utilities|GetDisplayName self)))))
              (:Is%20Not%20Valid
                (Development|PrintString "Please add a wind speed curve reference to:"))))
          (:Is%20Not%20Valid
            (Utilities|FlowControl|Sequence
              (:then_0
                (Development|PrintString "Please add a wind speed curve reference to:"))
              (:then_1
                (Utilities|IsValid _returnvalue
                  (:Is%20Not%20Valid
                    _))))))
        (Development|PrintString (Utilities|GetDisplayName self))
        (else
          (Rendering|SetVisibility (Building|GetBuildingMeshComponent) true false))))
    (:then_1
      ;[0 1 0.04499] Spawn falling leaves if true
      (if (Variables|WindFX|GetSpawnWindFX?)
        ;[0 1 0.04499] Check if mobile, if true don't spawn falling leaves
        (if (Fortnite|IsMobilePlatform)
          _
          (elif (TouchPlatform|IsNintendoSwitch)
            _
            (else
              ;[0 1 0.04499] Set windfx max draw distance variable to platform specific values
              (Variables|WindFX|SetWindFX_MaxDrawDistance 2.5e+03)
              ;[0 1 0.04499] Create wind fx
              (bind _output_get_3 (Variables|WindFX|SetWindFX_Transform (Math|Conversions|ToTransform(Vector) (+ (Math|Transform|InverseTransformLocation (Transformation|GetActorTransform) (Math|Vector|MakeVector (.x (Transformation|GetWorldLocation (Variables|Debug|GetStaticMeshComponent))) (.y (Transformation|GetWorldLocation (Variables|Debug|GetStaticMeshComponent))) (+ (.z (Transformation|GetWorldLocation (Variables|Debug|GetStaticMeshComponent))) (/ (.z _boxextent) 2.0)))) (Variables|WindFX|GetLeafCenter_Offset)))))
              (bind _returnvalue_8 (AddComponent|NODE_AddParticleSystemComponent-1 :RelativeTransform _output_get_3))
              (Variables|WindFX|SetWindFX _returnvalue_8)
              ;[1 0.2314 0.6519] Set Parameter: Particle Leaf Hue
              (Effects|Components|ParticleSystem|SetFloatParameter _windfx "DynamicHue" (Variables|WindFX|GetHue))
              ;[1 0.02604 0.5589] Set Parameter: Particle "Wind" Direction
              (Effects|Components|ParticleSystem|SetVectorParameter _windfx "WInd_Direction" (Variables|WindFX|GetWindDirection))
              ;[1 0.2314 0.6519] Set Parameter: Choose Emitter Shape (cylinder or sphere)
              (bind _output_get_4 (Variables|WindFX|SetUseCylinder (Variables|WindFX|GetUseCylinder)))
              (if _output_get_4
                ;[1 0.02604 0.5589] Set Parameter: Particle System Radius
                (Effects|Components|ParticleSystem|SetFloatParameter _windfx "Bounds_Cylinder_Radius" _boundradius)
                ;;[1 1 1] ZeroOut Unused Shape Radius
                (Effects|Components|ParticleSystem|SetFloatParameter _windfx "Bounds_Sphere_Radius" 0.0)
                ;[1 0.02604 0.5589] Set Parameter: Particle System Height
                (Effects|Components|ParticleSystem|SetFloatParameter _windfx "Bounds_Cylinder_Height" (Variables|WindFX|GetBoundsHeight(CylinderOnly)))
                (else
                  (Effects|Components|ParticleSystem|SetFloatParameter _windfx "Bounds_Sphere_Radius" _boundradius)
                  (Effects|Components|ParticleSystem|SetFloatParameter _windfx "Bounds_Cylinder_Radius" 0.0)
                  (Effects|Components|ParticleSystem|SetFloatParameter _windfx "Bounds_Cylinder_Height" 0.0))))))))))
