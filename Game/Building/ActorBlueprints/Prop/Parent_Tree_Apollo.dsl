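; Parent blueprint for Apollo tree props, handling seasonal foliage color and optional ambient effects.
; On construction and BeginPlay, where cosmetic events can run, it looks up the chosen season preset in shared tree data and writes color, variations and an option flag to custom primitive data.
; Construction also optionally adds a bug/firefly particle system and a falling-leaves static mesh.
; An ocean-submersion check exists, but the water-level and underwater-removal events are empty.
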(event Custom|OnWaterLevelSetOffline)

(event Custom|RemoveUnderwaterTrees)

(event EventBeginPlay
  (|CallParentFunction)
  (Networking|CanExecuteCosmeticEvents
    (:True
      (CallFunction|SetTreeSeasonalColor))))

(event EventEndPlay
  (params (EndPlayReason :type EEndPlayReason%20Enum :default "Destroyed"))
  (|CallParentFunction EndPlayReason)
  (Networking|CanExecuteCosmeticEvents))

(fn IsSubmergedByWater
  ;[1 1 1] Is flood height initialized?
  (if (> (Water|GetOceanFloodHeight (WorldSubsystems|GetWaterSubsystem)) 0.0)
    (return (> (- (Water|GetOceanTotalHeight (WorldSubsystems|GetWaterSubsystem)) 2e+01) (.z (Transformation|GetWorldLocation (Variables|Debug|GetStaticMeshComponent)))))))

(fn SetSeasonPresetColors
  (bind _array_element (Utilities|Array|ForEachLoopwithBreak (Variables|/Game/ArtTools/Procedural/Trees/TreeSeasonColor_Data.TreeSeasonColor_Data_C|GetTreeSeasonData (Variables|SeasonColor|GetTreeSeasonColor%20Data)))
    (:LoopBody
      (if (Utilities|Enum|Equal(Enum) (Utilities|Struct|BreakTreeSeasonColorStruct _array_element) (Variables|SeasonColor|GetTree%20Color%20Option))
        (Variables|SeasonColor|SetColor (.Color_5_E33006414CF6D999D4D482A4D2FF03F6 (Utilities|Struct|BreakTreeSeasonColorStruct _array_element)))
        (Variables|SeasonColor|SetValueVariation (.ValueVariation_9_4FAED9E04DA8955B0BC4EDB1E0A3C1F7 (Utilities|Struct|BreakTreeSeasonColorStruct _array_element)))
        (Variables|SeasonColor|SetHueVariation (.HueVariation_10_00BFFB9140C4AD7D16C9AA9AEB01A935 (Utilities|Struct|BreakTreeSeasonColorStruct _array_element)))))))

(fn SetTreeSeasonalColor
  (Utilities|FlowControl|Sequence
    (:then_0
      (if (Utilities|Enum|NotEqual(Enum) (Variables|SeasonColor|GetTree%20Color%20Option) "NewEnumerator0")
        (CallFunction|SetSeasonPresetColors)))
    (:then_1
      (CallFunction|UpdateColor))))

(fn UpdateColor
  (bind (_r _g _b _a) (Utilities|Struct|BreakLinearColor (Variables|SeasonColor|GetColor)))
  (Rendering|Material|SetDefaultCustomPrimitiveDataVector4 (Variables|Debug|GetStaticMeshComponent) 0 (Math|Vector4|MakeVector4 _r _g _b _a))
  (Rendering|Material|SetDefaultCustomPrimitiveDataFloat (Variables|Debug|GetStaticMeshComponent) 4 (Variables|SeasonColor|GetHueVariation))
  (Rendering|Material|SetDefaultCustomPrimitiveDataFloat (Variables|Debug|GetStaticMeshComponent) 5 (Variables|SeasonColor|GetValueVariation))
  (Rendering|Material|SetDefaultCustomPrimitiveDataFloat (Variables|Debug|GetStaticMeshComponent) 6 (Utilities|Select :Index (Variables|SeasonColor|GetTree%20Color%20Option) :NewEnumerator0 0.0 :NewEnumerator1 0.0 :NewEnumerator2 0.0 :NewEnumerator3 0.0 :NewEnumerator4 1.0)))

(fn ConstructionScript
  (Networking|CanExecuteCosmeticEvents
    (:True
      (Utilities|FlowControl|Sequence
        (:then_0
          (CallFunction|SetTreeSeasonalColor))
        (:then_1
          (Utilities|FlowControl|Sequence
            (:then_0
              ;[1 1 1] Fireflies
              (if (Variables|FX|GetHas%20Bugs)
                (bind _returnvalue (AddComponent|NODE_AddParticleSystemComponent-2 :RelativeTransform (Math|Conversions|ToTransform(Vector) (Variables|FX|GetBugOffset))))
                (Components|Activation|SetAutoActivate _returnvalue true)
                (Effects|Components|ParticleSystem|SetFloatParameter _returnvalue "IsLightningBug" (Variables|FX|GetUse%20Lightning%20Bug))))
            (:then_1
              ;[1 1 1] Falling Leaves
              (if (Variables|Default|GetHas%20Falling%20Leaves)
                (AddComponent|NODE_AddStaticMeshComponent-2 :RelativeTransform (Math|Conversions|ToTransform(Vector) (Variables|FX|GetLeavesOffset)))))))))))
