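; Pirate ship plank prop that works as a diving board, bouncing whatever overlaps its Touchbox upward.
; Fort pawns get an additive launch that cancels current vertical velocity for a consistent upward speed, then play the bounce sound and ignore their next falling damage.
; When the pawn cast fails, it attempts a hoverboard launch tagged DivingBoard and plays the sound only if that succeeds.
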(event OnComponentBeginOverlap(Touchbox)
  (params (OverlappedComponent :type Primitive%20Component%20Object%20Reference) (OtherActor :type Actor%20Object%20Reference) (OtherComp :type Primitive%20Component%20Object%20Reference) (OtherBodyIndex :type Integer) (bFromSweep :type Boolean) (SweepResult :type Hit%20Result%20Structure%20%28by%20ref%29))
  (bind _asfort_pawn (Utilities|Casting|CastToFortPawn OtherActor)
    (:then
      (Character|LaunchCharacter _asfort_pawn (Math|Vector|MakeVector 0.0 0.0 (+ (* (.z (Transformation|GetVelocity _asfort_pawn)) -1.0) 2e+03)) false false)
      (CallFunction|PlayBounceSound)
      (Variables|/Script/FortniteGame.FortPawn|SetbIgnoreNextFallingDamage true _asfort_pawn))
    (:CastFailed
      (bind _success (Vehicles|LaunchHoverboard 0 "0, 0,2000.000000" "DivingBoard"))
      (if _success
        (CallFunction|PlayBounceSound)))))

(fn PlayBounceSound
  (Audio|SpawnSoundatLocation (Variables|Components|GetSound) (Transformation|GetActorLocation self)))

(fn ConstructionScript)
