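; Destructible power transformer prop Blueprint that plays electrical effects and sound when damaged.
; Each damage event activates the on-hit electric effect and spawns an attached zap sound at that component.
; A DoOnce also activates the damage effect with the same zap cue on the first hit only.
; On death the damage effect is deactivated; the construction script is empty.
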(event EventOnDamagePlayEffects
  (params (Damage :type Float%20%28single-precision%29) (DamageTags :type Gameplay%20Tag%20Container%20Structure%20%28by%20ref%29) (Momentum :type Vector :default "0, 0, 0") (HitInfo :type Hit%20Result%20Structure%20%28by%20ref%29) (InstigatedBy :type Fort%20Pawn%20Object%20Reference) (DamageCauser :type Actor%20Object%20Reference) (EffectContext :type Gameplay%20Effect%20Context%20Handle%20Structure))
  (bind _effect_damage (Variables|PowerTransformer01|GetEffect_Damage))
  (bind _effect_electric_on_hit (Variables|PowerTransformer01|GetEffect_Electric_ON_Hit))
  (Utilities|FlowControl|Sequence
    (:then_0
      (Components|Activation|Activate _effect_electric_on_hit false)
      (Audio|SpawnSoundAttached "/Game/Sounds/Fort_Traps/Wall_Electric/Transformer_Electric_Zap_Cue.Transformer_Electric_Zap_Cue" _effect_electric_on_hit))
    (:then_1
      (Utilities|FlowControl|DoOnce false
        (:Completed
          (Components|Activation|Activate _effect_damage false)
          (Audio|SpawnSoundAttached "/Game/Sounds/Fort_Traps/Wall_Electric/Transformer_Electric_Zap_Cue.Transformer_Electric_Zap_Cue" _effect_damage))))))

(event EventOnDeathPlayEffects
  (params (Damage :type Float%20%28single-precision%29) (DamageTags :type Gameplay%20Tag%20Container%20Structure%20%28by%20ref%29) (Momentum :type Vector :default "0, 0, 0") (HitInfo :type Hit%20Result%20Structure%20%28by%20ref%29) (InstigatedBy :type Fort%20Pawn%20Object%20Reference) (DamageCauser :type Actor%20Object%20Reference) (EffectContext :type Gameplay%20Effect%20Context%20Handle%20Structure))
  (Components|Activation|Deactivate (Variables|PowerTransformer01|GetEffect_Damage)))

(fn ConstructionScript)
