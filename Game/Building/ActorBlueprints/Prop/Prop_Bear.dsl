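; Bear building prop that has a rare easter-egg bee hive effect with sound and particles.
; On BeginPlay the server rolls a 1-in-501 chance to mark the prop as already searched and turn on the replicated Effects flag.
; OnRep_Effects shows the hive and starts audio and particles, or fades the audio, deactivates particles and hides the hive.
; Server-side death clears the flag.
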(event EventBeginPlay
  (Networking|SwitchHasAuthority
    (:Authority
      (if (== (Math|Random|RandomIntegerinRange 0 500) 1)
        (Variables|Container|SetbAlreadySearched true)
        (Variables|Default|SetEffects true)))))

(event EventOnDeathServer
  (params (Damage :type Float%20%28single-precision%29) (DamageTags :type Gameplay%20Tag%20Container%20Structure%20%28by%20ref%29) (Momentum :type Vector :default "0, 0, 0") (HitInfo :type Hit%20Result%20Structure%20%28by%20ref%29) (InstigatedBy :type Controller%20Object%20Reference) (DamageCauser :type Actor%20Object%20Reference) (EffectContext :type Gameplay%20Effect%20Context%20Handle%20Structure))
  (Variables|Default|SetEffects false))

(fn OnRep_Effects
  (bind _audio (Variables|Default|GetAudio))
  (bind _particlesystem (Variables|Default|GetParticleSystem))
  (if (Variables|Default|GetEffects)
    (Audio|Components|Audio|Play (Variables|Default|GetAudio) 0.0)
    (Development|SetHiddeninGame (Variables|Default|GetS_Bee_Hive) false false)
    (Components|Activation|Activate (Variables|Default|GetParticleSystem) false)
    (Rendering|SetVisibility (Variables|Default|GetS_Bee_Hive) true false)
    (else
      (Utilities|IsValid _audio
        (:Is%20Valid
          (Audio|Components|Audio|FadeOut _audio 0.75 0.0 "Sin" :!class /Script/Engine.AudioComponent)
          (Utilities|IsValid _particlesystem
            (:Is%20Valid
              (Components|Activation|Deactivate _particlesystem))
            (:Is%20Not%20Valid
              _))
          (Rendering|SetVisibility (Variables|Default|GetS_Bee_Hive) false false))))))

(fn ConstructionScript)
