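; Brick door building prop Blueprint that zeroes a secondary-motion material setting on its static mesh when play begins.
; BeginPlay runs this only where cosmetic events can execute, casting material slot 0 to a dynamic instance and creating one if the cast fails.
; It sets SecondaryMotionFalloff to zero on the cast result and also zeroes custom primitive data float 2 on the mesh.
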(event EventBeginPlay
  (Networking|CanExecuteCosmeticEvents
    (:True
      (bind _asmaterial_instance_dynamic (Utilities|Casting|CastToMaterialInstanceDynamic (Rendering|Material|GetMaterial (Variables|Debug|GetStaticMeshComponent) 0))
        (:then
          _)
        (:CastFailed
          (bind _returnvalue (Rendering|Material|CreateDynamicMaterialInstance (Variables|Debug|GetStaticMeshComponent) 0 0 "None" :!class /Script/Engine.PrimitiveComponent))))
      (Rendering|Material|SetScalarParameterValue _asmaterial_instance_dynamic "SecondaryMotionFalloff" 0.0 :!class /Script/Engine.MaterialInstanceDynamic)
      (Rendering|Material|SetCustomPrimitiveDataFloat (Variables|Debug|GetStaticMeshComponent) 2 0.0))))

(fn ConstructionScript)
