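; Desert cactus building prop actor blueprint that removes itself when its AutoDestroy flag is set.
; The AutoDestroy rep-notify checks the flag, then switches on authority.
; The server calls SilentDie with a false argument, while remote clients destroy the actor directly.
; The construction script is empty.
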(fn OnRep_AutoDestroy
  (if (Variables|Default|GetAutoDestroy)
    (Networking|SwitchHasAuthority
      (:Authority
        (Gameplay|SilentDie false))
      (:Remote
        (Actor|DestroyActor)))))

(fn ConstructionScript)
