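; Blueprint actor for a helicopter movie set prop, placed among the building prop actors.
; It adds no logic of its own: BeginPlay is empty and the construction script only calls the parent class version.
; Any behaviour comes from the parent class.
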(event EventBeginPlay)

(fn ConstructionScript
  (|CallParentFunction))
