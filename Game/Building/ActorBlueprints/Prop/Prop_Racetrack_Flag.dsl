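; Racetrack flag prop Blueprint for a building actor that handles stopping its looping sound when the prop is destroyed.
; On the death-effects event it fetches the LoopingAudio component and, if valid, applies a short linear fade-out to silence.
; The Hit event and construction script are empty, so the death fade is the only logic in the graph.
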(event EventOnDeathPlayEffects
  (params (Damage :type Float%20%28single-precision%29) (DamageTags :type Gameplay%20Tag%20Container%20Structure%20%28by%20ref%29) (Momentum :type Vector :default "0, 0, 0") (HitInfo :type Hit%20Result%20Structure%20%28by%20ref%29) (InstigatedBy :type Fort%20Pawn%20Object%20Reference) (DamageCauser :type Actor%20Object%20Reference) (EffectContext :type Gameplay%20Effect%20Context%20Handle%20Structure))
  (bind _loopingaudio (Variables|Default|GetLoopingAudio))
  (Utilities|IsValid _loopingaudio
    (:Is%20Valid
      (if (Utilities|IsValid _loopingaudio :!class /Script/Engine.KismetSystemLibrary)
        (Audio|Components|Audio|FadeOut _loopingaudio 0.2 0.0 "Linear" :!class /Script/Engine.AudioComponent)))))

(event Collision|EventHit
  (params (MyComp :type Primitive%20Component%20Object%20Reference) (Other :type Actor%20Object%20Reference) (OtherComp :type Primitive%20Component%20Object%20Reference) (bSelfMoved :type Boolean) (HitLocation :type Vector :default "0, 0, 0") (HitNormal :type Vector :default "0, 0, 0") (NormalImpulse :type Vector :default "0, 0, 0") (Hit :type Hit%20Result%20Structure%20%28by%20ref%29)))

(fn ConstructionScript)
