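; Security camera prop that sweeps its camera component back and forth as a cosmetic effect, skipped on mobile.
; BeginPlay randomizes cone angle, pitch and sweep duration, then StartRotating eases the component to that rotation with MoveComponentTo and flips DirectionBias to swing back.
; When the actor was not recently rendered, sweeping pauses and a looping timer with a random interval polls visibility, clearing itself and resuming rotation once seen.
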(event EventBeginPlay
  (Networking|CanExecuteCosmeticEvents
    (:True
      (if (not (Fortnite|IsMobilePlatform))
        (Variables|Prop_SecurityCamera_01|SetRotConeInDegrees (Math|Random|RandomFloatinRange 35 80))
        (Variables|Prop_SecurityCamera_01|SetRotDuration (Math|Random|RandomFloatinRange 5 12))
        (Variables|Prop_SecurityCamera_01|SetRandomPitch (Math|Random|RandomFloatinRange 1 15))
        (CallFunction|StartRotating)))))

(event Custom|StartRotating
  (if (Rendering|WasActorRecentlyRendered 0.2)
    (Components|MoveComponentTo (Variables|Prop_SecurityCamera_01|GetCamera) (Variables|/Script/Engine.SceneComponent|GetRelativeLocation (Variables|Prop_SecurityCamera_01|GetCamera)) (Math|Rotator|MakeRotator 0.0 (* (Variables|Prop_SecurityCamera_01|GetRandomPitch) (Variables|Default|GetDirectionBias)) (* (Variables|Default|GetDirectionBias) (Variables|Prop_SecurityCamera_01|GetRotConeInDegrees))) true true (Variables|Prop_SecurityCamera_01|GetRotDuration) true)
    (Math|Float|NegateFloat (Variables|Default|GetDirectionBias)
      (:%20%20
        (CallFunction|StartRotating)))
    (else
      (bind _returnvalue (Utilities|Time|SetTimerbyEvent (delegate Custom|StartTickingIfRendered) (Math|Random|RandomFloatinRange 3 5) true))
      (Variables|Default|SetTimerHandle _returnvalue))))

(event Custom|StartTickingIfRendered
  (if (Rendering|WasActorRecentlyRendered 0.2)
    (Utilities|Time|ClearandInvalidateTimerbyHandle (Variables|Default|GetTimerHandle))
    (CallFunction|StartRotating)))

(fn ConstructionScript)
