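; Tire pile prop that acts as a bounce pad, launching whatever overlaps its jump collision and playing a wobble animation.
; Player and AI pawns get LaunchCharacter, a launched effect and fall-damage immunity; other actors take projectile bounce velocity or a hoverboard launch.
; A capsule trace (CeilingCheck) lowers launch height under obstructions.
; Cooldown-gated bounce sound, player force feedback and forced net updates accompany launches.
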(event Custom|MakeWobbleEffect
  (params (Center :type Vector))
  (CallFunction|StartBounceAnimation (Utilities|Time|GetTimeSeconds) 2e+02 (Math|Vector|MakeVector 0.0 0.0 1.0) Center))

(event OnComponentBeginOverlap(Jump_Collision)
  (params (OverlappedComponent :type Primitive%20Component%20Object%20Reference) (OtherActor :type Actor%20Object%20Reference) (OtherComp :type Primitive%20Component%20Object%20Reference) (OtherBodyIndex :type Integer) (bFromSweep :type Boolean) (SweepResult :type Hit%20Result%20Structure%20%28by%20ref%29))
  (bind _returnvalue (Actor|GetComponentbyClass OtherActor "/Script/Engine.ProjectileMovementComponent"))
  (bind (_bblockinghit _binitialoverlap _time _distance _location _impactpoint) (Collision|BreakHitResult SweepResult))
  (bind (_asfort_projectile_petrol _bsuccess) (Utilities|Casting|CastToFortProjectilePetrol OtherActor))
  (bind (_asfort_pickup _bsuccess_2) (Utilities|Casting|CastToFortPickup OtherActor))
  (bind _overlappawn (Variables|Default|GetOverlapPawn))
  (bind _overlappawn_2 (Variables|Default|GetOverlapPawn))
  (bind _overlappawn_3 (Variables|Default|GetOverlapPawn))
  (Variables|Default|SetBounceLocation (Transformation|GetActorLocation))
  (CallFunction|MakeWobbleEffect _impactpoint)
  (bind _asfort_player_pawn (Utilities|Casting|CastToFortPlayerPawn OtherActor)
    (:then
      (Variables|Default|SetOverlapPawn _asfort_player_pawn))
    (:CastFailed
      (bind _asfort_aipawn (Utilities|Casting|CastToFortAIPawn OtherActor)
        (:then
          (Variables|Default|SetOverlapPawn _asfort_aipawn))
        (:CastFailed
          (Utilities|FlowControl|Sequence
            (:then_0
              ;[1 0.03743 0] Launching Projectile
              (if (Math|Boolean|ANDBoolean (or (Utilities|IsValid _returnvalue :!class /Script/Engine.KismetSystemLibrary) (Utilities|IsValid (Actor|GetComponentbyClass OtherActor "/Script/FortniteGame.FortProjectileMovementComponent") :!class /Script/Engine.KismetSystemLibrary)) (not _bsuccess) (not _bsuccess_2) (Variables|Default|GetBBouncingProjectile))
                (bind _returnvalue_2 (Utilities|GetNonPawnBounceVel OtherActor)
                  (:Output
                    (bind (_returnvalue_x _returnvalue_y _returnvalue_z) (Math|Vector|BreakVector _returnvalue_2))
                    (Variables|/Script/Engine.MovementComponent|SetVelocity (Math|Vector|MakeVector _returnvalue_x _returnvalue_y _returnvalue_z) _returnvalue)))))
            (:then_1
              ;[1 1 1] Ceiling check
              (bind _success (CallFunction|CeilingCheck))
              (if _success
                (Variables|Default|SetBounceHeight 600)
                (Variables|Default|SetJumpHeight 100)
                (else
                  (Variables|Default|SetBounceHeight 2e+03)
                  (Variables|Default|SetJumpHeight 600)))
              (Variables|Default|SetFinalBounceJumpHeight (Variables|Default|GetBounceHeight))
              (bind _success_2 (Vehicles|LaunchHoverboard OtherActor (Math|Vector|MakeVector 0.0 0.0 (Variables|Default|GetFinalBounceJumpHeight)) "TireLaunch"))
              (if _success_2
                (Networking|ForceNetUpdate)
                (CallFunction|PlayBounceSound))))))))
  (Utilities|IsValid _overlappawn_3
    (:Is%20Valid
      (if (or (GameplayTags|HasMatchingGameplayTag (Ability|GetAbilitySystemComponent _overlappawn_3) "(TagName=\"Gameplay.Action.Player.Launched\")") (Variables|/Script/FortniteGame.FortPawn|GetbIsDBNO _overlappawn_3))
        _
        (else
          _)))
    (:Is%20Not%20Valid
      _))
  (bind _asfort_player_controller (Utilities|Casting|CastToFortPlayerController (Pawn|GetController _overlappawn_3))
    (:then
      (Variables|Default|SetOverlapController _asfort_player_controller))
    (:CastFailed
      _))
  ;[1 1 1] Ceiling check
  (bind _success_3 (CallFunction|CeilingCheck))
  (if _success_3
    (Variables|Default|SetBounceHeight 600)
    (Variables|Default|SetJumpHeight 100)
    (else
      (Variables|Default|SetBounceHeight (Variables|Default|GetNoCeilingBounceHeight))
      (Variables|Default|SetJumpHeight 600)))
  (Variables|Default|SetFinalBounceJumpHeight (Variables|Default|GetBounceHeight))
  (bind _asfort_player_pawn_2 (Utilities|Casting|CastToFortPlayerPawn (Variables|Default|GetOverlapPawn))
    (:then
      (RemoteControlPawn|ForceDetachfromRemoteControlledPawn _asfort_player_pawn_2))
    (:CastFailed
      _))
  (bind (_returnvalue_x_2 _returnvalue_y_2) (Math|Vector|BreakVector (Transformation|GetActorForwardVector _asfort_player_pawn_2)))
  (Character|LaunchCharacter _overlappawn_2 (Math|Vector|MakeVector (* _returnvalue_x_2 (Variables|Default|GetBounceXYVelocityMofifier)) (* _returnvalue_y_2 (Variables|Default|GetBounceXYVelocityMofifier)) (Variables|Default|GetFinalBounceJumpHeight)) false true)
  (Ability|SendGameplayEventtoActor _overlappawn_2 "(TagName=\"Event.Tire.Bounce\")")
  (GameplayEffects|ApplyGameplayEffectToSelf :self (Ability|GetAbilitySystemComponent _overlappawn) :GameplayEffectClass (Variables|Default|GetGE_PlayerLaunched) :Level 0.0)
  (Variables|/Script/FortniteGame.FortPawn|SetbIgnoreNextFallingDamage true _overlappawn)
  (Utilities|FlowControl|Sequence
    (:then_0
      (bind _overlapcontroller (Variables|Default|GetOverlapController)
        (:then
          (Game|Feedback|LocalPlayForceFeedback _overlapcontroller (Variables|Default|GetLaunchForceFeedback) false "TireLaunch"))
        (:else
          _))
      (Variables|Default|SetOverlapPawn 0)
      (Variables|Default|SetOverlapController 0))
    (:then_1
      (Networking|ForceNetUpdate)
      (CallFunction|PlayBounceSound))))

(event Custom|Play%20Bounce%20Sound
  ; Audio - Play bounce sound
  (Networking|SwitchHasAuthority
    (:Authority
      (CallFunction|PlaySound))
    (:Remote
      (if (Variables|Default|GetCan%20Play%20Bounce%20Sound)
        (Audio|PlaySoundatLocation (Variables|Default|GetBounceSound) (Transformation|GetActorLocation))
        (Utilities|Time|SetTimerbyEvent (delegate Custom|Reset%20Bounce%20Sound%20Cooldown) 0.2)
        (Variables|Default|SetCan%20Play%20Bounce%20Sound false)))))

(event Custom|Play%20Sound
  ; Audio - Play bounce sound
  (Networking|CanExecuteCosmeticEvents
    (:True
      (if (Variables|Default|GetCan%20Play%20Bounce%20Sound)
        (Audio|PlaySoundatLocation (Variables|Default|GetBounceSound) (Transformation|GetActorLocation))
        (Utilities|Time|SetTimerbyEvent (delegate Custom|Reset%20Bounce%20Sound%20Cooldown) 0.2)
        (Variables|Default|SetCan%20Play%20Bounce%20Sound false)))))

(event Custom|Reset%20Bounce%20Sound%20Cooldown
  ; Audio - Play bounce sound
  (Variables|Default|SetCan%20Play%20Bounce%20Sound true))

(fn CeilingCheck
  (bind _overlappawn (Variables|Default|GetOverlapPawn)
    (:then
      (Utilities|Array|Add (LocalVariables|GetActorsToIgnore) _overlappawn)
      (Utilities|Array|AppendArray (LocalVariables|GetActorsToIgnore) (Actor|GetAttachedActors _overlappawn true false))
      (for _array_element (Actor|GetAttachedActors _overlappawn true false)
        (Utilities|Array|AppendArray (LocalVariables|GetActorsToIgnore) (Actor|GetAttachedActors _array_element true false))))
    (:else
      _))
  (bind (_outhit _returnvalue) (Collision|CapsuleTraceByChannel (Math|Vector|MakeVector (.x (Transformation|GetWorldLocation (Variables|Transformation|GetRootComponent))) (.y (Transformation|GetWorldLocation (Variables|Transformation|GetRootComponent))) (+ (.z (Transformation|GetWorldLocation (Variables|Transformation|GetRootComponent))) 200)) (Math|Vector|MakeVector (.x (Transformation|GetWorldLocation (Variables|Transformation|GetRootComponent))) (.y (Transformation|GetWorldLocation (Variables|Transformation|GetRootComponent))) (+ (.z (Transformation|GetWorldLocation (Variables|Transformation|GetRootComponent))) 340)) 50 100 "TraceTypeQuery9" false (LocalVariables|GetActorsToIgnore)))
  (return _returnvalue))

(fn CheckJumpingOrientation
  (bind (_returnvalue_x _returnvalue_y) (Math|Vector|BreakVector (Transformation|GetActorForwardVector)))
  (bind _orientationmultiplier (Variables|Default|GetOrientationMultiplier))
  (if (== (Math|Float|Round (Math|Float|Absolute(Float) _returnvalue_x)) 1)
    (return (Math|Vector|vector*float (Math|Vector|MakeVector _returnvalue_x _returnvalue_y) _orientationmultiplier))
    (else
      (return (Math|Vector|vector*float (Math|Vector|MakeVector _returnvalue_x _returnvalue_y) (* _orientationmultiplier -1.0))))))

(fn ConstructionScript)
