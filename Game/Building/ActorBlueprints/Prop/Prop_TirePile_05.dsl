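; Prop_TirePile_05 is a tire pile prop Blueprint that prepares a sideways bounce vector.
; On BeginPlay it scales the actor's right vector by JumpModifier and stores the X and Y components as Final Bounce X Vector and Final Bounce Y Vector, leaving Z unused.
; Its overlap and tick events are empty, and the construction script only calls the parent.
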(event EventBeginPlay
  (bind _returnvalue (Transformation|GetActorRightVector))
  (bind _jumpmodifier (Variables|Default|GetJumpModifier))
  (bind (_returnvalue_x _returnvalue_y) (Math|Vector|BreakVector (* _returnvalue _jumpmodifier)))
  (Variables|Default|SetFinal%20Bounce%20X%20Vector _returnvalue_x)
  (Variables|Default|SetFinal%20Bounce%20Y%20Vector _returnvalue_y))

(event Collision|EventActorBeginOverlap
  (params (OtherActor :type Actor%20Object%20Reference)))

(event EventTick
  (params (DeltaSeconds :type Float%20%28single-precision%29)))

(fn ConstructionScript
  (|CallParentFunction))
