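; Wild West wooden windmill building prop that animates its head and blades and fades out its creaking audio on death.
; On BeginPlay, non-mobile platforms start two endless latent MoveComponentTo loops through a Sequence, with no Tick or Timeline.
; One sways the Z-rotation element back and forth with easing, while the other spins the blades linearly by repeatedly decrementing a Rotation variable.
; The death effects event calls the parent, then quickly fades the creaking sound.
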(event EventBeginPlay
  (bind _wildwest_windmill_zrotationelement (Variables|Default|GetWildWest_Windmill_ZRotationElement))
  (if (not (Fortnite|IsMobilePlatform))
    (Utilities|FlowControl|Sequence
      (:then_0
        (Components|MoveComponentTo _wildwest_windmill_zrotationelement "-0.004860,-6.729903,1070.704102" "0,3.000000, 0" true true 2 true :!label _label)
        (Components|MoveComponentTo _wildwest_windmill_zrotationelement "-0.004860,-6.729903,1070.704102" "0,-3.000000, 0" true true 2 true)
        (goto _label))
      (:then_1
        (Components|MoveComponentTo (Variables|Default|GetWildWest_Windmill_Blades_WildWestWindmill_Blades) "0.001420,32.140686,70.983826" (Math|Rotator|MakeRotator 0.0 (stale _output_get)) false false 1 true :!label _label_2)
        (bind _output_get (Variables|Default|SetRotation (+ (Variables|Default|GetRotation) -20)))
        (goto _label_2)))))

(event EventOnDeathPlayEffects
  (params (Damage :type Float%20%28single-precision%29) (DamageTags :type Gameplay%20Tag%20Container%20Structure%20%28by%20ref%29) (Momentum :type Vector :default "0, 0, 0") (HitInfo :type Hit%20Result%20Structure%20%28by%20ref%29) (InstigatedBy :type Fort%20Pawn%20Object%20Reference) (DamageCauser :type Actor%20Object%20Reference) (EffectContext :type Gameplay%20Effect%20Context%20Handle%20Structure))
  ;[1 1 1] Quickly fade out audio
  (|CallParentFunction Damage DamageTags Momentum HitInfo InstigatedBy DamageCauser EffectContext)
  (Audio|Components|Audio|FadeOut (Variables|Default|Getwindmill_creaking_Cue) 0.5 0.0 "Linear" :!class /Script/Engine.AudioComponent))

(fn ConstructionScript
  (|CallParentFunction))
