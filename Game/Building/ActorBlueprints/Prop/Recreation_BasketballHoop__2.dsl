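; Basketball hoop prop that detects when a thrown toy scores and triggers the scoring effect.
; On trigger-box overlap, the authority checks the actor for the basketball score gameplay tag, executes the score cue, forces a net update and calls PlayScoringCueAndDestruct through MinigameBasketballInterface.
; That call receives start and end points for lerping through the net, offset along the hoop's up vector by a distance scaled with mesh Z scale.
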(event OnComponentBeginOverlap(HoopTriggerBox)
  (params (OverlappedComponent :type Primitive%20Component%20Object%20Reference) (OtherActor :type Actor%20Object%20Reference) (OtherComp :type Primitive%20Component%20Object%20Reference) (OtherBodyIndex :type Integer) (bFromSweep :type Boolean) (SweepResult :type Hit%20Result%20Structure%20%28by%20ref%29))
  (bind (_startposition _endposition) (CallFunction|GetHoopLerpPositions))
  (Networking|SwitchHasAuthority
    (:Authority
      ;[1 0.1269 0.916] Throwingtoy Gameplay, play a cue and destroy the actor when a player manages to score. We need to delay before destroying to ensure that it plays
      (Utilities|IsValid OtherActor
        (:Is%20Valid
          (bind _returnvalue (GameplayTags|HasMatchingGameplayTag(Message) OtherActor "(TagName=\"Creative.Event.ThrowToy.Score.Basketball\")"))
          (if _returnvalue
            (GameplayCue|ExecuteGameplayCue OtherActor (Variables|Default|GetScore%20Cue))
            (Networking|ForceNetUpdate OtherActor)
            (bind _asminigame_basketball_interface (Utilities|Casting|CastToMinigameBasketballInterface OtherActor))
            (Class|MinigameBasketballInterface|PlayScoringCueAndDestruct _asminigame_basketball_interface _startposition _endposition)))))))

(fn GetHoopLerpPositions
  (bind _staticmeshcomponent (Variables|Debug|GetStaticMeshComponent))
  (bind _hooptriggerbox (Variables|Default|GetHoopTriggerBox))
  ;[1 0.7981 0.2486] Returns the start and end position of where and actor should lerp through the net
  (Utilities|IsValid _staticmeshcomponent
    (:Is%20Valid
      (Utilities|IsValid _hooptriggerbox
        (:Is%20Valid
          (return (Transformation|GetWorldLocation _hooptriggerbox) (+ (Transformation|GetWorldLocation _hooptriggerbox) (* (Transformation|GetActorUpVector) (* (Variables|Default|GetHoopLerpDistance) (.z (Transformation|GetWorldScale _staticmeshcomponent)))))))))))

(fn ConstructionScript)
