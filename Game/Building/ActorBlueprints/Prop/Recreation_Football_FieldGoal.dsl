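; Football field goal prop that detects thrown toy footballs entering its trigger box and announces the scoring distance.
; On authority, the overlap measures horizontal meters from the ball's spawn to the goal, plays the ball's scoring cue and silently kills it.
; If toy messaging is enabled, it sets a message struct with thrower, distance and serial number.
; On clients, OnRep_DisplayMessage prints it to the message feed, widening the audience from squad to team for long kicks.
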(event OnComponentBeginOverlap(FieldGoalTriggerBox)
  (params (OverlappedComponent :type Primitive%20Component%20Object%20Reference) (OtherActor :type Actor%20Object%20Reference) (OtherComp :type Primitive%20Component%20Object%20Reference) (OtherBodyIndex :type Integer) (bFromSweep :type Boolean) (SweepResult :type Hit%20Result%20Structure%20%28by%20ref%29))
  (Networking|SwitchHasAuthority
    (:Authority
      (bind _asb_toy_football_base (Utilities|Casting|CastToB_ToyFootball_Base OtherActor))
      (bind _meterstraveled (CallFunction|CalcDistanceTraveled (Variables|/Game/Building/ActorBlueprints/Prop/PhysicsBall_Master.PhysicsBall_Master_C|GetInitialSpawnPosition _asb_toy_football_base)))
      (CallFunction|CreateMessage (Utilities|Casting|CastToFortPlayerState (Actor|GetOwner _asb_toy_football_base)) _meterstraveled)
      (Class|BToyFootballBase|PlayScoringCueandDestruct _asb_toy_football_base (Transformation|GetActorLocation OtherActor) (+ (Transformation|GetActorLocation OtherActor) (* (Transformation|GetActorUpVector) -1.2e+02)))
      (Gameplay|SilentDie _asb_toy_football_base false))))

(fn CalcDistanceTraveled
  (params (BallInitialLocation :type Vector))
  (bind _output_get (Variables|Default|SetMetersTraveled (/ (Math|Vector|VectorLengthXY (- (Transformation|GetActorLocation) BallInitialLocation)) 1e+02)))
  (return _output_get))

(fn CreateMessage
  (params (PlayerWhoThrew :type Fort%20Player%20State%20Object%20Reference) (MetersTraveled :type Float%20%28double-precision%29))
  (if (Variables|/Script/FortniteGame.FortRuntimeOptions|GetbToyMessagingEnabled (Options|GetRuntimeOptions))
    (Variables|Default|SetDisplayMessage (Utilities|Struct|MakeStructFootballFieldGoalMessage PlayerWhoThrew MetersTraveled (Variables|Default|GetMessageSN)))
    (Variables|Default|SetMessageSN (Math|Byte|Byte+Byte (Variables|Default|GetMessageSN) 1))))

(fn OnRep_DisplayMessage
  (bind _returnvalue (BlueprintContext|GetFortToyContext))
  (bind (_displaymessage_playerwhothrew_22_ef5d780b4020e911ab429d9ba6b977ba _displaymessage_fieldgoaldistance_27_a600d2a048cb01620b914bb4107a6cd5) (Utilities|Struct|BreakStructFootballFieldGoalMessage (Variables|GetDisplayMessage_PlayerWhoThrew_22_EF5D780B4020E911AB429D9BA6B977BA)))
  (Networking|SwitchHasAuthority
    (:Remote
      (Utilities|IsValid _displaymessage_playerwhothrew_22_ef5d780b4020e911ab429d9ba6b977ba
        (:Is%20Valid
          (bind _returnvalue_2 (Class|FortToyContext|FormatPlayerNameforDisplay (BlueprintContext|GetFortToyContext) _displaymessage_playerwhothrew_22_ef5d780b4020e911ab429d9ba6b977ba "ToyOwner"))
          (bind _newparam (CallFunction|PickFormatString))
          (Utilities|IsValid _returnvalue
            (:Is%20Valid
              (Class|FortToyContext|ConditionalPrinttoMessageFeed _returnvalue _displaymessage_playerwhothrew_22_ef5d780b4020e911ab429d9ba6b977ba (Transformation|GetActorLocation) (select (> _displaymessage_fieldgoaldistance_27_a600d2a048cb01620b914bb4107a6cd5 (Variables|Default|GetMinimumDistanceforTeamRelevance)) "SquadAndTeamMembers" "SquadMembersOnly") (Utilities|Text|FormatText _newparam _returnvalue_2 (Math|Float|Round _displaymessage_fieldgoaldistance_27_a600d2a048cb01620b914bb4107a6cd5))))))))))

(fn PickFormatString
  (return (Variables|Default|GetFieldGoalMessage)))

(fn ConstructionScript)
