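; Actor Blueprint for a park swing prop that reacts physically to point damage.
; On damage, it checks whether the struck bone of its skeletal mesh is simulating physics.
; If so, it applies a large impulse at the hit location against the hit normal; otherwise it plays a swing chain impact sound there.
; The collision hit event and construction script are empty.
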(event Game|Damage|EventPointDamage
  (params (Damage :type Float%20%28single-precision%29) (DamageType :type Damage%20Type%20Object%20Reference) (HitLocation :type Vector :default "0, 0, 0") (HitNormal :type Vector :default "0, 0, 0") (HitComponent :type Primitive%20Component%20Object%20Reference) (BoneName :type Name :default "None") (ShotFromDirection :type Vector :default "0, 0, 0") (InstigatedBy :type Controller%20Object%20Reference) (DamageCauser :type Actor%20Object%20Reference) (HitInfo :type Hit%20Result%20Structure%20%28by%20ref%29))
  (bind _skeletalmesh1 (Variables|Recreation_Park_Swing03|GetSkeletalMesh1))
  (if (Physics|IsSimulatingPhysics _skeletalmesh1 BoneName)
    (Physics|AddImpulseatLocation _skeletalmesh1 (* HitNormal "-99999.000000,-99999.000000,-99999.000000") HitLocation BoneName)
    (Audio|PlaySoundatLocation "/Game/Sounds/Fort_GamePlay_Sounds/Storms/swing_chain_impact_Cue.swing_chain_impact_Cue" HitLocation)))

(event Collision|EventHit
  (params (MyComp :type Primitive%20Component%20Object%20Reference) (Other :type Actor%20Object%20Reference) (OtherComp :type Primitive%20Component%20Object%20Reference) (bSelfMoved :type Boolean) (HitLocation :type Vector :default "0, 0, 0") (HitNormal :type Vector :default "0, 0, 0") (NormalImpulse :type Vector :default "0, 0, 0") (Hit :type Hit%20Result%20Structure%20%28by%20ref%29)))

(fn ConstructionScript)
