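; Diving board prop that bounces pawns and hoverboards upward when they touch its trigger box.
; Overlapping FortPawns launch upward with their vertical velocity inverted plus a fixed boost, ignore their next fall damage, and send a diving board quest stat event.
; Other actors fall back to a hoverboard launch attempt.
; A successful launch calls a replicated SFX event that plays the sound at the board on cosmetic-capable instances.
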(event OnComponentBeginOverlap(TouchBox)
  (params (OverlappedComponent :type Primitive%20Component%20Object%20Reference) (OtherActor :type Actor%20Object%20Reference) (OtherComp :type Primitive%20Component%20Object%20Reference) (OtherBodyIndex :type Integer) (bFromSweep :type Boolean) (SweepResult :type Hit%20Result%20Structure%20%28by%20ref%29))
  (bind _asfort_pawn (Utilities|Casting|CastToFortPawn OtherActor)
    (:then
      (Character|LaunchCharacter _asfort_pawn (Math|Vector|MakeVector 0.0 0.0 (+ (* (.z (Transformation|GetVelocity _asfort_pawn)) -1) 2e+03)) false false)
      (CallFunction|PlayReplicatedDivingBoardSFX)
      (Variables|/Script/FortniteGame.FortPawn|SetbIgnoreNextFallingDamage true _asfort_pawn)
      ;[1 1 1] Complex Custom Event for Diving Board Quests
      (Quest|SendComplexCustomStatEvent (Quests|GetQuestManager (Fortnite|GetFortPlayerControllerfromActor _asfort_pawn) "Athena") 0 "(GameplayTags=((TagName=\"Athena.Quests.DivingBoard\")))" "(GameplayTags=)" 1))
    (:CastFailed
      (bind _success (Vehicles|LaunchHoverboard OtherActor "0, 0,2000.000000" "DivingBoard"))
      (if _success
        (CallFunction|PlayReplicatedDivingBoardSFX)))))

(event Custom|Play%20Replicated%20Diving%20Board%20SFX
  (Networking|CanExecuteCosmeticEvents
    (:True
      (Audio|PlaySoundatLocation (Variables|Default|GetSound) (Transformation|GetActorLocation)))))

(fn OnRep_Sound)

(fn ConstructionScript)
