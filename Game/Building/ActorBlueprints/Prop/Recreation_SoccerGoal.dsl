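; Soccer goal prop Actor Blueprint in the building system's recreation props.
; It declares only an empty Destroyed event and an empty Construction Script, with no custom logic of its own.
; Any behavior or appearance comes from its parent class or default settings, which this graph does not show.
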(event EventDestroyed)

(fn ConstructionScript)
