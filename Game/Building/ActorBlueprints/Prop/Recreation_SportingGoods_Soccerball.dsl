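; Soccer ball recreation prop Blueprint that plays a hit sound when something overlaps its collision sphere.
; The overlap event passes through a DoOnce gate and spawns the sound attached to the Ball component.
; After a short delay it calls the Reset Ball Hit Sound custom event, which leads back into DoOnce, apparently as a cooldown against rapid repeat sounds.
; The construction script is empty.
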(event OnComponentBeginOverlap(SphereCollision)
  (params (OverlappedComponent :type Primitive%20Component%20Object%20Reference) (OtherActor :type Actor%20Object%20Reference) (OtherComp :type Primitive%20Component%20Object%20Reference) (OtherBodyIndex :type Integer) (bFromSweep :type Boolean) (SweepResult :type Hit%20Result%20Structure%20%28by%20ref%29))
  (Utilities|FlowControl|DoOnce false
    (:Completed
      (Audio|SpawnSoundAttached (Variables|Default|GetBall%20Hit%20Sound) (Variables|Default|GetBall))
      (Utilities|FlowControl|Delay 0.2)
      (CallFunction|ResetBallHitSound))))

(event Custom|Reset%20Ball%20Hit%20Sound
  (Utilities|FlowControl|DoOnce false
    (:Completed
      (Audio|SpawnSoundAttached (Variables|Default|GetBall%20Hit%20Sound) (Variables|Default|GetBall))
      (Utilities|FlowControl|Delay 0.2)
      (CallFunction|ResetBallHitSound))))

(fn ConstructionScript)
