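; Rubber duck building prop that can drop Duct Tape ingredient pickups when destroyed.
; On BeginPlay it caches its location, if loot drops are allowed.
; On server-side death, a DoOnce-guarded loop makes six independent 50% drop rolls, each success spawning one pickup slightly above that cached start location rather than where the duck died.
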(event EventBeginPlay
  (if (Variables|Default|GetAllowLootDrop)
    (Variables|Default|SetStartLoc (Transformation|GetActorLocation self))))

(event EventOnDeathServer
  (params (Damage :type Float%20%28single-precision%29) (DamageTags :type Gameplay%20Tag%20Container%20Structure%20%28by%20ref%29) (Momentum :type Vector :default "0, 0, 0") (HitInfo :type Hit%20Result%20Structure%20%28by%20ref%29) (InstigatedBy :type Controller%20Object%20Reference) (DamageCauser :type Actor%20Object%20Reference) (EffectContext :type Gameplay%20Effect%20Context%20Handle%20Structure))
  (if (Variables|Default|GetAllowLootDrop)
    (Utilities|FlowControl|DoOnce false
      (:Completed
        (for _ (range 6)
          (bind _dropresult (Class|MissionBlueprintFunctionLibrary|DropChance 5e+01))
          (if _dropresult
            (Inventory|SpawnPickupInWorldForInstancedLoot "/Game/Items/Ingredients/Ingredient_Duct_Tape.Ingredient_Duct_Tape" 1 (+ (Variables|Default|GetStartLoc) "0, 0,25.000000") "0.000000,0.000000,1.000000" 2)))))))

(fn ConstructionScript)
