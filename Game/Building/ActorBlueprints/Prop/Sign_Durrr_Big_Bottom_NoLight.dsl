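; Building prop blueprint for the big bottom piece of a Durrr sign, named as the no-light variant, with an optional sparks effect.
; The construction script reads the SparksActive variable and sets both auto-activation and visibility of the Sparks component to match.
; The visibility change does not propagate to child components.
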(fn ConstructionScript
  (bind _sparks (Variables|Default|GetSparks))
  (if (Variables|Sparks|GetSparksActive)
    (Components|Activation|SetAutoActivate _sparks true)
    (Rendering|SetVisibility _sparks true false)
    (else
      (Components|Activation|SetAutoActivate _sparks false)
      (Rendering|SetVisibility _sparks false false))))
