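; Busted motel sign building prop that flickers its lighting and emits an electrical effect when damaged.
; On damage, a timeline drives a light intensity that scales the fully lit color into the "emissive multiplier" parameter of the mesh's dynamic material.
; A retriggerable delay then sets the timeline to a fixed time after hits stop, while the Electricity component from Sign_Motel_2 is activated on each hit.
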(event EventOnDamagePlayEffects
  (params (Damage :type Float%20%28single-precision%29) (DamageTags :type Gameplay%20Tag%20Container%20Structure%20%28by%20ref%29) (Momentum :type Vector :default "0, 0, 0") (HitInfo :type Hit%20Result%20Structure%20%28by%20ref%29) (InstigatedBy :type Fort%20Pawn%20Object%20Reference) (DamageCauser :type Actor%20Object%20Reference) (EffectContext :type Gameplay%20Effect%20Context%20Handle%20Structure))
  (Utilities|FlowControl|Sequence
    (:then_0
      (bind (_direction _lightintesity) (Timeline|Timeline_0 Play :NewTime 0.3)
        (:Update
          (bind _returnvalue (CallFunction|GetorCreateMIDOnDefaultMesh 0))
          (Rendering|Material|SetVectorParameterValue _returnvalue "emissive multiplier" (Math|Conversions|ToLinearColor(Vector) (* (Variables|Default|Getcolor%20of%20sign%20when%20fully%20lit) _lightintesity)) :!class /Script/Engine.MaterialInstanceDynamic))))
    (:then_1
      (Utilities|FlowControl|RetriggerableDelay 1.2)
      (Timeline|Timeline_0 SetNewTime))
    (:then_2
      (Components|Activation|Activate (Variables|Sign_Motel_2|GetElectricity) false))))

(fn ConstructionScript)
