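; Reactive dance floor prop whose material responds to player pawns standing inside its detection volume.
; Overlap events maintain a pawn array; with one to three dancers, a looping timer writes their world locations into vector parameters on the mesh's dynamic material.
; At four or more, position updates stop and a timeline raises a full-panel glow, reversing when the count drops below four.
; Departures reset unused dancer parameters to clear stale positions.
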(event OnComponentBeginOverlap(CharacterDetection)
  (params (OverlappedComponent :type Primitive%20Component%20Object%20Reference) (OtherActor :type Actor%20Object%20Reference) (OtherComp :type Primitive%20Component%20Object%20Reference) (OtherBodyIndex :type Integer) (bFromSweep :type Boolean) (SweepResult :type Hit%20Result%20Structure%20%28by%20ref%29))
  ;[0 0 0] Track the number of dancers and reinit when needed
  (bind _asfort_player_pawn (Utilities|Casting|CastToFortPlayerPawn OtherActor))
  (Utilities|Array|Add (Variables|Default|GetDanceFloorPawnArray) _asfort_player_pawn)
  (CallFunction|NumberOfPlayersUpdated))

(event OnComponentEndOverlap(CharacterDetection)
  (params (OverlappedComponent :type Primitive%20Component%20Object%20Reference) (OtherActor :type Actor%20Object%20Reference) (OtherComp :type Primitive%20Component%20Object%20Reference) (OtherBodyIndex :type Integer))
  ;[0 0 0] Track the number of dancers and reinit when needed
  (bind _asfort_player_pawn (Utilities|Casting|CastToFortPlayerPawn OtherActor))
  (Utilities|Array|RemoveItem (Variables|Default|GetDanceFloorPawnArray) _asfort_player_pawn)
  (CallFunction|RemoveOldMIDData)
  (CallFunction|NumberOfPlayersUpdated))

(event Custom|NumberOfPlayersUpdated
  (bind _dancefloorpawnarray (Variables|Default|GetDanceFloorPawnArray))
  (bind _returnvalue (Utilities|Array|Length _dancefloorpawnarray))
  (bind _dancefloorpawnarray_2 (Variables|Default|GetDanceFloorPawnArray))
  (bind _returnvalue_2 (Utilities|Array|Length _dancefloorpawnarray_2))
  (bind _previous_number_of_dancers (Variables|Default|GetPrevious%20number%20of%20dancers))
  ;[0 0 0] Trigger MID updates
  (Utilities|FlowControl|Sequence
    (:then_0
      (if (and (< _returnvalue_2 4) (> _returnvalue_2 0))
        (Variables|Default|SetActively%20Update true)
        (Utilities|Time|SetTimerbyFunctionName 0 "UpdatePlayerPositionsWithinTheMID" (Variables|Default|GetActor%20Position%20Update%20Frequency) true)
        (else
          (Variables|Default|SetActively%20Update false)
          (Utilities|Time|ClearTimerbyFunctionName 0 "UpdatePlayerPositionsWithinTheMID"))))
    (:then_1
      (if (and (< _previous_number_of_dancers 4) (> _returnvalue 3))
        (bind (_direction _fullpanelglow) (Timeline|OverallGlow Play)
          (:Update
            (bind _returnvalue_3 (CallFunction|GetorCreateMIDOnDefaultMesh 0))
            (Rendering|Material|SetScalarParameterValue _returnvalue_3 "FullPanelGlow" _fullpanelglow :!class /Script/Engine.MaterialInstanceDynamic)))
        (elif (and (> _previous_number_of_dancers 3) (< _returnvalue 4))
          (Timeline|OverallGlow Reverse))))
    (:then_2
      (Variables|Default|SetPrevious%20number%20of%20dancers _returnvalue_2))))

(fn RemoveOldMIDData
  (bind _dancefloorpawnarray (Variables|Default|GetDanceFloorPawnArray))
  (bind _returnvalue (Utilities|Array|Length _dancefloorpawnarray))
  (bind _returnvalue_2 (CallFunction|GetorCreateMIDOnDefaultMesh 0))
  (if (< _returnvalue 1)
    ;[0 0 0] Reset variables to remove previous model position data when a player leaves the volume
    (Rendering|Material|SetVectorParameterValue :self _returnvalue_2 :ParameterName "DancerVariable1" :!class /Script/Engine.MaterialInstanceDynamic)
    (elif (< _returnvalue 2)
      _
      (elif (< _returnvalue 3)
        (Rendering|Material|SetVectorParameterValue :self _returnvalue_2 :ParameterName "DancerVariable3" :!class /Script/Engine.MaterialInstanceDynamic :!label _label))))
  (Rendering|Material|SetVectorParameterValue :self _returnvalue_2 :ParameterName "DancerVariable2" :!class /Script/Engine.MaterialInstanceDynamic)
  (goto _label))

(fn UpdatePlayerPositionsWithinTheMID
  (bind _dancefloorpawnarray (Variables|Default|GetDanceFloorPawnArray))
  (bind _returnvalue (Utilities|Array|Length _dancefloorpawnarray))
  (bind _returnvalue_2 (CallFunction|GetorCreateMIDOnDefaultMesh 0))
  ;[0 0 0] Update material positions as needed
  (if (> _returnvalue 0)
    (Rendering|Material|SetVectorParameterValue _returnvalue_2 "DancerVariable1" (Utilities|Struct|MakeLinearColor (.x (Transformation|GetActorLocation (Utilities|Array|Get(acopy) _dancefloorpawnarray))) (.y (Transformation|GetActorLocation (Utilities|Array|Get(acopy) _dancefloorpawnarray))) (.z (Transformation|GetActorLocation (Utilities|Array|Get(acopy) _dancefloorpawnarray))) 1.0) :!class /Script/Engine.MaterialInstanceDynamic)
    (if (> _returnvalue 1)
      (Rendering|Material|SetVectorParameterValue _returnvalue_2 "DancerVariable2" (Utilities|Struct|MakeLinearColor (.x (Transformation|GetActorLocation (Utilities|Array|Get(acopy) _dancefloorpawnarray 1))) (.y (Transformation|GetActorLocation (Utilities|Array|Get(acopy) _dancefloorpawnarray 1))) (.z (Transformation|GetActorLocation (Utilities|Array|Get(acopy) _dancefloorpawnarray 1))) 1.0) :!class /Script/Engine.MaterialInstanceDynamic)
      (if (> _returnvalue 2)
        (Rendering|Material|SetVectorParameterValue _returnvalue_2 "DancerVariable3" (Utilities|Struct|MakeLinearColor (.x (Transformation|GetActorLocation (Utilities|Array|Get(acopy) _dancefloorpawnarray 2))) (.y (Transformation|GetActorLocation (Utilities|Array|Get(acopy) _dancefloorpawnarray 2))) (.z (Transformation|GetActorLocation (Utilities|Array|Get(acopy) _dancefloorpawnarray 2))) 1.0) :!class /Script/Engine.MaterialInstanceDynamic)))))

(fn ConstructionScript
  (Components|Timeline|SetPlayRate (Variables|Stage_DanceFloor_Reactive|GetOverallGlow) 3.0))
