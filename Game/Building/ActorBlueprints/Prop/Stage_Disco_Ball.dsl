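; A stage disco ball prop that optionally shows light ray effects and spins.
; The construction script sets the Niagara disco ball FX to auto-activate only when the DiscoEffects flag is on.
; On BeginPlay, where cosmetic events can run, it passes the ray clip toggle and clip height to the FX.
; It also drives spin through the body material's Rotation Speed parameter, randomized when effects are on and zero otherwise.
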(event EventBeginPlay
  (bind _fx_discoballfx (Variables|Default|GetFX_DiscoBallFX))
  (Networking|CanExecuteCosmeticEvents
    (:True
      (Niagara|SetNiagaraVariable(Float) _fx_discoballfx "DoClip" (Variables|Default|GetClip%20Rays))
      (Niagara|SetNiagaraVariable(Float) _fx_discoballfx "Clip Height" (Variables|Default|GetRays%20Clip%20Height))
      (if (Variables|Default|GetDiscoEffects)
        (Rendering|Material|SetScalarParameterValueonMaterials (Variables|Default|GetBody) "Rotation Speed" (Math|Random|RandomFloatinRange 0.05 0.1))
        (else
          (Rendering|Material|SetScalarParameterValueonMaterials (Variables|Default|GetBody) "Rotation Speed" 0.0))))))

(fn ConstructionScript
  (bind _fx_discoballfx (Variables|Default|GetFX_DiscoBallFX))
  (if (Variables|Default|GetDiscoEffects)
    (Components|Activation|SetAutoActivate _fx_discoballfx true)
    (else
      (Components|Activation|SetAutoActivate _fx_discoballfx false))))
