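; Urban street light prop Blueprint whose construction script limits shadow casting on its two spot lights.
; After the parent script runs, Nintendo Switch disables shadows on SpotLight only, leaving SpotLight1 untouched.
; On other platforms, when both lights cast shadows, the WhichLightCastsShadow enum keeps shadows on the first, the second or neither.
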(fn ConstructionScript
  (|CallParentFunction)
  (if (TouchPlatform|IsNintendoSwitch)
    (Rendering|Components|Light|SetCastShadows (Variables|Default|GetSpotLight) false)
    (elif (and (Variables|/Script/Engine.LightComponentBase|GetCastShadows (Variables|Default|GetSpotLight)) (Variables|/Script/Engine.LightComponentBase|GetCastShadows (Variables|Default|GetSpotLight1)))
      (switch Utilities|FlowControl|Switch|Switchon2SidedEnum (Variables|Default|GetWhichLightCastsShadow)
        (:NewEnumerator0
          (Rendering|Components|Light|SetCastShadows (Variables|Default|GetSpotLight) true)
          (Rendering|Components|Light|SetCastShadows (Variables|Default|GetSpotLight1) false))
        (:NewEnumerator1
          (Rendering|Components|Light|SetCastShadows (Variables|Default|GetSpotLight1) true)
          (Rendering|Components|Light|SetCastShadows (Variables|Default|GetSpotLight) false))
        (:NewEnumerator2
          (Rendering|Components|Light|SetCastShadows (Variables|Default|GetSpotLight1) false)
          (Rendering|Components|Light|SetCastShadows (Variables|Default|GetSpotLight) false))))))
