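; Streetlight prop Blueprint that switches its spotlight, light-ray mesh and fog card according to the time of day.
; BeginPlay, day-phase changes and the construction script all call SetVisualParmeters, which gates cosmetic updates on LightActive and a time-of-day flag.
; It passes colors to the meshes via custom primitive data and sizes the ray from cone angles and capped attenuation radius.
; Ray roll and a data float are randomized unless Deterministic; a debug event refreshes all instances.
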(event Custom|UpdateVisualsOnAllActorsInMap
  ;[1 1 1] Editor Only - Debug Purposes
  (bind _outactors (Actor|GetAllActorsOfClass "/Game/Building/ActorBlueprints/Prop/Streetlight_BlackMonday_01.Streetlight_BlackMonday_01_C"))
  (for _array_element _outactors
    (CallFunction|SetVisualParmeters _array_element false false)))

(event EventBeginPlay
  ;[1 1 1] Begin Play
  (CallFunction|SetVisualParmeters (not (Utilities|Enum|Equal(Enum) (TimeOfDay|GetCurrentDayPhase :!class /Script/FortniteGame.FortKismetLibrary) "Night"))))

(event Custom|UpdateVisuals
  ;[1 1 1] Begin Play
  (CallFunction|SetVisualParmeters (not (Utilities|Enum|Equal(Enum) (TimeOfDay|GetCurrentDayPhase :!class /Script/FortniteGame.FortKismetLibrary) "Night"))))

(event EventOnDayPhaseChanged
  (params (CurrentDayPhase :type EFortDayPhase%20Enum :default "Morning") (PreviousDayPhase :type EFortDayPhase%20Enum :default "Morning") (bAtCreation :type Boolean))
  ;[1 1 1] Update lights based on time of day
  (Networking|CanExecuteCosmeticEvents
    (:True
      (switch Utilities|FlowControl|Switch|SwitchonEFortDayPhase CurrentDayPhase
        (:Morning
          _)
        (:Day
          (CallFunction|SetVisualParmeters true))
        (:Evening
          _)
        (:Night
          _))
      (CallFunction|SetVisualParmeters false))))

(fn GetLightRayScale
  (return (Math|Vector|MakeVector (/ (Math|Float|Min(Float) (Variables|Default|GetMaxLightrayLength) (Variables|Default|GetLightAttenuationRadius)) 2e+02) (* (/ (+ (Variables|Default|GetLightInnerConeAngle) (Variables|Default|GetLightOuterConeAngle)) 4.0) (/ (Math|Float|Min(Float) (Variables|Default|GetMaxLightrayLength) (Variables|Default|GetLightAttenuationRadius)) 4e+03)) (* (/ (+ (Variables|Default|GetLightInnerConeAngle) (Variables|Default|GetLightOuterConeAngle)) 4.0) (/ (Math|Float|Min(Float) (Variables|Default|GetMaxLightrayLength) (Variables|Default|GetLightAttenuationRadius)) 4e+03)))))

(fn SetVisualParmeters
  (params (DisabledForTimeOfDay :type Boolean) (Deterministic :type Boolean))
  (bind (_fogcardcolorandopacity_r _fogcardcolorandopacity_g _fogcardcolorandopacity_b _fogcardcolorandopacity_a) (Math|Color|BreakColor (Variables|GetFogCardColorAndOpacity_R)))
  (bind (_lightcolor_r _lightcolor_g _lightcolor_b _lightcolor_a) (Math|Color|BreakColor (Variables|GetLightColor_R)))
  (Networking|CanExecuteCosmeticEvents
    (:True
      (if (and (Variables|Default|GetLightActive) (not (Variables|GetDisabledForTimeOfDay)))
        (Transformation|SetRelativeRotation (Variables|Default|GetLightRay) (Math|Rotator|MakeRotator (select Deterministic 0.0 (Math|Random|RandomFloatinRange 0.0 3.6e+02))) false false)
        (Development|SetHiddeninGame (Variables|Default|GetLightRay) false false)
        (Rendering|Components|Light|SetIntensity (Variables|Default|GetSpotLight) (Variables|Default|GetLightIntensity))
        (Transformation|SetRelativeScale3D (Variables|Default|GetLightRay) (CallFunction|GetLightRayScale))
        (Rendering|Material|SetCustomPrimitiveDataFloat (Variables|Default|GetLightRay) 27 (select Deterministic 0.0 (Math|Random|RandomFloatinRange 0.0 1.0)))
        (Development|SetHiddeninGame (Variables|Default|GetFogCard) (Math|Color|NearEqual(LinearColor) (Variables|Default|GetFogCardColorAndOpacity) "(R=0.000000,G=0.000000,B=0.000000,A=0.000000)" 0.01) false)
        (Rendering|Material|SetCustomPrimitiveDataVector4 (Variables|Default|GetFogCard) (Variables|Default|GetMeshEmissiveCustomPrimitiveDataIndex) (Math|Vector4|MakeVector4 _fogcardcolorandopacity_r _fogcardcolorandopacity_g _fogcardcolorandopacity_b _fogcardcolorandopacity_a))
        (Rendering|Material|SetCustomPrimitiveDataVector4 (Variables|Default|GetLightRay) (Variables|Default|GetMeshEmissiveCustomPrimitiveDataIndex) (Math|Vector4|MakeVector4 _lightcolor_r _lightcolor_g _lightcolor_b _lightcolor_a))
        (Rendering|Components|Light|SetLightColor (Variables|Default|GetSpotLight) (Variables|Default|GetLightColor) true)
        (Rendering|Lighting|SetAttenuationRadius (Variables|Default|GetSpotLight) (Variables|Default|GetLightAttenuationRadius))
        (Rendering|Lighting|SetInnerConeAngle (Variables|Default|GetSpotLight) (Variables|Default|GetLightInnerConeAngle))
        (Rendering|Lighting|SetOuterConeAngle (Variables|Default|GetSpotLight) (Variables|Default|GetLightOuterConeAngle))
        (else
          (Rendering|Components|Light|SetIntensity (Variables|Default|GetSpotLight) 0.0)
          (Rendering|Lighting|SetAttenuationRadius (Variables|Default|GetSpotLight) 0.0)
          (Development|SetHiddeninGame (Variables|Default|GetLightRay) true false)
          (Development|SetHiddeninGame (Variables|Default|GetFogCard) true false))))))

(fn ConstructionScript
  (|CallParentFunction)
  (CallFunction|SetVisualParmeters false true))
