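; String light prop that builds a chain of spline meshes along its spline and destroys itself when any of its light attachment actors dies.
; The construction script adds one mesh component per spline segment, with optional override material and per-point scaling.
; Segments after the first can auto-smooth tangents by scaling their direction with segment length and Tension.
; BeginPlay binds each valid attachment actor's OnDied event to the self-destroy handler.
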(event EventBeginPlay
  (if (> (Utilities|Array|Length (Variables|Default|GetLightAttachmentActors)) 0)
    (for _array_element (Variables|Default|GetLightAttachmentActors)
      (Utilities|IsValid _array_element
        (:Is%20Valid
          (EventDispatchers|BindEventtoOnDied _array_element (delegate Custom|IReallyDied)))))))

(event Custom|IReallyDied
  (params (DamagedActor :type Actor%20Object%20Reference) (Damage :type Float%20%28single-precision%29) (InstigatedBy :type Controller%20Object%20Reference) (DamageCauser :type Actor%20Object%20Reference) (HitLocation :type Vector) (FHitComponent :type Primitive%20Component%20Object%20Reference) (BoneName :type Name) (Momentum :type Vector))
  (Utilities|IsValid DamagedActor
    (:Is%20Valid
      (Actor|DestroyActor self))))

(fn ConstructionScript
  (CallFunction|ValidateSplineMesh))

(fn ValidateSplineMesh
  (bind (_bchannel0 _bchannel1 _bchannel2) (Utilities|Struct|BreakLightingChannels (Variables|/Script/Engine.PrimitiveComponent|GetLightingChannels (LocalVariables|Getl_splineMeshComp))))
  (bind (_location _tangent) (Spline|GetLocationandTangentatSplinePoint (Variables|Default|GetSpline) 1 "Local"))
  (bind (_location_2 _tangent_2) (Spline|GetLocationandTangentatSplinePoint (Variables|Default|GetSpline) 0 "Local"))
  (bind (_location_3 _tangent_3) (Spline|GetLocationandTangentatSplinePoint (Variables|Default|GetSpline) (LocalVariables|Getl_nextIndex) "Local"))
  (bind (_location_4 _tangent_4) (Spline|GetLocationandTangentatSplinePoint (Variables|Default|GetSpline) (LocalVariables|Getl_currentIndex) "Local"))
  (bind _overridematerial (Variables|Default|GetOverrideMaterial))
  (bind _returnvalue (AddComponent|NODE_AddSplineMeshComponent-1))
  (bind _output_get (LocalVariables|Setl_splineMeshComp _returnvalue))
  (SplineMesh|SetStartandEnd _output_get _location_2 _tangent_2 _location _tangent true :!class /Script/Engine.SplineMeshComponent)
  (Utilities|FlowControl|Sequence
    (:then_0
      (if (Utilities|IsValid (Variables|Default|GetOverrideMaterial) :!class /Script/Engine.KismetSystemLibrary)
        (Rendering|Material|SetMaterial (LocalVariables|Getl_splineMeshComp) 0 (Variables|Default|GetOverrideMaterial))))
    (:then_1
      (if (Variables|Default|GetUseScaling)
        (SplineMesh|SetStartScale (LocalVariables|Getl_splineMeshComp) (Math|Vector2D|MakeVector2D (.y (Spline|GetScaleatSplinePoint (Variables|Default|GetSpline) 0)) (.z (Spline|GetScaleatSplinePoint (Variables|Default|GetSpline) 0))) true :!class /Script/Engine.SplineMeshComponent)
        (SplineMesh|SetEndScale (LocalVariables|Getl_splineMeshComp) (Math|Vector2D|MakeVector2D (.y (Spline|GetScaleatSplinePoint (Variables|Default|GetSpline) 1)) (.z (Spline|GetScaleatSplinePoint (Variables|Default|GetSpline) 1))) true :!class /Script/Engine.SplineMeshComponent))
      (for _index (range 1 (+ (- (Spline|GetNumberOfSplinePoints (Variables|Default|GetSpline)) 2) 1))
        (LocalVariables|Setl_currentIndex _index)
        (LocalVariables|Setl_nextIndex (+ _index 1))
        (bind _returnvalue_2 (AddComponent|NODE_AddSplineMeshComponent-0))
        (bind _output_get_2 (LocalVariables|Setl_splineMeshComp _returnvalue_2))
        (Rendering|Components|SetLightingChannels _output_get_2 _bchannel0 _bchannel1 _bchannel2)
        (Utilities|FlowControl|Sequence
          (:then_0
            (if (Utilities|IsValid _overridematerial :!class /Script/Engine.KismetSystemLibrary)
              (Rendering|Material|SetMaterial (LocalVariables|Getl_splineMeshComp) 0 _overridematerial)))
          (:then_1
            (SplineMesh|SetForwardAxis (LocalVariables|Getl_splineMeshComp) (SplineMesh|GetForwardAxis (LocalVariables|Getl_splineMeshComp)) true)
            (Rendering|SetCastShadow (LocalVariables|Getl_splineMeshComp) (Variables|/Script/Engine.PrimitiveComponent|GetCastShadow (LocalVariables|Getl_splineMeshComp)))
            (SplineMesh|SetStartandEnd (LocalVariables|Getl_splineMeshComp) _location_4 (Math|Vector|SelectVector (* (Math|Vector|Normalize _tangent_4) (* (Math|Vector|VectorLength (- _location_3 _location_4)) (Variables|Default|GetTension))) _tangent_4 (Variables|Default|GetAutoSmoothTangent)) _location_3 (Math|Vector|SelectVector (* (Math|Vector|Normalize _tangent_3) (* (Math|Vector|VectorLength (- _location_3 _location_4)) (Variables|Default|GetTension))) _tangent_3 (Variables|Default|GetAutoSmoothTangent)) true :!class /Script/Engine.SplineMeshComponent)
            (if (Variables|Default|GetUseScaling)
              (SplineMesh|SetStartScale (LocalVariables|Getl_splineMeshComp) (Math|Vector2D|MakeVector2D (.y (Spline|GetScaleatSplinePoint (Variables|Default|GetSpline) (LocalVariables|Getl_currentIndex))) (.z (Spline|GetScaleatSplinePoint (Variables|Default|GetSpline) (LocalVariables|Getl_currentIndex)))) true :!class /Script/Engine.SplineMeshComponent)
              (SplineMesh|SetEndScale (LocalVariables|Getl_splineMeshComp) (Math|Vector2D|MakeVector2D (.y (Spline|GetScaleatSplinePoint (Variables|Default|GetSpline) (LocalVariables|Getl_nextIndex))) (.z (Spline|GetScaleatSplinePoint (Variables|Default|GetSpline) (LocalVariables|Getl_nextIndex)))) true :!class /Script/Engine.SplineMeshComponent))))))))
