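; Palm tree prop actor Blueprint that can spawn a cosmetic firefly effect when play begins.
; After calling the parent BeginPlay, it continues only where cosmetic events can execute and the HasBugs flag is set.
; It spawns the palm-tree firefly particle system at the static mesh location, shifts it by BugOffset, and passes UseLightningBug into the IsLightningBug float parameter.
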(event EventBeginPlay
  (|CallParentFunction)
  (Networking|CanExecuteCosmeticEvents
    (:True
      (if (Variables|FX|GetHasBugs)
        (bind _returnvalue (Effects|SpawnEmitteratLocation :EmitterTemplate "/Game/Effects/Fort_Effects/Effects/Bugs/Fireflies/P_FireFlies_PalmTree.P_FireFlies_PalmTree" :Location (Transformation|GetWorldLocation (Variables|Debug|GetStaticMeshComponent)) :Scale "1.000000,1.000000,1.000000" :bAutoDestroy false :PoolingMethod "None" :bAutoActivateSystem true))
        (Transformation|AddRelativeLocation _returnvalue (Variables|FX|GetBugOffset) false false)
        (Effects|Components|ParticleSystem|SetFloatParameter _returnvalue "IsLightningBug" (Variables|FX|GetUseLightningBug))))))

(fn ConstructionScript)
