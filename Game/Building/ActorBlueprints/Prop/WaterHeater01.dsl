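; WaterHeater01 is a building prop actor Blueprint for a water heater that gives electrical feedback when damaged or destroyed.
; Each hit activates the electric hit effect and plays a transformer zap sound at the actor, while a DoOnce activates the damage zap effect on the first hit only.
; Bursts from that zap effect play the zap audio component, syncing sound to particles.
; On death, the zap effect is deactivated.
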(event EventOnDamagePlayEffects
  (params (Damage :type Float%20%28single-precision%29) (DamageTags :type Gameplay%20Tag%20Container%20Structure%20%28by%20ref%29) (Momentum :type Vector :default "0, 0, 0") (HitInfo :type Hit%20Result%20Structure%20%28by%20ref%29) (InstigatedBy :type Fort%20Pawn%20Object%20Reference) (DamageCauser :type Actor%20Object%20Reference) (EffectContext :type Gameplay%20Effect%20Context%20Handle%20Structure))
  (Utilities|FlowControl|Sequence
    (:then_0
      (Components|Activation|Activate (Variables|WaterHeater01|GetEffect_Electric_ON_Hit) false)
      (Audio|PlaySoundatLocation "/Game/Sounds/Fort_Traps/Wall_Electric/Transformer_Electric_Zap_Cue.Transformer_Electric_Zap_Cue" (Transformation|GetActorLocation)))
    (:then_1
      (Utilities|FlowControl|DoOnce false
        (:Completed
          (Components|Activation|Activate (Variables|WaterHeater01|GetEffect_Damage_Zap) false))))))

(event OnParticleBurst(Effect_Damage_Zap)
  (params (EventName :type Name :default "None") (EmitterTime :type Float%20%28single-precision%29) (ParticleCount :type Integer))
  (Audio|Components|Audio|Play (Variables|WaterHeater01|GetAudio_Zap) 0.0))

(event EventOnDeathPlayEffects
  (params (Damage :type Float%20%28single-precision%29) (DamageTags :type Gameplay%20Tag%20Container%20Structure%20%28by%20ref%29) (Momentum :type Vector :default "0, 0, 0") (HitInfo :type Hit%20Result%20Structure%20%28by%20ref%29) (InstigatedBy :type Fort%20Pawn%20Object%20Reference) (DamageCauser :type Actor%20Object%20Reference) (EffectContext :type Gameplay%20Effect%20Context%20Handle%20Structure))
  (Components|Activation|Deactivate (Variables|WaterHeater01|GetEffect_Damage_Zap)))

(fn ConstructionScript)
