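; Winter string lights building prop that removes itself when something it is attached to dies.
; On BeginPlay it loops over its RugAttachmentActors array and binds each valid actor's OnDied event to a custom IReallyDied handler.
; When any bound actor dies, the handler checks that the damaged actor is valid, then destroys the string lights actor.
; The construction script is empty.
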(event EventBeginPlay
  (if (> (Utilities|Array|Length (Variables|Default|GetRugAttachmentActors)) 0)
    (for _array_element (Variables|Default|GetRugAttachmentActors)
      (Utilities|IsValid _array_element
        (:Is%20Valid
          (EventDispatchers|BindEventtoOnDied _array_element (delegate Custom|IReallyDied)))))))

(event Custom|IReallyDied
  (params (DamagedActor :type Actor%20Object%20Reference) (Damage :type Float%20%28single-precision%29) (InstigatedBy :type Controller%20Object%20Reference) (DamageCauser :type Actor%20Object%20Reference) (HitLocation :type Vector) (FHitComponent :type Primitive%20Component%20Object%20Reference) (BoneName :type Name) (Momentum :type Vector))
  (Utilities|IsValid DamagedActor
    (:Is%20Valid
      (Actor|DestroyActor self))))

(fn ConstructionScript)
